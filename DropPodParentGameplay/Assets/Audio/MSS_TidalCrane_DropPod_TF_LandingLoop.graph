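# Stereo MetaSound source that plays the Tidal Crane drop pod landing loop mix, using crate descent attenuation and single-instance concurrency.
# On Play starts one wave player, whose output passes through a two-channel stereo mixer to the outputs.
# The player's On Finished feeds one input of a two-input trigger accumulator driving One Shot OnFinished.
# Prehit, delay and tone inputs are declared but unwired, as is the second mixer channel.
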MetaSoundSource /DropPodParentGameplay/Assets/Audio/MSS_TidalCrane_DropPod_TF_LandingLoop
inputs [
 Prehit {type=WaveAsset}
 PrehitDelay {type=Float}
 Time.ElapsedTime {type=Float}
 Time.GameTimeSeconds {type=Float}
 Time.PlayTime {type=Float}
 Time.TimeSinceLastPlayed {type=Float}
 Tone {type=WaveAsset}
 TriggerPrehitDelay {type=Trigger}
 UE.Source.OnPlay {type=Trigger}
]
interfaces [Fort.Time:1.0, UE.OutputFormat.Stereo:1.0, UE.Source.OneShot:1.0, UE.Source:1.1]
outputs [
 "UE.OutputFormat.Stereo.Audio:0" {type=Audio}
 "UE.OutputFormat.Stereo.Audio:1" {type=Audio}
 UE.Source.OneShot.OnFinished {type=Trigger}
]
pages [Default {
  input_defaults {
   Prehit=/DropPodParentGameplay/Assets/Audio/wav/TidalCraneDropPod_FallingDoppler
   PrehitDelay=5
   Time.ElapsedTime=0
   Time.GameTimeSeconds=0
   Time.PlayTime=0
   Time.TimeSinceLastPlayed=0
   Tone=/Game/Sounds/Fort_Test_Sounds/Tones/1000hz_Saw_Tone_Short
   TriggerPrehitDelay=null
   UE.Source.OnPlay=null
  }
  nodes [
   WavePlayer_0 = UE.Wave Player:Stereo(Play=^UE.Source.OnPlay) {also {
     defaults {
      Loop="<MetaSoundBuilderResult.SUCCEEDED: 0>"
      Loop Duration=8.5
      Wave Asset=/DropPodParentGameplay/Assets/Audio/wav/TidalCraneDropPod_LandingLoopMix
     }
     version [1,1]
    }}
   AudioMixerAudioMixerStereo2_0 = "AudioMixer.Audio Mixer (Stereo, 2)"(In 0 L=@WavePlayer_0.Out Left, In 0 R=@WavePlayer_0.Out Right) {also {defaults {Gain 1=1.5}}}
   InitVariable.Float()
   InitVariable.Trigger()
   TriggerAccumulatorTriggerAccumulate2_0 = "TriggerAccumulator.Trigger Accumulate (2)"(In 0=@WavePlayer_0.On Finished)
  ]
  links [
   {dst="@UE.OutputFormat.Stereo.Audio:0", src=AudioMixerAudioMixerStereo2_0.Out L}
   {dst="@UE.OutputFormat.Stereo.Audio:1", src=AudioMixerAudioMixerStereo2_0.Out R}
   {dst="@UE.Source.OneShot.OnFinished", src=TriggerAccumulatorTriggerAccumulate2_0.Out}
  ]
 }]
properties {
 assetImportData=&self:AssetImportData
 attenuationSettings=&/ATT_TidalCraneDropPod_TF_Crate_Descent
 bOverrideConcurrency=true
 maxDistance=16000
 numChannels=2
 outputFormat=Stereo
 sampleRate=48000
 soundClassObject=&Fort_Gadgets
 concurrencyOverrides {maxCount=1, retriggerTime=0.1}
}
