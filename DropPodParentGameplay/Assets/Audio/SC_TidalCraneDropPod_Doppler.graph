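# Sound cue for the falling Doppler sound of the Tidal Crane drop pod in the DropPodParentGameplay plugin.
# A single wave player feeds a modulator that boosts and randomizes its volume.
# The cue uses a crate descent attenuation asset and an ambient sound class, with virtualization disabled and a small cap on concurrent instances.
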SoundCue /DropPodParentGameplay/Assets/Audio/SC_TidalCraneDropPod_Doppler
root=Modulator_0
nodes [
 {id=Modulator_0, children [WavePlayer_0], properties {volumeMax=2, volumeMin=1.5}}
 {id=WavePlayer_0, properties {soundWaveAssetPtr=/DropPodParentGameplay/Assets/Audio/wav/TidalCraneDropPod_FallingDopplershort}}
]
properties {
 attenuationSettings=&/ATT_TidalCraneDropPod_Crate_Descent
 maxDistance=63000
 soundClassObject=&Fort_Ambient_Athena
 virtualizationMode=Disabled
 volumeMultiplier=1.25
 concurrencyOverrides {maxCount=4}
}
