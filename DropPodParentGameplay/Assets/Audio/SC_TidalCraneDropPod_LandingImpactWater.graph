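# Sound cue for the Tidal Crane drop pod's landing impact on water.
# A mixer layers two equal-weight random picks at boosted volume: one from the pod's water impact waves, the other from water waves borrowed from the Ostrich vehicle's FS set.
# It uses the pod's crate explosion attenuation and the gadgets sound class, with a concurrency cap and virtualization disabled.
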SoundCue /DropPodParentGameplay/Assets/Audio/SC_TidalCraneDropPod_LandingImpactWater
root=Mixer_1
nodes [
 {id=Mixer_1, children [Random_1,Random_3], properties {inputVolume [1.5,1.5]}}
 {
  id=Random_1
  children [WavePlayer_5, WavePlayer_6, WavePlayer_7, WavePlayer_8, WavePlayer_9, WavePlayer_10, WavePlayer_11, WavePlayer_12]
  properties {weights [1,1,1,1,1,1,1,1]}
 }
 {
  id=Random_3
  children [WavePlayer_17, WavePlayer_18, WavePlayer_19, WavePlayer_20, WavePlayer_21, WavePlayer_22]
  properties {weights [1,1,1,1,1,1]}
 }
 {id=WavePlayer_10, properties {soundWaveAssetPtr=/DropPodParentGameplay/Assets/Audio/wav/TidalCraneDropPod_WaterImpact_A_06}}
 {id=WavePlayer_11, properties {soundWaveAssetPtr=/DropPodParentGameplay/Assets/Audio/wav/TidalCraneDropPod_WaterImpact_A_07}}
 {id=WavePlayer_12, properties {soundWaveAssetPtr=/DropPodParentGameplay/Assets/Audio/wav/TidalCraneDropPod_WaterImpact_A_08}}
 {id=WavePlayer_17, properties {soundWaveAssetPtr=/DropPodParentGameplay/Assets/Audio/wav/Veh_Ostrich_FS_Water_01}}
 {id=WavePlayer_18, properties {soundWaveAssetPtr=/DropPodParentGameplay/Assets/Audio/wav/Veh_Ostrich_FS_Water_02}}
 {id=WavePlayer_19, properties {soundWaveAssetPtr=/DropPodParentGameplay/Assets/Audio/wav/Veh_Ostrich_FS_Water_03}}
 {id=WavePlayer_20, properties {soundWaveAssetPtr=/DropPodParentGameplay/Assets/Audio/wav/Veh_Ostrich_FS_Water_04}}
 {id=WavePlayer_21, properties {soundWaveAssetPtr=/DropPodParentGameplay/Assets/Audio/wav/Veh_Ostrich_FS_Water_05}}
 {id=WavePlayer_22, properties {soundWaveAssetPtr=/DropPodParentGameplay/Assets/Audio/wav/Veh_Ostrich_FS_Water_06}}
 {id=WavePlayer_5, properties {soundWaveAssetPtr=/DropPodParentGameplay/Assets/Audio/wav/TidalCraneDropPod_WaterImpact_A_01}}
 {id=WavePlayer_6, properties {soundWaveAssetPtr=/DropPodParentGameplay/Assets/Audio/wav/TidalCraneDropPod_WaterImpact_A_02}}
 {id=WavePlayer_7, properties {soundWaveAssetPtr=/DropPodParentGameplay/Assets/Audio/wav/TidalCraneDropPod_WaterImpact_A_03}}
 {id=WavePlayer_8, properties {soundWaveAssetPtr=/DropPodParentGameplay/Assets/Audio/wav/TidalCraneDropPod_WaterImpact_A_04}}
 {id=WavePlayer_9, properties {soundWaveAssetPtr=/DropPodParentGameplay/Assets/Audio/wav/TidalCraneDropPod_WaterImpact_A_05}}
]
properties {
 attenuationSettings=&/ATT_TidalCraneDropPod_Crate_Explosion
 maxDistance=43000
 soundClassObject=&Fort_Gadgets
 virtualizationMode=Disabled
 concurrencyOverrides {maxCount=4}
}
