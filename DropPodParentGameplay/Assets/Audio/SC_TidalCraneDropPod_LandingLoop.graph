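# Looping sound cue for the Tidal Crane drop pod's descent and landing, in the gadgets sound class with crate-descent attenuation.
# Six looping falling, flame and flare layers are weighted in a mixer, randomly pitched slightly down by a modulator, then faded in by an enveloper's volume curve.
# A rattle-shake loop joins at half volume in the final mixer, bypassing both the pitch modulation and the fade.
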SoundCue /DropPodParentGameplay/Assets/Audio/SC_TidalCraneDropPod_LandingLoop
root=Mixer_1
nodes [
 {
  id=Enveloper_0
  children [Modulator_2]
  properties {pitchCurve {editorCurveData {keys [{time=0}]}}, volumeCurve {editorCurveData {keys [<at 0.0162 0.0061>,<at 0.9994 1>]}}}
 }
 {id=Mixer_0, children [WavePlayer_4,WavePlayer_1,WavePlayer_6,WavePlayer_0,WavePlayer_3,WavePlayer_7], properties {inputVolume [1,2,1,3,3,2]}}
 {id=Mixer_1, children [Enveloper_0,WavePlayer_5], properties {inputVolume [1,0.5]}}
 {id=Modulator_2, children [Mixer_0], properties {pitchMax=0.95, pitchMin=0.85, volumeMax=1, volumeMin=1}}
 {id=WavePlayer_0, properties {bLooping=true, soundWaveAssetPtr=/DropPodParentGameplay/Assets/Audio/wav/TidalCraneDropPod_FallingFlameLoopA}}
 {id=WavePlayer_1, properties {bLooping=true, soundWaveAssetPtr=/DropPodParentGameplay/Assets/Audio/wav/SupplyDrop_Falling_Loop_02}}
 {id=WavePlayer_3, properties {bLooping=true, soundWaveAssetPtr=/DropPodParentGameplay/Assets/Audio/wav/TidalCraneDropPod_FallingFlameLoopB}}
 {id=WavePlayer_4, properties {bLooping=true, soundWaveAssetPtr=/DropPodParentGameplay/Assets/Audio/wav/SupplyDrop_Falling_Loop_01}}
 {id=WavePlayer_5, properties {bLooping=true, soundWaveAssetPtr=/DropPodParentGameplay/Assets/Audio/wav/PremiumSupplyDrop_RattleShakeLoop}}
 {id=WavePlayer_6, properties {bLooping=true, soundWaveAssetPtr=/DropPodParentGameplay/Assets/Audio/wav/TidalCraneDropPod_Flare_Loop}}
 {id=WavePlayer_7, properties {bLooping=true, soundWaveAssetPtr=/DropPodParentGameplay/Assets/Audio/wav/PremiumSupplyDrop_FallingLoopClose}}
]
properties {
 attenuationSettings=&/ATT_TidalCraneDropPod_Crate_Descent
 maxDistance=63000
 soundClassObject=&Fort_Gadgets
 concurrencyOverrides {maxCount=4}
}
