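# Sound cue for the TidalCrane drop pod landing impact, mixing three layers that each have their own attenuation.
# The crate-explosion layer blends random impact and explosion waves with a metal hit that an enveloper cuts short.
# The medium-explosion layer pairs random debris with louder random engine waves, while the distant layer plays fixed distant and impact waves.
# It uses the Gadgets sound class, overrides concurrency to limit overlap, and disables virtualization.
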SoundCue /DropPodParentGameplay/Assets/Audio/SC_TidalCraneDropPod_TF_LandingImpact
root=Mixer_1
nodes [
 {id=Attenuation_0, children [Mixer_4], properties {attenuationSettings=/Game/Sounds/Attenuation/Fort_Explosion_Medium_Athena_Att}}
 {
  id=Attenuation_1
  children [Mixer_0]
  properties {attenuationSettings=&/ATT_TidalCraneDropPod_TF_Crate_Explosion}
 }
 {id=Attenuation_2, children [Mixer_2], properties {attenuationSettings=&/ATT_TidalCraneDropPod_Distant}}
 {
  id=Enveloper_0
  children [WavePlayer_0]
  properties {
   pitchCurve {editorCurveData {keys [{time=0}]}}
   volumeCurve {editorCurveData {keys [<at 0 1>,<at 0.2649 1.0002>,<at 0.2968 0.9159>,<at 0.3107 0.7872>,<at 0.3311 0.4569>,<at 0.5427 -0.007>]}}
  }
 }
 {id=Mixer_0, children [Random_0,Enveloper_0,Random_2], properties {inputVolume [1,0.5,1]}}
 {id=Mixer_1, children [Attenuation_1, Attenuation_0, Attenuation_2], properties {inputVolume [1.1,1.2,0.65]}}
 {id=Mixer_2, children [WavePlayer_1,WavePlayer_4], properties {inputVolume [1,0.2]}}
 {id=Mixer_4, children [Random_3,Random_1], properties {inputVolume [0.5,2]}}
 {id=Random_0, children [WavePlayer_2,WavePlayer_3], properties {weights [1,1]}}
 {id=Random_1, children [WavePlayer_7,WavePlayer_8], properties {weights [1,1]}}
 {id=Random_2, children [WavePlayer_10, WavePlayer_11, WavePlayer_12], properties {weights [1,1,1]}}
 {
  id=Random_3
  children [WavePlayer_5, WavePlayer_6, WavePlayer_9, WavePlayer_13, WavePlayer_14, WavePlayer_15, WavePlayer_16]
  properties {weights [1,1,1,1,1,1,1]}
 }
 {id=WavePlayer_0, properties {soundWaveAssetPtr=/DropPodParentGameplay/Assets/Audio/wav/TidalCraneDropPod_LandMetalImpact}}
 {id=WavePlayer_1, properties {soundWaveAssetPtr=/DropPodParentGameplay/Assets/Audio/wav/TidalCraneDropPod_LandDistant}}
 {id=WavePlayer_10, properties {soundWaveAssetPtr=/DropPodParentGameplay/Assets/Audio/wav/PremiumSupplyDrop_LandingExplo_02}}
 {id=WavePlayer_11, properties {soundWaveAssetPtr=/DropPodParentGameplay/Assets/Audio/wav/PremiumSupplyDrop_LandingExplo_03}}
 {id=WavePlayer_12, properties {soundWaveAssetPtr=/DropPodParentGameplay/Assets/Audio/wav/PremiumSupplyDrop_LandingExplo_04}}
 {id=WavePlayer_13, properties {soundWaveAssetPtr=/DropPodParentGameplay/Assets/Audio/wav/TidalCraneDropPod_Impact_Debris_04}}
 {id=WavePlayer_14, properties {soundWaveAssetPtr=/DropPodParentGameplay/Assets/Audio/wav/TidalCraneDropPod_Impact_Debris_05}}
 {id=WavePlayer_15, properties {soundWaveAssetPtr=/DropPodParentGameplay/Assets/Audio/wav/TidalCraneDropPod_Impact_Debris_06}}
 {id=WavePlayer_16, properties {soundWaveAssetPtr=/DropPodParentGameplay/Assets/Audio/wav/TidalCraneDropPod_Impact_Debris_07}}
 {id=WavePlayer_2, properties {soundWaveAssetPtr=/DropPodParentGameplay/Assets/Audio/wav/TidalCraneDropPod_LandImpact_01}}
 {id=WavePlayer_3, properties {soundWaveAssetPtr=/DropPodParentGameplay/Assets/Audio/wav/TidalCraneDropPod_LandImpact_02}}
 {id=WavePlayer_4, properties {soundWaveAssetPtr=/DropPodParentGameplay/Assets/Audio/wav/TidalCraneDropPod_LandImpact_03}}
 {id=WavePlayer_5, properties {soundWaveAssetPtr=/DropPodParentGameplay/Assets/Audio/wav/TidalCraneDropPod_Impact_Debris_01}}
 {id=WavePlayer_6, properties {soundWaveAssetPtr=/DropPodParentGameplay/Assets/Audio/wav/TidalCraneDropPod_Impact_Debris_02}}
 {id=WavePlayer_7, properties {soundWaveAssetPtr=/DropPodParentGameplay/Assets/Audio/wav/PremiumSupplyDrop_LandingEngine_03}}
 {id=WavePlayer_8, properties {soundWaveAssetPtr=/DropPodParentGameplay/Assets/Audio/wav/PremiumSupplyDrop_LandingEngine_04}}
 {id=WavePlayer_9, properties {soundWaveAssetPtr=/DropPodParentGameplay/Assets/Audio/wav/TidalCraneDropPod_Impact_Debris_03}}
]
properties {
 bOverrideConcurrency=true
 maxDistance=40000
 soundClassObject=&Fort_Gadgets
 virtualizationMode=Disabled
 concurrencyOverrides {maxCount=2, retriggerTime=0.1}
}
