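# Translucent, unlit, two-sided smoke trail material for the Rocket Ram effect, built for Niagara ribbons and mesh particles.
# Two panning noise layers, masked along U, distort a stretchable base shape texture.
# Color blends gradients, an additional texture and hue shift, while opacity supports view-angle and near-camera fades.
# Niagara dynamic parameters supply time, noise reduction and grow, and a cosine term builds a swinging world position offset.
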Material /DropPodParentGameplay/Assets/FX/Materials/M_RocketRam_Smoke_Trail
graph {
 properties {
  bUsedWithNiagaraMeshParticles=true
  bUsedWithNiagaraRibbons=true
  blendMode=BLEND_Translucent
  pixelDepthOffsetMode=PDOM_Legacy
  shadingModel=MSM_Unlit
  twoSided=true
  usedShadingModels=Unlit
 }
 nodes [
  MaterialFunctionCall(Emissive=Multiply(A=ParticleColor().RGB, B=NamedRerouteUsage())) {materialFunction=&MF_FX_EmissiveAdjustments}
  n1 = ComponentMask(None=TextureCoordinate()) {r=true}
  n2 = Multiply(A=Multiply(A=@n1, B=OneMinus(None=@n1))) {constB=5}
  Multiply(A=Multiply(A=NamedRerouteUsage(), B=Saturate(None=@n2)), B=OneMinus(None=Fresnel() {baseReflectFraction=0.2, exponent=1}))
  n3 = TextureCoordinate()
  n4 = MaterialFunctionCall(Float2=@n3) {materialFunction=&BreakOutFloat2Components}
  n5 = OneMinus(None=@n4.R)
  n6 = ScalarParameter() {group=01.GlobalUVSettings, parameterName=FlipU}
  n7 = OneMinus(None=@n4.G)
  n8 = ScalarParameter() {group=01.GlobalUVSettings, parameterName=FlipV}
  NamedRerouteDeclaration(None=MaterialFunctionCall(X=LinearInterpolate(A=@n4.R, B=@n5, Alpha=@n6), Y=LinearInterpolate(A=@n4.G, B=@n7, Alpha=@n8)) {materialFunction=&MakeFloat2}.Result) {name=R_UV, nodeColor=<rgba 0 1 0.2266 1>}
  n9 = DynamicParameter() {parameterIndex=3, defaultValue=<rgba 0 0 0 0>, paramNames [Timer, MeshScale, "NoiseU(SetPar)", "NoiseV(SetPar)"]}
  n10 = Ceil(None=@n9.Timer)
  NamedRerouteDeclaration(None=LinearInterpolate(A=Time() {bOverride_Period=true, period=360}, B=@n9.Timer, Alpha=Saturate(None=@n10))) {name=R_TimeRandom, nodeColor=<rgba 0.8281 1 0 1>}
  n11 = VectorParameter() {group=40.NiagaraAnimation, parameterName=GrowAnimation_Detail2, defaultValue=<rgba 1 0.75 0 0>}
  n12 = DynamicParameter() {
   parameterIndex=2
   defaultValue=<rgba 1 0 0 1>
   paramNames ["MeshPosZ(OnlyOnSplineMesh)", "EnableFollow(OnlyOnSplineMesh)", NoiseReducer, Grow]
  }
  n13 = Add(
   A=VectorParameter() {group=40.NiagaraAnimation, parameterName=GrowAnimation_Detail, defaultValue=<rgba 1.5 1 -1.5 0>}.A
   B=OneMinus(None=Saturate(None=@n12.Grow))
  )
  NamedRerouteDeclaration(None=Saturate(None=LinearInterpolate(A=@n11.R, B=@n11.R, Alpha=@n13))) {name=R_NiagaraAnimation_Grow_BaseShapeWidthScale, nodeColor=<rgba 0.7109 1 0 1>}
  n14 = MaterialFunctionCall(Float2=NamedRerouteUsage()) {materialFunction=&BreakOutFloat2Components}
  n15 = Add(
   A=Multiply(A=OneMinus(None=@n14.R), B=ScalarParameter() {group=03.BaseShape, parameterName=BaseShape_Stretch})
   B=ScalarParameter() {group=03.BaseShape, parameterName=BaseShape_StretchOffset}
  )
  n16 = Constant()
  n17 = Add(A=Add(A=AppendVector(A=@n15, B=@n16), B=NamedRerouteUsage()), B=NamedRerouteUsage())
  NamedRerouteDeclaration(None=@n17) {name=R_BaseShapeUV, nodeColor=<rgba 0 1 0.625 1>}
  n18 = VectorParameter() {group=03.BaseShape, parameterName=BaseShape_UVScale, channelNames {a="", b="", g=VScale, r=UScale}, defaultValue=<rgba 1 1 0 1>}
  n19 = Multiply(A=NamedRerouteUsage(), B=@n18.VScale)
  n20 = TextureSampleParameter2D(UVs=MaterialFunctionCall(UVs=@n17, Texture Scale=AppendVector(A=@n18.UScale, B=@n19)) {materialFunction=&ScaleUVsByCenter}.UVs) {
   group=03.BaseShape
   parameterName=BaseShape_Texture
   samplerSource=SSM_Clamp_WorldGroupSettings
   samplerType=SAMPLERTYPE_Grayscale
   texture=/Game/Gadgets/Assets/VinderTech_GliderChute/Glider_Thrive_Female/FX/Texture/T_RibbonBaseShape
  }
  NamedRerouteDeclaration(None=@n20.R) {name=R_BaseShapeSampledValue, nodeColor=<rgba 0.6466 0.4328 1 1>}
  n21 = VectorParameter() {group=22.UVBaseColorGradient, parameterName=Gradient_U_Detail, channelNames {a=ColorHueshift, b=ColorMul, g=hardness, r=Uoffset}}
  n22 = Add(
   A=MaterialFunctionCall(Float2=NamedRerouteUsage()) {materialFunction=&BreakOutFloat2Components}.R
   B=@n21.Uoffset
  )
  n23 = Multiply(A=Saturate(None=Multiply(A=@n22, B=@n21.hardness)), B=@n21.ColorMul)
  n24 = Multiply(A=@n23, B=@n21.ColorHueshift)
  n25 = VectorParameter() {group=22.UVBaseColorGradient, parameterName=Gradient_V_Detail, channelNames {a=ColorHueshift, b=ColorMul, g=hardness, r=Voffset}}
  n26 = Add(
   A=MaterialFunctionCall(Float2=NamedRerouteUsage()) {materialFunction=&BreakOutFloat2Components}.G
   B=@n25.Voffset
  )
  n27 = Multiply(A=Saturate(None=Multiply(A=@n26, B=@n25.hardness)), B=@n25.ColorMul)
  n28 = Multiply(A=@n27, B=@n25.ColorHueshift)
  n29 = Add(A=@n27)
  n30 = Multiply(A=Add(A=@n23), B=NamedRerouteUsage())
  NamedRerouteDeclaration(None=MaterialFunctionCall(Hue Shift Percentage=Add(A=@n24, B=@n28), Texture=Multiply(A=@n29, B=@n30)) {materialFunction=&HueShift}.Result) {name=R_Color, nodeColor=<rgba 0 0.1094 1 1>}
  n31 = Multiply(
   A=ComponentMask(None=NamedRerouteUsage()) {r=true}
   B=ScalarParameter() {defaultValue=1, group=SimpleColorGradient, parameterName=Shift_Gradient}
  )
  n32 = LinearInterpolate(
   A=VectorParameter() {group=SimpleColorGradient, parameterName=ColorA, defaultValue=<rgba 0 0.132 1 1>}.RGB
   B=VectorParameter() {group=SimpleColorGradient, parameterName=ColorB, defaultValue=<rgba 1 0.0693 0 1>}.RGB
   Alpha=Power(Base=Clamp(None=@n31), Exp=ScalarParameter() {defaultValue=2, group=SimpleColorGradient, parameterName=Gradient_Power})
  )
  n33 = VectorParameter() {group=22.UVBaseColorGradient, parameterName=SpecialGradient_U_Detail, channelNames {a="", b=Colormul, g=hardness, r=Uoffset}}
  n34 = Add(
   A=MaterialFunctionCall(Float2=NamedRerouteUsage()) {materialFunction=&BreakOutFloat2Components}.R
   B=@n33.Uoffset
  )
  n35 = Multiply(A=Saturate(None=Multiply(A=@n34, B=@n33.hardness)), B=@n33.Colormul)
  n36 = VectorParameter() {group=21.AdditionalTexture, parameterName=AdditionalTexture_Color_Frontface, defaultValue=<rgba 0 0 0 1>}
  n37 = VectorParameter() {
   group=21.AdditionalTexture
   parameterName=AdditionalTexture_UVScaleAndNoiseDistortStrength
   channelNames {a=NoiseVScale, b=NoiseUScale, g=VScale, r=UScale}
  }
  n38 = MaterialFunctionCall(
   UVs=Add(A=NamedRerouteUsage(), B=AppendVector(A=@n37.NoiseUScale, B=@n37.NoiseUScale))
   Texture Scale=AppendVector(A=@n37.UScale, B=@n37.UScale)
  ) {materialFunction=&ScaleUVsByCenter}
  n39 = VectorParameter() {
   group=21.AdditionalTexture
   parameterName=AdditionalTexture_ValueTweakDetail
   channelNames {a="", b=contrast, g=mul, r=add}
   defaultValue=<rgba 0 1 0 0>
  }
  n40 = Multiply(
   A=Add(
    A=TextureSampleParameter2D(UVs=@n38.UVs) {group=21.AdditionalTexture, parameterName="AdditionalTexture(R)", texture=&T_BlacksRGB}.R
    B=@n39.add
   )
   B=@n39.mul
  )
  n41 = LinearInterpolate(
   A=Multiply(A=VectorParameter() {group=20.BaseColor, parameterName=BaseColor_FrontFace, defaultValue=<rgba 1 0 0.6179 1>}.RGB, B=Add(A=@n35))
   B=@n36.RGB
   Alpha=MaterialFunctionCall(In=Saturate(None=@n40), Contrast=@n39.contrast) {materialFunction=&CheapContrast}.Result
  )
  n42 = NamedRerouteUsage()
  n43 = Add(
   A=NamedRerouteUsage()
   B=Multiply(
    A=MaterialFunctionCall(XY=@n42) {materialFunction=&Swizzle}.YX
    B=ScalarParameter() {group=20.BaseColor, parameterName=HueShiftNoiseScale}
   )
  )
  n44 = Add(
   A=Multiply(
    A=MaterialFunctionCall(Float2=@n43) {materialFunction=&BreakOutFloat2Components}.R
    B=ScalarParameter() {group=20.BaseColor, parameterName=HueShiftStr}
   )
   B=Multiply(A=NamedRerouteUsage(), B=ScalarParameter() {group=20.BaseColor, parameterName=HueShiftUVPanSpeed})
  )
  NamedRerouteDeclaration(None=LinearInterpolate(
    A=Multiply(A=@n32, B=@n41)
    B=MaterialFunctionCall(Hue Shift Percentage=@n44, Texture=@n41) {materialFunction=&HueShift}.Result
    Alpha=ScalarParameter() {defaultValue=1, group=SimpleColorGradient, parameterName=Use_HueShift}
   )) {name=R_BaseColorWithAdditionalTexture, nodeColor=<rgba 1 0 0.7266 1>}
  n45 = VectorParameter() {
   group=10.UVDistortNoise_A
   parameterName=NoiseA_UVTransform
   channelNames {a=PannerV, b=PannerU, g=ScaleV, r=ScaleU}
   defaultValue=<rgba 1 0.1 0.2 0>
  }
  n46 = Multiply(A=NamedRerouteUsage(), B=AppendVector(A=@n45.ScaleU, B=@n45.ScaleU))
  n47 = VectorParameter() {group=10.UVDistortNoise_A, parameterName=NoiseA_UVOffset, channelNames {a="", b="", g=OffsetV, r=OffsetU}, defaultValue=<rgba 1 0.1 0.2 0>}
  n48 = AppendVector(A=@n47.OffsetU, B=@n47.OffsetU)
  n49 = TextureSampleParameter2D(UVs=Panner(Coordinate=Add(A=@n46, B=@n48), Time=NamedRerouteUsage(), Speed=AppendVector(A=@n45.PannerU, B=@n45.PannerU)) {speedX=0.216}) {
   group=10.UVDistortNoise_A
   parameterName=NoiseA_Texture
   samplerType=SAMPLERTYPE_LinearGrayscale
   texture=/Game/Gadgets/Assets/VinderTech_GliderChute/Glider_Thrive_Female/FX/Texture/T_AntiMaterialCharge_Dissolve_01
  }
  n50 = VectorParameter() {
   group=10.UVDistortNoise_A
   parameterName=NoiseA_Property
   channelNames {a=NoiseStrength_V, b=NoiseStrength_U, g=NoiseStrengthOffset_V, r=NoiseStrengthOffset_U}
   defaultValue=<rgba -0.5 -0.5 0.2 0>
  }
  n51 = Add(A=AppendVector(A=@n49.R, B=@n49.R), B=AppendVector(A=@n50.NoiseStrengthOffset_U, B=@n50.NoiseStrengthOffset_U)) {constB=-0.5}
  n52 = AppendVector(A=@n50.NoiseStrength_U, B=@n50.NoiseStrength_U)
  n53 = MaterialFunctionCall(Float2=NamedRerouteUsage()) {materialFunction=&BreakOutFloat2Components}
  n54 = OneMinus(None=@n53.R)
  n55 = VectorParameter() {
   group=10.UVDistortNoise_A
   parameterName=NoiseA_UMaskInfo
   channelNames {a=Buttomoffset, b=ButtomHardness, g=TopOffset, r=TopHardness}
   defaultValue=<rgba 5 0 25 5>
  }
  n56 = Saturate(None=Add(A=Multiply(A=@n54, B=@n55.TopHardness) {constB=15}, B=@n55.TopOffset))
  n57 = Saturate(None=Add(A=Multiply(A=@n53.R, B=@n55.ButtomHardness) {constB=15}, B=@n55.Buttomoffset))
  NamedRerouteDeclaration(None=Multiply(A=Multiply(A=@n51, B=@n52), B=Multiply(A=@n56, B=@n57))) {name=R_NoiseA, nodeColor=<rgba 0 1 0.25 1>}
  n58 = Multiply(A=NamedRerouteUsage(), B=Multiply(A=ScalarParameter() {defaultValue=1, group=50.Opacity, parameterName=OpacityScale}, B=ParticleColor().A))
  n59 = ScalarParameter() {defaultValue=2, parameterName=ViewAngleSensitivity}
  n60 = LinearInterpolate(
   A=@n58
   B=Multiply(
    A=@n58
    B=MaterialFunctionCall("FadeSensitivity (optional)"=@n59) {materialFunction=&FadeBasedOnViewAngle}.Result
   )
   Alpha=ScalarParameter() {parameterName=UseViewAngle}
  )
  n61 = OneMinus(None=SphereMask(
    A=PixelDepth().R
    B=Constant()
    Radius=ScalarParameter() {defaultValue=400, parameterName=NearCameraFadeRadius}
    Hardness=ScalarParameter() {parameterName=NearCameraFadeHardness}
   ))
  NamedRerouteDeclaration(None=StaticSwitchParameter(True=Multiply(A=@n60, B=@n61), False=@n60) {defaultValue=true, parameterName=IsUsingCameraFade}) {name=R_Opacity, nodeColor=<rgba 0 1 0.6016 1>}
  NamedRerouteDeclaration(None=@n3) {name=R_OriginalUV, nodeColor=<rgba 0 1 0.4869 1>}
  n62 = VectorParameter() {
   group=10.UVDistortNoise_B
   parameterName=NoiseB_UVTransform
   channelNames {a=PannerV, b=PannerU, g=ScaleV, r=ScaleU}
   defaultValue=<rgba 1 0.1 0.2 0>
  }
  n63 = Multiply(A=NamedRerouteUsage(), B=AppendVector(A=@n62.ScaleU, B=@n62.ScaleU))
  n64 = VectorParameter() {group=10.UVDistortNoise_B, parameterName=NoiseB_UVOffset, channelNames {a="", b="", g=OffsetV, r=OffsetU}, defaultValue=<rgba 1 0.1 0.2 0>}
  n65 = AppendVector(A=@n64.OffsetU, B=@n64.OffsetU)
  n66 = TextureSampleParameter2D(UVs=Panner(Coordinate=Add(A=@n63, B=@n65), Time=NamedRerouteUsage(), Speed=AppendVector(A=@n62.PannerU, B=@n62.PannerU)) {speedX=0.216}) {
   group=10.UVDistortNoise_B
   parameterName=NoiseB_Texture
   samplerType=SAMPLERTYPE_LinearGrayscale
   texture=&T_AntiMaterialCharge_Dissolve_01
  }
  n67 = VectorParameter() {
   group=10.UVDistortNoise_B
   parameterName=NoiseB_Property
   channelNames {a=NoiseStrength_V, b=NoiseStrength_U, g=NoiseStrengthOffset_V, r=NoiseStrengthOffset_U}
   defaultValue=<rgba -0.5 -0.5 0.2 0>
  }
  n68 = Add(A=AppendVector(A=@n66.R, B=@n66.R), B=AppendVector(A=@n67.NoiseStrengthOffset_U, B=@n67.NoiseStrengthOffset_U)) {constB=-0.5}
  n69 = AppendVector(A=@n67.NoiseStrength_U, B=@n67.NoiseStrength_U)
  n70 = MaterialFunctionCall(Float2=NamedRerouteUsage()) {materialFunction=&BreakOutFloat2Components}
  n71 = OneMinus(None=@n70.R)
  n72 = VectorParameter() {
   group=10.UVDistortNoise_B
   parameterName=NoiseB_UMaskInfo
   channelNames {a=Buttomoffset, b=ButtomHardness, g=TopOffset, r=TopHardness}
   defaultValue=<rgba 3 0 5 2.4>
  }
  n73 = Saturate(None=Add(A=Multiply(A=@n71, B=@n72.TopHardness) {constB=15}, B=@n72.TopOffset))
  n74 = Saturate(None=Add(A=Multiply(A=@n70.R, B=@n72.ButtomHardness) {constB=15}, B=@n72.Buttomoffset))
  NamedRerouteDeclaration(None=Multiply(A=Multiply(A=@n68, B=@n69), B=Multiply(A=@n73, B=@n74))) {name=R_NoiseB, nodeColor=<rgba 1 0.7969 0 1>}
  n75 = DynamicParameter() {
   parameterIndex=2
   defaultValue=<rgba 1 0 0 1>
   paramNames ["MeshPosZ(OnlyOnSplineMesh)", "EnableFollow(OnlyOnSplineMesh)", NoiseReducer, Grow]
  }
  n76 = Saturate(None=@n75.NoiseReducer)
  n77 = NamedRerouteUsage()
  n78 = NamedRerouteUsage()
  NamedRerouteDeclaration(None=Multiply(A=OneMinus(None=@n76), B=Add(A=@n77, B=@n78))) {name=R_MixedNoise, nodeColor=<rgba 0 1 0.0859 1>}
  n79 = VectorParameter() {group=30.SwingWPO, parameterName=Swing_Cosine_Mask, channelNames {a="", b="", g=offset, r=Multiply}, defaultValue=<rgba 2 -1.1 0 1>}
  n80 = MaterialFunctionCall(Float2=NamedRerouteUsage()) {materialFunction=&BreakOutFloat2Components}
  n81 = Saturate(None=Add(A=Multiply(A=@n79.Multiply, B=@n80.R), B=@n79.offset))
  n82 = VectorParameter() {group=30.SwingWPO, parameterName=Swing_Cosine_Property, channelNames {a="", b=Frequency, g=Strength, r=Speed}, defaultValue=<rgba 1.5 5 2 1>}
  n83 = Add(A=Multiply(A=@n80.R, B=@n82.Frequency), B=Multiply(A=NamedRerouteUsage(), B=@n82.Speed))
  n84 = Multiply(A=Cosine(None=@n83), B=@n82.Strength)
  n85 = Transform(None=VectorParameter() {group=30.SwingWPO, parameterName=Swing_Cosine_Axis, defaultValue=<rgba 0 1 0 0>}.RGB)
  NamedRerouteDeclaration(None=Multiply(A=OneMinus(None=@n81), B=Multiply(A=@n84, B=@n85))) {name=R_SwingWPO, nodeColor=<rgba 1 0.7734 0 1>}
  NamedRerouteUsage()
  Sine(None=@n83)
 ]
}
