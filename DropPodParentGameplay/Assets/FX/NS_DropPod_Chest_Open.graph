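# Niagara burst effect for a drop pod chest opening, made of three emitters with warm gold-orange colors.
# A combined emitter drives a fading, growing sprite and a faster expanding mesh shape through separate custom particle attributes and curves.
# Rays adds slowly rising sprite streaks, while Sparks throws small particles outward with gravity, curl noise and drag.
# User.GlobalOffset offsets all emitters, while User.u_Scale sizes only the combined sprite and mesh.
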NiagaraSystem /DropPodParentGameplay/Assets/FX/NS_DropPod_Chest_Open
dependencies {
 usedDataInterfaces [
  &NiagaraDataInterfaceCurve
  &NiagaraDataInterfaceVector2DCurve
  &NiagaraDataInterfaceVectorCurve
 ]
 usedDynamicInputs []
 usedModules [&SystemState]
}
emitters {
 Combined_Emitter_Sprite {
  emitterUpdateScript [
   &EmitterState()
   &SpawnBurst_Instantaneous(Use Spawn Probability=false, UseLoopCountLimit=false)
  ]
  particleSpawnScript [
   &InitializeParticle(
    Color Mode=::NewEnumerator0
    Lifetime=1.5
    Position Offset=$User.GlobalOffset
    Sprite Rotation Mode=::NewEnumerator2
    UsePositionOffset=true
    Write Material Random=true
   )
   /SetVariables(
    Particles.Sprite_01_CameraOffset=150
    Particles.Sprite_01_Color:LinearColor=<rgba 2 1.1864 0.22 0.1>
    Particles.Sprite_01_Size:Vector2f=&Multiply_Vector2D_ByFloat(Float=$User.u_Scale, Vector2D=<200 200>)
   ) {isSetParameters=true}
   /SetVariables(
    Particles.Mesh_Color_Shape:LinearColor=<rgba 5 1.6247 0.4688 0.25>
    Particles.Mesh_DynamicParameter_Shape:Vector4f=&MakeVector4(X=1, Y=1, Z=1, W=&UniformRangedFloat(Minimum=0, Maximum=20))
    Particles.Mesh_Size_Shape:Vector3f=&VectorFromFloat(Value=&Multiply_Float(A=&RandomRangeFloat(Minimum=0.8, Maximum=1), B=$User.u_Scale))
   ) {isSetParameters=true}
   &SystemLocation(MaskSpawnGroup=false, Offset=$User.GlobalOffset, Offset Coordinate Space=::World)
  ]
  particleUpdateScript [
   &ParticleState()
   /SetVariables(
    Particles.Sprite_01_Color:LinearColor=&MultiplyLinearColor(
     Linear Color=$Particles.Initial.Sprite_01_Color
     Scale Factor=&MakeFloat4FromVectorAndFloat(
      "Vector (RGB)"=&VectorFromFloat(Value=&FloatFromCurve(
        FloatCurve=DataInterface() {propertyValues {Curve {keys [<at -0.0046 1.0113>,<at 1.0011 0.002>]}}}
        CurveIndex=&Divide_Float(A=$Particles.NormalizedAge, B=1)
       ))
      "Float (Alpha)"=&FloatFromCurve(
       FloatCurve=DataInterface() {propertyValues {Curve {keys [<at -0.0046 1.0113>,<at 1.0011 0.002>]}}}
       CurveIndex=&Divide_Float(A=$Particles.NormalizedAge, B=1)
      )
     )
    )
    Particles.Sprite_01_Size:Vector2f=&Multiply_Vector2D(
     A=$Particles.Initial.Sprite_01_Size
     B=&Vector2DFromCurve(
      Vector2Curve=DataInterface() {propertyValues {
        XCurve {keys [
          {arriveTangent=4.1785, interpMode=RCIM_Cubic, leaveTangent=4.1785, tangentMode=RCTM_User, time=0, value=0}
          {arriveTangent=1.1179, interpMode=RCIM_Cubic, leaveTangent=1.1179, tangentMode=RCTM_User, time=0.2042, value=0.6228}
          {arriveTangent=0.0224, interpMode=RCIM_Cubic, leaveTangent=0.0224, tangentMode=RCTM_User, time=1, value=1}
         ]}
        YCurve {keys [
          {arriveTangent=4.1785, interpMode=RCIM_Cubic, leaveTangent=4.1785, tangentMode=RCTM_User, time=0, value=0}
          {arriveTangent=1.1179, interpMode=RCIM_Cubic, leaveTangent=1.1179, tangentMode=RCTM_User, time=0.2042, value=0.6228}
          {arriveTangent=0.0224, interpMode=RCIM_Cubic, leaveTangent=0.0224, tangentMode=RCTM_User, time=1, value=1}
         ]}
       }}
      CurveIndex=&Divide_Float(A=$Particles.NormalizedAge, B=1)
      ScaleCurve=<1 1>
     )
    )
   ) {isSetParameters=true}
   /SetVariables(
    Particles.Mesh_Color_Shape:LinearColor=&MultiplyLinearColor(
     Linear Color=$Particles.Initial.Mesh_Color_Shape
     Scale Factor=&MakeFloat4FromVectorAndFloat(
      "Vector (RGB)"=&VectorFromCurve(
       VectorCurve=DataInterface() {propertyValues {
         XCurve {keys [<at -0.0011 1.0067>,<at 0.9988 0>]}
         YCurve {keys [<at -0.0011 1.0067>,<at 0.9988 0>]}
         ZCurve {keys [<at -0.0011 1.0067>,<at 0.9988 0>]}
        }}
       CurveIndex=&Divide_Float(A=$Particles.NormalizedAge, B=0.35)
      )
      "Float (Alpha)"=&FloatFromCurve(
       FloatCurve=DataInterface() {propertyValues {Curve {keys [
           {interpMode=RCIM_Cubic, tangentMode=RCTM_User, time=-0.0046, value=1.0113}
           {arriveTangent=-1.0036, interpMode=RCIM_Cubic, leaveTangent=-1.0036, tangentMode=RCTM_User, time=0.7114, value=0.7365}
           {arriveTangent=-2.4196, interpMode=RCIM_Cubic, leaveTangent=-2.4196, tangentMode=RCTM_User, time=1.0011, value=0.002}
          ]}}}
       CurveIndex=&Divide_Float(A=$Particles.NormalizedAge, B=0.35)
      )
     )
    )
    Particles.Mesh_DynamicParameter_Shape_01:Vector4f=&MakeVector4(
     X=&FloatFromCurve(
      FloatCurve=DataInterface() {propertyValues {Curve {keys [<at 0 0>,<at 0.22 0.5>,<at 1 0>]}}}
      CurveIndex=&Divide_Float(A=$Particles.NormalizedAge, B=0.35)
     )
     Y=&FloatFromCurve(
      FloatCurve=DataInterface() {propertyValues {Curve {keys [<at 0 15>,<at 1 0>]}}}
      CurveIndex=&Divide_Float(A=$Particles.NormalizedAge, B=0.35)
     )
     Z=0
     W=0
    )
    Particles.Mesh_Size_Shape:Vector3f=&Multiply_Vector(
     A=$Particles.Initial.Mesh_Size_Shape
     B=&VectorFromFloat(Value=&FloatFromCurve(
       FloatCurve=DataInterface() {propertyValues {Curve {keys [
           {arriveTangent=4.7007, interpMode=RCIM_Cubic, leaveTangent=4.7007, tangentMode=RCTM_User, time=0, value=0}
           {arriveTangent=2, arriveTangentWeight=0.0833, interpMode=RCIM_Cubic, leaveTangent=2, tangentMode=RCTM_User, time=0.25, value=1}
           {arriveTangentWeight=0.25, interpMode=RCIM_Cubic, tangentMode=RCTM_User, time=1, value=2}
          ]}}}
       CurveIndex=&Divide_Float(A=$Particles.NormalizedAge, B=0.35)
      ))
    )
   ) {isSetParameters=true}
   &DynamicMaterialParameters(
    Index 0 Param 1=500
    Module.Index0Float0Write=true
    Module.Index0Float1Write=false
    Module.Index0Float2Write=false
    Module.Index0Float3Write=false
   )
  ]
  rendererClasses [&NiagaraSpriteRendererProperties, &NiagaraMeshRendererProperties]
  renderers [
   {rendererClass=&NiagaraSpriteRendererProperties}
   {rendererClass=&NiagaraMeshRendererProperties, rendererIndex=1}
  ]
 }
 Rays {
  emitterUpdateScript [
   &EmitterState()
   &SpawnBurst_Instantaneous(Spawn Count=8, Use Spawn Probability=false, UseLoopCountLimit=false)
  ]
  particleSpawnScript [
   &InitializeParticle(
    Color=<rgba 5 1.8966 0 0.1>
    Lifetime Max=1.5
    Lifetime Mode=::NewEnumerator1
    Sprite Rotation Mode=::NewEnumerator2
    Sprite Size Max=<200 400>
    Sprite Size Min=<150 150>
    Sprite Size Mode=::NewEnumerator2
    UsePositionOffset=false
    Write Material Random=true
   )
   /Niagara/Modules/Update/SubUV/SubUVAnimation()
   &ShapeLocation(
    MaskSpawnGroup=false
    Offset=$User.GlobalOffset
    Offset Coordinate Space=::World
    Offset Mode=::NewEnumerator0
    Override Seed=false
    Sphere Radius=25
    UseOwnerScale=true
   )
   &CameraOffset(Camera Offset Amount=50)
   &AddVelocity(
    Module.Use Falloff Distance=false
    Override Seed=false
    Velocity Mode=::NewEnumerator1
    Velocity Origin=$Particles.ShapeLocation.ShapePosition
   )
  ]
  particleUpdateScript [
   &ParticleState()
   &ScaleSpriteSize(Scale Factor=&Vector2DFromFloat(Value=&FloatFromCurve(FloatCurve=DataInterface() {propertyValues {Curve {keys [
          <at -0.0017 0.0417 RCIM_Cubic>
          {arriveTangent=1.1486, interpMode=RCIM_Cubic, leaveTangent=1.1486, tangentMode=RCTM_User, time=0.0317, value=1.2332}
          <at 0.57 1.2568>
          {arriveTangent=-0.1439, interpMode=RCIM_Cubic, leaveTangent=-0.1439, tangentMode=RCTM_User, time=0.9845, value=1.0006}
          {interpMode=RCIM_Cubic, tangentMode=RCTM_User, time=1.0011, value=0.002}
         ]}}})))
   &ScaleColor(
    Scale Alpha=&FloatFromCurve(FloatCurve=DataInterface() {propertyValues {Curve {keys [<at -0.0046 1.0113 RCIM_Cubic>,<at 1.0011 0.002 RCIM_Cubic>]}}}, Scale Curve=2)
    ScaleA=true
    ScaleRGB=true
   )
   &GravityForce(Gravity=<0 0 11>)
   &SolveForcesAndVelocity(Clamp Velocity=false, Limit Acceleration=false)
  ]
  rendererClasses [&NiagaraSpriteRendererProperties]
  renderers [{rendererClass=&NiagaraSpriteRendererProperties}]
 }
 Sparks {
  emitterUpdateScript [
   &EmitterState(Life Cycle Mode=::NewEnumerator1, Loop Behavior=::NewEnumerator1, Loop Duration=1, UseLoopDelay=false)
   &SpawnBurst_Instantaneous(Spawn Count=30, Use Spawn Probability=false, UseLoopCountLimit=false)
  ]
  particleSpawnScript [
   &InitializeParticle(
    Color=<rgba 5 3.515 1 1>
    Lifetime Max=1.5
    Lifetime Min=0.5
    Lifetime Mode=::NewEnumerator1
    Sprite Size Mode=::NewEnumerator1
    Uniform Sprite Size Max=30
    Uniform Sprite Size Min=10
    UsePositionOffset=false
    Write Material Random=true
   )
   &ShapeLocation(
    MaskSpawnGroup=false
    Offset=$User.GlobalOffset
    Offset Mode=::NewEnumerator0
    Override Seed=true
    Random Seed=2147483647
    Randomness Mode:ENiagaraRandomnessMode=::NewEnumerator1
    Sphere Radius=25
    UseOwnerScale=true
   )
   &AddVelocity(
    Module.Use Falloff Distance=false
    Origin Offset=<0 0 -10>
    Override Seed=false
    Velocity Mode=::NewEnumerator1
    Velocity Origin=$Particles.ShapeLocation.ShapePosition
    Velocity Speed=&RandomRangeFloat(Minimum=50, Maximum=500)
   )
  ]
  particleUpdateScript [
   &ParticleState()
   &GravityForce(Gravity=<0 0 -50>)
   &CurlNoiseForce(Noise Frequency=5, Noise Strength=500, Pan Noise=false, Random Seed=5)
   &Drag(
    Drag=&RandomRangeFloat(Minimum=1.5, Maximum=3, Randomness Mode=::NewEnumerator1, Random Seed=5)
    Use Linear Drag=true
    Use Rotational Drag=false
   )
   &ScaleColor(
    Scale Alpha=&FloatFromCurve(
     FloatCurve=DataInterface() {propertyValues {Curve {keys [<at -0.0092 0.0144>,<at 0.0967 0.9767>,<at 1.0011 0.002>]}}}
     Scale Curve=&RandomRangeFloat(Minimum=0.25, Maximum=0.75)
    )
    ScaleA=true
    ScaleRGB=true
   )
   &ScaleSpriteSize(Scale Factor=&Vector2DFromFloat(Value=&FloatFromCurve(FloatCurve=DataInterface() {propertyValues {Curve {keys [<at -0.0046 0>,<at 0.1746 0.9952>,<at 1.0011 0.002>]}}})))
   &SolveForcesAndVelocity(Clamp Velocity=false, Limit Acceleration=false)
  ]
  rendererClasses [&NiagaraSpriteRendererProperties]
  renderers [{rendererClass=&NiagaraSpriteRendererProperties}]
 }
}
summary {
 systemName=NS_DropPod_Chest_Open
 emitters [
  {
   emitterName=Combined_Emitter_Sprite
   rendererClasses [&NiagaraSpriteRendererProperties, &NiagaraMeshRendererProperties]
  }
  {emitterName=Rays, rendererClasses [&NiagaraSpriteRendererProperties]}
  {emitterName=Sparks, rendererClasses [&NiagaraSpriteRendererProperties]}
 ]
 userVariables [
  User.GlobalOffset {type=Vector3f, description="", defaultValue {struct=Vector3f, value=<0 0 125>}}
  User.u_Scale {type=NiagaraFloat, defaultValue=3, description=""}
 ]
}
