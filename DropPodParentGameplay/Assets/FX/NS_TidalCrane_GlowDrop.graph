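# Niagara glow effect for the Tidal Crane drop: a single Corona emitter bursts a large, pulsing orange sprite after the User.Delay wait.
# Looping particle lifetime drives curves for size, alpha and material parameters, with system age and Timeline progress shaping growth and fade.
# FN_Sprite_Attributes computes a second color and size set alongside two sprite renderers.
# Camera distance remaps scale and brightness; User.bIsGold gates alpha, zeroing it when false.
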NiagaraSystem /DropPodParentGameplay/Assets/FX/NS_TidalCrane_GlowDrop
dependencies {
 usedDataInterfaces [&NiagaraDataInterfaceCurve]
 usedDynamicInputs [&SetBoolByFloatComparison]
 usedModules [&SystemState, &Timeline]
}
emitters {Corona {
  emitterUpdateScript [
   &EmitterState(
    Inactive Response=::NewEnumerator1
    Life Cycle Mode=::NewEnumerator1
    Loop Behavior=::NewEnumerator1
    Loop Delay=$User.Delay
    Loop Duration Mode=::NewEnumerator1
    UseLoopDelay=true
   )
   &SpawnBurst_Instantaneous(Use Spawn Probability=false, UseLoopCountLimit=false)
  ]
  particleSpawnScript [&InitializeParticle(
    Color=<rgba 5 1.25 0 1>
    Lifetime=3
    Position Offset=<0 0 150>
    Sprite Size=<1750 1750>
    Sprite Size Mode=::NewEnumerator4
    UsePositionOffset=true
    Write Material Random=true
   )]
  particleUpdateScript [
   &ParticleState(
    Kill Particles When Lifetime Has Elapsed=false
    Let Infinitely Lived Particles Die When Emitter Deactivates=true
    Loop Particles Lifetime=true
   )
   &ScaleSpriteSize(
    Scale Sprite Size Mode=::NewEnumerator0
    Uniform Scale Factor=&Multiply_Float(
     A=&FloatFromCurve(
      FloatCurve=DataInterface() {propertyValues {Curve {
         defaultValue=0.5
         keys [
          {interpMode=RCIM_Cubic, tangentMode=RCTM_User, time=0, value=0.5}
          {interpMode=RCIM_Cubic, tangentMode=RCTM_User, time=0.5043, value=0.45}
          {tangentMode=RCTM_User, time=1, value=0.5}
         ]
        }}}
      Scale Curve=&FN_LOD_Distance_Remap(Start Distance=50, End Distance=2500, Minimum Output=0.25, Maximum Output=0.75, Clamp Output=true)
     )
     B=&FloatFromCurve(
      FloatCurve=DataInterface() {propertyValues {Curve {
         defaultValue=1
         keys [
          {arriveTangent=4.1785, interpMode=RCIM_Cubic, leaveTangent=4.1785, tangentMode=RCTM_User, time=0, value=0}
          {arriveTangent=1.1179, interpMode=RCIM_Cubic, leaveTangent=1.1179, tangentMode=RCTM_User, time=0.2042, value=0.6228}
          {arriveTangent=0.0224, interpMode=RCIM_Cubic, leaveTangent=0.0224, tangentMode=RCTM_User, time=1, value=1}
         ]
        }}}
      CurveIndex=$Engine.System.Age
     )
    )
   )
   &ScaleColor(
    Scale Alpha=&Multiply_Float(
     A=$System.Timeline.NormalizedTime
     B=&FloatFromCurve(
      FloatCurve=DataInterface() {propertyValues {Curve {
         defaultValue=1
         keys [
          {interpMode=RCIM_Cubic, tangentMode=RCTM_User, time=0, value=0.5}
          {interpMode=RCIM_Cubic, tangentMode=RCTM_User, time=0.608, value=0.9888}
          {tangentMode=RCTM_User, time=1, value=0.5}
         ]
        }}}
      Scale Curve=&MakeCustomFloatFromBool(Bool=$User.bIsGold, True Float=0.75, False Float=0)
     )
    )
    ScaleA=true
    ScaleRGB=true
   )
   &CameraOffset(Camera Offset Amount=50)
   &DynamicMaterialParameters(
    Index 0 Param 1=&FloatFromCurve(
     FloatCurve=DataInterface() {propertyValues {Curve {
        defaultValue=0.2
        keys [
         {interpMode=RCIM_Cubic, tangentMode=RCTM_User, time=0.0042, value=0.15}
         {interpMode=RCIM_Cubic, tangentMode=RCTM_User, time=0.349, value=0.1794}
         {tangentMode=RCTM_User, time=1.0042, value=0.15}
        ]
       }}}
     Scale Curve=5
    )
    Index 0 Param 2=1
    Index 0 Param 3=1
    Index 0 Param 4=1
    Module.Index0Param0Write=true
    Module.Index0Param1Write=true
    Module.Index0Param2Write=true
    Module.Index0Param3Write=true
   )
   &FN_Sprite_Attributes(
    Color=&MultiplyLinearColor(
     Linear Color=<rgba 2.5 0.625 0 1>
     Scale Factor=&MakeVector4FromVectorAndFloat(
      "Vector (RGB)"=&VectorFromFloat(Value=&FN_LOD_Distance_Remap(Start Distance=50, End Distance=2500, Minimum Output=1, Maximum Output=1.5, Clamp Output=true))
      "Float (Alpha)"=&Multiply_Float(
       A=&FloatFromCurve(
        FloatCurve=DataInterface() {propertyValues {Curve {
           defaultValue=0.5
           keys [
            {interpMode=RCIM_Cubic, tangentMode=RCTM_User, time=0, value=0.5}
            {interpMode=RCIM_Cubic, tangentMode=RCTM_User, time=0.5043, value=0.45}
            {tangentMode=RCTM_User, time=1, value=0.5}
           ]
          }}}
        Scale Curve=&MakeCustomFloatFromBool(Bool=$User.bIsGold, True Float=2, False Float=0)
       )
       B=$System.Timeline.NormalizedTime
      )
     )
    )
    Create Color Attribute=true
    Create SpriteSize Attribute=true
    SpriteSize=&Multiply_Vector2D_ByFloat(
     Float=&Multiply_Float(
      A=$System.Timeline.NormalizedTime
      B=&FloatFromCurve(
       FloatCurve=DataInterface() {propertyValues {Curve {
          defaultValue=0.5
          keys [
           {interpMode=RCIM_Cubic, tangentMode=RCTM_User, time=0, value=0.6}
           {interpMode=RCIM_Cubic, tangentMode=RCTM_User, time=0.5, value=0.4}
           {tangentMode=RCTM_User, time=1, value=0.6}
          ]
         }}}
       Scale Curve=&FN_LOD_Distance_Remap(Start Distance=50, End Distance=2500, Minimum Output=0.25, Maximum Output=0.75, Clamp Output=true)
      )
     )
     Vector2D=<3000 3000>
    )
    Write to Module Namespace=true
   )
  ]
  rendererClasses [&NiagaraSpriteRendererProperties]
  renderers [
   {rendererClass=&NiagaraSpriteRendererProperties}
   {rendererClass=&NiagaraSpriteRendererProperties, rendererIndex=1}
  ]
 }}
summary {
 systemName=NS_TidalCrane_GlowDrop
 emitters [{emitterName=Corona, rendererClasses [&NiagaraSpriteRendererProperties]}]
 userVariables [
  User.bIsGold {type=NiagaraBool}
  User.Delay {type=NiagaraFloat, defaultValue=3.5, description=""}
 ]
}
