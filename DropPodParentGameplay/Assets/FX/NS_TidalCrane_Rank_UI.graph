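# Niagara system for the TidalCrane rank display in the drop pod, with user inputs for rank, unlock state, MMO pod flag and delay.
# Active emitters cover ranking, wide holo beams and a light; holo bits and a second beam emitter are disabled.
# The misleadingly named SimpleSpriteBurst renders only a light, using User.Delay as loop delay, taking the system color and pulsing its RGB through a waveform.
# Ranking and beam emitters list no renderers or scripts here.
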NiagaraSystem /DropPodParentGameplay/Assets/FX/NS_TidalCrane_Rank_UI
dependencies {
 usedDataInterfaces []
 usedDynamicInputs [
  &MakeCustomFloatFromBool
  &MergeHelper_ColorSwitch
  &MakeIntFromFloat
  &Multiply_Vector2D_ByFloat
  &FN_LOD_Distance_Remap
  &Lerp_Vector2
  &MergeHelper_FloatSwitch
 ]
 usedModules [
  &SystemState
  &Timeline
  &self:SystemSpawnScript.SystemScriptSource.SystemScriptGraph.NiagaraNodeAssignment_2.SetVariables_#1
  &self:SystemSpawnScript.SystemScriptSource.SystemScriptGraph.NiagaraNodeAssignment_0.SetVariables_#2
 ]
}
emitters {
 Em_HoloBeams_Wide {rendererClasses [], renderers []}
 Em_HoloBeams_Wide001 {bEnabled=false, rendererClasses [], renderers []}
 Em_HoloBits {bEnabled=false, rendererClasses [], renderers []}
 Em_Ranking {rendererClasses [], renderers []}
 SimpleSpriteBurst {
  emitterUpdateScript [
   &EmitterState(
    Inactive Response=::NewEnumerator1
    Life Cycle Mode=::NewEnumerator1
    Loop Behavior=::NewEnumerator1
    Loop Delay=$User.Delay
    Loop Duration Mode=::NewEnumerator1
    UseLoopDelay=true
   )
   &SpawnBurst_Instantaneous(Use Spawn Probability=false, UseLoopCountLimit=false)
  ]
  particleSpawnScript [&InitializeParticle(Color=$System.Color, UsePositionOffset=true, Write Material Random=true)]
  particleUpdateScript [
   &ParticleState(Kill Particles When Lifetime Has Elapsed=false, Let Infinitely Lived Particles Die When Emitter Deactivates=true)
   &Light_Attributes(Falloff Exponent=1, Light Radius=150, Radius=1)
   &ScaleColor(
    Scale RGB=&VectorFromFloat(Value=&Waveform("Waveform [1]"=::NewEnumerator12, "[1] Frequency"=2.5, "[1] Amplitude Min/Max"=<25 40>))
    ScaleA=false
    ScaleRGB=true
   )
  ]
  rendererClasses [&NiagaraLightRendererProperties]
  renderers [{rendererClass=&NiagaraLightRendererProperties, rendererIndex=0}]
 }
}
summary {
 systemName=NS_TidalCrane_Rank_UI
 emitters [
  {emitterName=Em_Ranking, rendererClasses []}
  {emitterName=Em_HoloBeams_Wide, rendererClasses []}
  {bEnabled=false, emitterName=Em_HoloBits, rendererClasses []}
  {bEnabled=false, emitterName=Em_HoloBeams_Wide001, rendererClasses []}
  {emitterName=SimpleSpriteBurst, rendererClasses [&NiagaraLightRendererProperties]}
 ]
 userVariables [
  User.bIsMMOPod {type=NiagaraBool}
  User.Delay {type=NiagaraFloat, defaultValue=3.5, description=""}
  User.Rank {type=NiagaraFloat, defaultValue=0, description=""}
  User.Unlocked {type=NiagaraBool, defaultValue=true}
 ]
}
