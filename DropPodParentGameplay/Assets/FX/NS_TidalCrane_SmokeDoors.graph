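# Niagara system for Tidal Crane door smoke, rendering rising steam sprites.
# Steam splits particles into four partitions that start on opposite sides and drift outward and up under drag and positive Z gravity.
# User.bIsGold enlarges the offsets, speed and sprite size.
# Time-of-day color and alpha scaling tint the darkening, fading sub-UV puffs; a second Mist emitter with tangential swirl is disabled.
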NiagaraSystem /DropPodParentGameplay/Assets/FX/NS_TidalCrane_SmokeDoors
dependencies {
 usedDataInterfaces [
  &NiagaraDataInterfaceArrayFloat3
  &NiagaraDataInterfaceCurve
  &NiagaraDataInterfaceColorCurve
  &NiagaraDataInterfaceSpriteRendererInfo
 ]
 usedDynamicInputs [&SetBoolByFloatComparison]
 usedModules [
  &SystemState
  &self:SystemSpawnScript.SystemScriptSource.SystemScriptGraph.NiagaraNodeAssignment_0.SetVariables_#1
  &Timeline
 ]
}
emitters {
 Mist {
  bEnabled=false
  emitterUpdateScript [
   &EmitterState()
   &SpawnRate(SpawnRate=20, Use Spawn Probability=false)
  ]
  particleSpawnScript [
   &InitializeParticle(
    Color=<rgba 0.75 0.75 0.75 0.4>
    Lifetime Min=1.5
    Lifetime Mode=::NewEnumerator1
    Mass Mode=::NewEnumerator3
    Sprite Rotation Mode=::NewEnumerator2
    Sprite Size Mode=::NewEnumerator3
    Uniform Sprite Size=&RandomRangeFloat(Minimum=75, Maximum=150)
    UsePositionOffset=true
    Write Material Random=true
   )
   &ShapeLocation(MaskSpawnGroup=false, Override Seed=false, Ring Radius=1, Shape Primitive=::NewEnumerator4, UseOwnerScale=true)
   &AddVelocity(
    Velocity=$Particles.ShapeLocation.ShapeTangent
    Velocity Speed Scale=&RandomRangeFloat(Minimum=22, Maximum=66)
   )
   &AddVelocity(Velocity=<0 0 1>, Velocity Speed Scale=33) {module=AddVelocity001}
   &CameraOffset(Camera Offset Amount=20)
  ]
  particleUpdateScript [
   &ParticleState()
   &ScaleSpriteSize(Uniform Curve Sprite Scale=DataInterface() {propertyValues {Curve {defaultValue=0, keys [{tangentMode=RCTM_User, time=0, value=0.5}, {tangentMode=RCTM_User, time=1, value=1}]}}})
   &ScaleColor(
    Scale Alpha=&FloatFromCurve(
     FloatCurve=DataInterface() {propertyValues {Curve {keys [
         {arriveTangent=8.6057, interpMode=RCIM_Cubic, leaveTangent=8.6057, tangentMode=RCTM_User, time=0, value=0}
         {interpMode=RCIM_Cubic, tangentMode=RCTM_User, time=0.2, value=1}
         {interpMode=RCIM_Cubic, tangentMode=RCTM_User, time=1, value=0}
        ]}}}
     Scale Curve=0.35
    )
    Scale RGB=&MakeVectorFromLinearColorRGB(Color=&HueShiftLinearColor(
      Hue=0
      Saturation=&FloatFromCurve(FloatCurve=DataInterface() {propertyValues {Curve {
          defaultValue=0
          keys [
           {arriveTangent=2.3888, interpMode=RCIM_Cubic, leaveTangent=2.3888, tangentMode=RCTM_User, time=0, value=0.5}
           {arriveTangent=1.776, interpMode=RCIM_Cubic, leaveTangent=1.776, time=0.0905, value=0.7276}
           {arriveTangent=-1.014, interpMode=RCIM_Cubic, leaveTangent=-1.014, tangentMode=RCTM_User, time=0.1408, value=0.75}
           {arriveTangent=-0.0405, interpMode=RCIM_Cubic, leaveTangent=-0.0405, tangentMode=RCTM_User, time=1, value=0.5}
          ]
         }}})
      Value=&FloatFromCurve(FloatCurve=DataInterface() {propertyValues {Curve {
          defaultValue=0
          keys [
           {arriveTangent=-7.0816, interpMode=RCIM_Cubic, leaveTangent=-7.0816, tangentMode=RCTM_User, time=0, value=2.5}
           {tangentMode=RCTM_User, time=0.5, value=1}
          ]
         }}})
      Linear Color=&MultiplyLinearColor(
       Linear Color=&RandomRangeLinearColor(Minimum=<rgba 0.7747 0.8936 0.9115 1>, Maximum=<rgba 0.8203 0.9115 0.907 1>)
       Scale Factor=&MakeVector4FromVectorAndFloat(
        "Vector (RGB)"=&MakeVectorFromLinearColorRGB(Color=&MultiplyLinearColor(
          Linear Color=&ColorFromCurve(ColorCurve=DataInterface() {propertyValues {
             AlphaCurve {keys [<at 0 1 RCIM_Cubic>]}
             BlueCurve {keys [<at 0 1 RCIM_Cubic>,<at 1 0.594 RCIM_Cubic>]}
             GreenCurve {keys [<at 0 1 RCIM_Cubic>,<at 1 0.6784 RCIM_Cubic>]}
             RedCurve {keys [<at 0 1 RCIM_Cubic>,<at 1 0.7121 RCIM_Cubic>]}
            }})
          Scale Factor=&MakeVector4_FromLinearColor(LinearColor=$System.TODColor)
         ))
        "Float (Alpha)"=1
       )
      )
     ))
    ScaleA=true
    ScaleRGB=true
   )
   &SubUVAnimation(
    "Sprite Renderer "=DataInterface() {propertyValues {SpriteRenderer=&self:Mist_1.NiagaraSpriteRendererProperties_0}}
    UseEndFrame=false
    UseStartFrame=false
   )
   &Drag(Rotational Drag=2, Use Linear Drag=true, Use Rotational Drag=true)
   &AccelerationForce(Acceleration=&Multiply_VectorByFloat(Vector=$Particles.ShapeLocation.ShapeTangent, Float=99))
   &SolveForcesAndVelocity(Clamp Velocity=false, Limit Acceleration=false)
   &DynamicMaterialParameters(Module.Index0Param0Write=true, Module.Index0Param1Write=true, Module.Index0Param2Write=true, Module.Index0Param3Write=true)
  ]
  rendererClasses [&NiagaraSpriteRendererProperties]
  renderers [{rendererClass=&NiagaraSpriteRendererProperties}]
 }
 Steam {
  emitterUpdateScript [
   &EmitterState(Life Cycle Mode=::NewEnumerator1, Loop Behavior=::NewEnumerator1, UseLoopDelay=true)
   &SpawnRate(SpawnRate=20, Use Spawn Probability=false)
  ]
  particleSpawnScript [
   &PartitionParticles(Number Of Partitions=4)
   &InitializeParticle(
    Color=<rgba 0.75 0.75 0.75 0.2>
    Lifetime Min=1.25
    Lifetime Mode=::NewEnumerator1
    Position Offset=&Multiply_Vector(
     A=&Multiply_Vector(
      A=<50 50 50>
      B=&MakeCustomVectorFromBool(Bool=$User.bIsGold, True Vector=<1.5 1.5 1>, False Vector=<1 1 1>)
     )
     B=&SelectVectorFromArray(
      Direct Array Index=$Particles.PartitionParticles.Partition
      Vector Selection Array=DataInterface() {propertyValues {FloatData [<-1 0 1>, <1 0 0.75>, <0 1 1>, <0 -1 0.75>]}}
     )
    )
    Sprite Rotation Mode=::NewEnumerator2
    Sprite Size Max=<80 200>
    Sprite Size Min=<60 150>
    Sprite Size Mode=::NewEnumerator2
    UsePositionOffset=true
    Write Material Random=true
   )
   &InheritVelocity(Module.Limit Speed=true, Velocity Limit=5, Velocity Scale=<0.25 0.25 0.25>)
   &CameraOffset(Camera Offset Amount=50)
   &AddVelocity(
    Velocity=&SelectVectorFromArray(
     Direct Array Index=$Particles.PartitionParticles.Partition
     Vector Selection Array=DataInterface() {propertyValues {FloatData [<-0.5 0 1>, <0.5 0 1>, <0 0.5 1>, <0 -0.5 1>]}}
    )
    Velocity Speed Scale=&Multiply_Float(
     A=&RandomRangeFloat(Minimum=50, Maximum=150)
     B=&MakeCustomFloatFromBool(Bool=$User.bIsGold, True Float=1.5, False Float=1)
    )
   )
  ]
  particleUpdateScript [
   &ParticleState()
   &SubUVAnimation(
    Sub UV Play Rate:NiagaraFloat=&RandomRangeFloat(Minimum=0.05, Maximum=0.1)
    SubUV Animation Mode:ENiagara_SubUVLookupMode=::NewEnumerator3
    SubUV Start Time Offset:NiagaraFloat=&RandomRangeFloat(Minimum=0, Maximum=63)
   )
   &DynamicMaterialParameters(
    Index 0 Param 1=&FloatFromCurve(FloatCurve=DataInterface() {propertyValues {Curve {keys [<at 0 -0.25>,<at 1 -1>]}}})
    Index 0 Param 4=&OneMinusFloat(Float=$Particles.NormalizedAge)
    Module.Param0WriteEnabled=true
    Module.Param1WriteEnabled=false
    Module.Param2WriteEnabled=false
    Module.Param3WriteEnabled=true
   )
   &ScaleColor(
    Scale Alpha=&FloatFromCurve(
     FloatCurve=DataInterface() {propertyValues {Curve {keys [
         {arriveTangent=11.8579, interpMode=RCIM_Cubic, leaveTangent=11.8579, tangentMode=RCTM_User, time=0, value=0}
         {arriveTangent=-2.9102, interpMode=RCIM_Cubic, leaveTangent=-2.9102, tangentMode=RCTM_User, time=0.1756, value=1.0142}
         {arriveTangent=-0.1231, interpMode=RCIM_Cubic, leaveTangent=-0.1231, tangentMode=RCTM_User, time=0.9987, value=0}
        ]}}}
     Scale Curve=&FloatFromCurve(
      FloatCurve=DataInterface() {propertyValues {Curve {
         defaultValue=1
         keys [
          {arriveTangent=4.1785, interpMode=RCIM_Cubic, leaveTangent=4.1785, tangentMode=RCTM_User, time=0, value=0}
          {arriveTangent=1.1179, interpMode=RCIM_Cubic, leaveTangent=1.1179, tangentMode=RCTM_User, time=0.2042, value=0.6228}
          {arriveTangent=0.0224, interpMode=RCIM_Cubic, leaveTangent=0.0224, tangentMode=RCTM_User, time=1, value=1}
         ]
        }}}
      CurveIndex=$Engine.System.Age
      Scale Curve=&FNTimeOfDayScalar(Night Scale=2, Morning Scale=1.5, Day Scale=1.25, Evening Scale=1.5)
     )
    )
    Scale RGB=&MakeVectorFromLinearColorRGB(Color=&MultiplyLinearColor(
      Linear Color=&ColorFromCurve(ColorCurve=DataInterface() {propertyValues {
         AlphaCurve {keys [<at 0 1 RCIM_Cubic>]}
         BlueCurve {keys [<at 0 1 RCIM_Cubic>,<at 1 0.019 RCIM_Cubic>]}
         GreenCurve {keys [<at 0 1 RCIM_Cubic>,<at 1 0.0243 RCIM_Cubic>]}
         RedCurve {keys [<at 0 1 RCIM_Cubic>,<at 1 0.0335 RCIM_Cubic>]}
        }})
      Scale Factor=&MakeVector4_FromLinearColor(LinearColor=$System.TODColor)
     ))
    ScaleA=true
    ScaleRGB=true
   )
   &ScaleSpriteSize(Scale Factor=&Multiply_Vector2D_ByFloat(
     Float=&MakeCustomFloatFromBool(Bool=$User.bIsGold, True Float=1.5, False Float=1)
     Vector2D=&MakeVector2D(
      X=&FloatFromCurve(
       FloatCurve=DataInterface() {propertyValues {Curve {keys [
           {arriveTangent=0.2384, interpMode=RCIM_Cubic, leaveTangent=0.2384, tangentMode=RCTM_User, time=0, value=0.25}
           {arriveTangent=-0.0499, interpMode=RCIM_Cubic, leaveTangent=-0.0499, tangentMode=RCTM_User, time=1, value=1.5}
          ]}}}
       Scale Curve=6
      )
      Y=&FloatFromCurve(
       FloatCurve=DataInterface() {propertyValues {Curve {keys [
           {arriveTangent=0.2384, interpMode=RCIM_Cubic, leaveTangent=0.2384, tangentMode=RCTM_User, time=0, value=0.25}
           {arriveTangent=0.72, interpMode=RCIM_Cubic, leaveTangent=0.72, tangentMode=RCTM_User, time=1, value=1.25}
          ]}}}
       Scale Curve=3.5
      )
     )
    ))
   &Drag(Drag=2, Use Linear Drag=true, Use Rotational Drag=false)
   &GravityForce(Gravity=<0 0 111>)
   &SolveForcesAndVelocity(Clamp Velocity=false, Limit Acceleration=false)
  ]
  rendererClasses [&NiagaraSpriteRendererProperties]
  renderers [{rendererClass=&NiagaraSpriteRendererProperties}]
 }
}
summary {
 systemName=NS_TidalCrane_SmokeDoors
 emitters [
  {emitterName=Steam, rendererClasses [&NiagaraSpriteRendererProperties]}
  {bEnabled=false, emitterName=Mist, rendererClasses [&NiagaraSpriteRendererProperties]}
 ]
 userVariables [
  User.bIsGold {type=NiagaraBool}
  User.Delay {type=NiagaraFloat, defaultValue=0, description=""}
 ]
}
