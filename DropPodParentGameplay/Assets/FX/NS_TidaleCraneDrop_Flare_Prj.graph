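# Niagara system for a blue flare glow, apparently on the TidaleCrane drop projectile, built from one Flare emitter.
# It spawns a single persistent particle on the first loop, after User.Loop_Delay.
# The particle drives two sprite layers sized by User.FX_Scale and a light renderer.
# Sine waves on emitter age flicker each color, two multiplied for the light, while time-of-day scalars adjust brightness and alpha.
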NiagaraSystem /DropPodParentGameplay/Assets/FX/NS_TidaleCraneDrop_Flare_Prj
dependencies {
 usedDataInterfaces []
 usedDynamicInputs [&SimulationPosition]
 usedModules [&SystemState]
}
emitters {Flare {
  emitterUpdateScript [
   &EmitterState(
    Life Cycle Mode=::NewEnumerator1
    Loop Behavior=::NewEnumerator1
    Loop Delay=$User.Loop_Delay
    Loop Duration Mode=::NewEnumerator1
    UseLoopDelay=true
   )
   &SpawnBurst_Instantaneous(Spawn Count=1, Spawn Probability=HlslExpression() {hlslExpression="Emitter.LoopCount == 0"}, Use Spawn Probability=true)
  ]
  particleSpawnScript [&InitializeParticle(UsePositionOffset=false, Write Material Random=true) {module=InitializeParticle001}]
  particleUpdateScript [
   &ParticleState(Kill Particles When Lifetime Has Elapsed=false)
   &FN_Sprite_Attributes(
    Color=&MultiplyLinearColor(
     Linear Color=<rgba 0 2.1119 5 1>
     Scale Factor=&MakeVector4_FromFloat(Float=&FNTimeOfDayScalar(Night Scale=1, Morning Scale=1, Day Scale=0.7, Evening Scale=1))
    )
    Create Color Attribute=true
    Create SpriteSize Attribute=true
    SpriteSize=&Multiply_Vector2D_ByFloat(Float=&Multiply_Float(A=$User.FX_Scale, B=0.5), Vector2D=<75 75>)
    Write to Module Namespace=true
   ) {module=FN_Sprite_Attributes002}
   &FN_Sprite_Attributes(
    Color=&MultiplyLinearColor(
     Linear Color=<rgba 0 0.6063 2 2>
     Scale Factor=&MakeVector4_FromFloat(Float=&FNTimeOfDayScalar(Night Scale=0.6, Morning Scale=0.3, Day Scale=0.3, Evening Scale=0.3))
    )
    Create Color Attribute=true
    Create SpriteSize Attribute=true
    SpriteSize=&Multiply_Vector2D_ByFloat(Float=&Multiply_Float(A=$User.FX_Scale, B=0.5), Vector2D=<100 100>)
    Write to Module Namespace=true
   ) {module=FN_Sprite_Attributes003}
   /SetVariables(Particles.FN_Sprite_Attributes002.Color:LinearColor=&MultiplyLinearColor(
     Linear Color=$Particles.FN_Sprite_Attributes002.Color
     Scale Factor=&MakeFloat4FromVectorAndFloat(
      "Vector (RGB)"=&VectorFromFloat(Value=&Sine(
        Normalized Angle=$Emitter.Age
        Period=&RandomRangeFloat(Minimum=0.15, Maximum=0.25)
        Scale=0.2
        Bias=0.6
       ))
      "Float (Alpha)"=&Multiply_Float(A=&FNTimeOfDayScalar(Night Scale=0.2, Morning Scale=0.5, Evening Scale=0.5), B=1)
     )
    )) {isSetParameters=true}
   /SetVariables(Particles.FN_Sprite_Attributes003.Color:LinearColor=&MultiplyLinearColor(
     Linear Color=$Particles.FN_Sprite_Attributes003.Color
     Scale Factor=&MakeFloat4FromVectorAndFloat(
      "Vector (RGB)"=&VectorFromFloat(Value=&Sine(
        Normalized Angle=$Emitter.Age
        Period=&RandomRangeFloat(Minimum=0.2, Maximum=0.3)
        Scale=0.2
        Bias=0.75
       ))
      "Float (Alpha)"=&Multiply_Float(A=&FNTimeOfDayScalar(Night Scale=0.2, Morning Scale=0.5, Day Scale=0.5, Evening Scale=0.5), B=1)
     )
    )) {isSetParameters=true}
   /SetVariables(Particles.Flare_LightColor:LinearColor=&MultiplyLinearColor(
     Linear Color=<rgba 0 3652.5249 15000 1>
     Scale Factor=&MakeFloat4FromVectorAndFloat(
      "Vector (RGB)"=&VectorFromFloat(Value=&Multiply_Float(
        A=&Sine(Normalized Angle=$Emitter.Age, Period=&RandomRangeFloat(Minimum=0.22, Maximum=0.22), Scale=0.2, Bias=1)
        B=&Sine(
         Normalized Angle=$Emitter.Age
         Period=&RandomRangeFloat(Minimum=0.33, Maximum=0.33)
         Scale=0.2
         Bias=0.5
        )
       ))
      "Float (Alpha)"=&Multiply_Float(A=&FNTimeOfDayScalar(Night Scale=0.2, Morning Scale=0.5, Evening Scale=0.5), B=1)
     )
    )) {isSetParameters=true}
  ]
  rendererClasses [&NiagaraSpriteRendererProperties, &NiagaraLightRendererProperties]
  renderers [
   {rendererClass=&NiagaraSpriteRendererProperties}
   {rendererClass=&NiagaraSpriteRendererProperties, rendererIndex=1}
   {rendererClass=&NiagaraLightRendererProperties, rendererIndex=2}
  ]
 }}
summary {
 systemName=NS_TidaleCraneDrop_Flare_Prj
 emitters [{emitterName=Flare, rendererClasses [&NiagaraSpriteRendererProperties, &NiagaraLightRendererProperties]}]
 userVariables [
  User.FX_Scale {type=NiagaraFloat, defaultValue=10, description=""}
  User.Loop_Delay {type=NiagaraFloat, defaultValue=0, description=""}
 ]
}
