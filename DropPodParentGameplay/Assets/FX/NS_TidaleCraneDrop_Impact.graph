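# Niagara system for a fiery crane-drop impact, named for the Tidale crane drop in drop pod gameplay.
# Combined emitters pack flash, glow, shockwave, light and burst meshes into multi-renderer particles, using spawn-group visibility tags and per-layer lifetimes.
# GPU debris, falling drops and sparkles scatter outward, while a decal emitter fades cracks and scorch.
# System and User parameters scale layers; time-of-day scalars feed one mesh's material.
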NiagaraSystem /DropPodParentGameplay/Assets/FX/NS_TidaleCraneDrop_Impact
dependencies {
 usedDataInterfaces [
  &NiagaraDataInterfaceCurve
  &NiagaraDataInterfaceColorCurve
  &NiagaraDataInterfaceVectorCurve
  &NiagaraDataInterfaceVector2DCurve
  &NiagaraDataInterfaceCamera
 ]
 usedDynamicInputs [
  &SimulationPosition
  &ClampFloat
  &Subtract_Float
  &VectorLength
  &Subtract_Position
  &ReturnCameraPosition
 ]
 usedModules [
  &self:SystemSpawnScript.SystemScriptSource.SystemScriptGraph.NiagaraNodeAssignment_2.SetVariables_#1
  &self:SystemSpawnScript.SystemScriptSource.SystemScriptGraph.NiagaraNodeAssignment_3.SetVariables_#2
  &self:SystemSpawnScript.SystemScriptSource.SystemScriptGraph.NiagaraNodeAssignment_4.SetVariables_#3
  &self:SystemSpawnScript.SystemScriptSource.SystemScriptGraph.NiagaraNodeAssignment_7.SetVariables_#4
  &self:SystemSpawnScript.SystemScriptSource.SystemScriptGraph.NiagaraNodeAssignment_0.SetVariables_#5
  &self:SystemSpawnScript.SystemScriptSource.SystemScriptGraph.NiagaraNodeAssignment_1.SetVariables_#6
  &SystemState
 ]
}
emitters {
 Combined_Emitter_Mesh {
  emitterUpdateScript [
   &EmitterState(Life Cycle Mode=::NewEnumerator1, Loop Behavior=::NewEnumerator1, Loop Duration=1, UseLoopDelay=false)
   &SpawnBurst_Instantaneous(Spawn Time=0.075, Use Spawn Probability=false, UseLoopCountLimit=false)
   &SpawnBurst_Instantaneous(Spawn Group=1, Use Spawn Probability=false, UseLoopCountLimit=false) {module=SpawnBurst_Instantaneous001}
   &SpawnBurst_Instantaneous(Spawn Count=2, Spawn Group=2, Spawn Time=0.1, Use Spawn Probability=false, UseLoopCountLimit=false) {module=SpawnBurst_Instantaneous002}
   &SpawnBurst_Instantaneous(Spawn Count=4, Spawn Group=3, Use Spawn Probability=false, UseLoopCountLimit=false) {module=SpawnBurst_Instantaneous003}
  ]
  particleSpawnScript [
   /SetVariables(Particles.SpawnGroup=$Emitter.SpawnGroup) {isSetParameters=true}
   /SetVariables(Particles.VisibilityTag=$Particles.SpawnGroup) {isSetParameters=true}
   &InitializeParticle(Color Mode=::NewEnumerator0, UsePositionOffset=false, Write Material Random=true)
   /SetVariables(
    Particles.Mesh_01_CameraOffset=100
    Particles.Mesh_01_Orientation:Quat4f=&OrientMeshToVector(Look At Direction=<1 0 0>, Reference Axis Vector=<1 0 0>)
    Particles.Mesh_01_Size:Vector3f=&VectorFromFloat(Value=&Multiply_Float(A=3, B=$System.ScaleFX))
   ) {isSetParameters=true, module=SetVariables_#7}
   /SetVariables(
    Particles.Mesh_02_Color:LinearColor=<rgba 1 0.4518 0.0403 1>
    Particles.Mesh_02_Orientation:Quat4f=&OrientMeshToVector(Look At Direction=<0 1 0>, Reference Axis Vector=<1 0 0>)
    Particles.Mesh_02_Size:Vector3f=&Multiply_VectorByFloat(Vector=<0.6 0.05 0.6>, Float=$System.ScaleFX)
   ) {isSetParameters=true}
   /SetVariables(
    Particles.Mesh_03_Color:LinearColor=<rgba 1 0.881 1 0.375>
    Particles.Mesh_03_Orientation:Quat4f=&MultiplyQuaternion(
     Quaternion A=&OrientMeshToVector(Look At Direction=<1 0 0>, Reference Axis Vector=<1 0 0>)
     Quaternion B=&MakeQuatFromAxisAngle(Axis=<0 0 1>, AngleInDegrees=&RandomRangeFloat(Minimum=-180, Maximum=180))
    )
    Particles.Mesh_03_Size:Vector3f=&VectorFromFloat(Value=&Multiply_Float(A=1.8, B=$System.ScaleFX))
   ) {isSetParameters=true}
   /SetVariables(
    Particles.Mesh_04_Color:LinearColor=<rgba 1 1 1 0.275>
    Particles.Mesh_04_Orientation:Quat4f=&MultiplyQuaternion(
     Quaternion A=&MultiplyQuaternion(
      Quaternion A=&OrientMeshToVector(Look At Direction=<1 0 0>, Reference Axis Vector=<1 0 0>)
      Quaternion B=&MakeQuatFromAxisAngle(Axis=<0 0 1>, AngleInDegrees=&RandomRangeFloat(Minimum=-180, Maximum=180))
     )
     Quaternion B=&MakeQuatFromAxisAngle(
      Axis=<1 0 0>
      AngleInDegrees=&RandomRangeFloat(Minimum=-20, Maximum=20, Randomness Mode=::NewEnumerator1)
     )
    )
    Particles.Mesh_04_Size:Vector3f=&VectorFromFloat(Value=&Multiply_Float(A=&RandomRangeFloat(Minimum=1.25, Maximum=4), B=$System.ScaleFX))
   ) {isSetParameters=true}
  ]
  particleUpdateScript [
   &ParticleState()
   /SetVariables(
    Particles.Mesh_01_Color:LinearColor=&MultiplyLinearColor(
     Linear Color=$Particles.Initial.Mesh_01_Color
     Scale Factor=&MakeFloat4FromVectorAndFloat(
      "Vector (RGB)"=&VectorFromCurve(
       VectorCurve=DataInterface() {propertyValues {XCurve {keys [<at 0 1>,<at 1 0.526>]}, YCurve {keys [<at 0 1>,<at 1 0.526>]}, ZCurve {keys [<at 0 1>,<at 1 0.526>]}}}
       CurveIndex=&Divide_Float(A=$Particles.NormalizedAge, B=0.7)
      )
      "Float (Alpha)"=&FloatFromCurve(
       FloatCurve=DataInterface() {propertyValues {Curve {keys [<at 0 1>,<at 0.1 0.5>,<at 1 0>]}}}
       CurveIndex=&Divide_Float(A=$Particles.NormalizedAge, B=0.7)
       Scale Curve=0.075
      )
     )
    )
    Particles.Mesh_01_DynamicParameter:Vector4f=&MakeVector4(
     X=&FloatFromCurve(
      FloatCurve=DataInterface() {propertyValues {Curve {keys [{arriveTangent=2.5068, interpMode=RCIM_Cubic, leaveTangent=2.5068, tangentMode=RCTM_User, time=0, value=-0.4}, <at 1 0.55>]}}}
      CurveIndex=&Divide_Float(A=$Particles.NormalizedAge, B=0.7)
     )
     Y=1
     Z=0
     W=0
    )
    Particles.Mesh_01_Size:Vector3f=&Multiply_Vector(
     A=$Particles.Initial.Mesh_01_Size
     B=&VectorFromFloat(Value=&FloatFromCurve(
       FloatCurve=DataInterface() {propertyValues {Curve {keys [
           {arriveTangent=3.5711, interpMode=RCIM_Cubic, leaveTangent=3.5711, tangentMode=RCTM_User, time=0, value=0}
           {interpMode=RCIM_Cubic, tangentMode=RCTM_User, time=1, value=1.5}
          ]}}}
       CurveIndex=&Divide_Float(A=$Particles.NormalizedAge, B=0.7)
       Scale Curve=1.25
      ))
    )
   ) {isSetParameters=true, module=SetVariables_#8}
   /SetVariables(
    Particles.Mesh_02_Color:LinearColor=&MultiplyLinearColor(
     Linear Color=$Particles.Initial.Mesh_02_Color
     Scale Factor=&MakeFloat4FromVectorAndFloat(
      "Vector (RGB)"=&VectorFromCurve(
       VectorCurve=DataInterface() {propertyValues {
         XCurve {keys [
           <at 0 1 RCIM_Cubic>
           {arriveTangent=24.4178, interpMode=RCIM_Cubic, leaveTangent=24.4178, time=0.1536, value=2}
           {arriveTangent=-1.1814, interpMode=RCIM_Cubic, leaveTangent=-1.1814, time=0.7781, value=20}
           <at 1 1 RCIM_Cubic>
          ]}
         YCurve {keys [
           <at 0 1 RCIM_Cubic>
           {arriveTangent=13.7464, interpMode=RCIM_Cubic, leaveTangent=13.7464, time=0.1536, value=0.6932}
           {arriveTangent=0.3625, interpMode=RCIM_Cubic, leaveTangent=0.3625, time=0.7781, value=11.6964}
           <at 1 1 RCIM_Cubic>
          ]}
         ZCurve {keys [
           <at 0 1 RCIM_Cubic>
           {arriveTangent=6.91, interpMode=RCIM_Cubic, leaveTangent=6.91, time=0.1536, value=0.1681}
           {arriveTangent=0.9829, interpMode=RCIM_Cubic, leaveTangent=0.9829, time=0.7781, value=6.3768}
           <at 1 1 RCIM_Cubic>
          ]}
        }}
       CurveIndex=&Divide_Float(A=$Particles.NormalizedAge, B=0.3)
       Scale Curve=&VectorFromFloat(Value=10)
      )
      "Float (Alpha)"=&FloatFromCurve(
       FloatCurve=DataInterface() {propertyValues {Curve {keys [
           {interpMode=RCIM_Cubic, leaveTangentWeight=0.3333, tangentMode=RCTM_User, time=0, value=0}
           {
            arriveTangent=-3.9986
            arriveTangentWeight=0.0551
            interpMode=RCIM_Cubic
            leaveTangent=-3.9986
            leaveTangentWeight=0.2783
            tangentMode=RCTM_User
            time=0.1652
            value=1
           }
           {arriveTangentWeight=0.2783, interpMode=RCIM_Cubic, tangentMode=RCTM_User, time=1, value=0}
          ]}}}
       CurveIndex=&Divide_Float(A=$Particles.NormalizedAge, B=0.3)
      )
     )
    )
    Particles.Mesh_02_DynamicParameter:Vector4f=&MakeVector4(
     X=&FloatFromCurve(
      FloatCurve=DataInterface() {propertyValues {Curve {keys [{interpMode=RCIM_Cubic, tangentMode=RCTM_User, time=0, value=1}, {interpMode=RCIM_Cubic, tangentMode=RCTM_User, time=1, value=0}]}}}
      CurveIndex=&Divide_Float(A=$Particles.NormalizedAge, B=0.3)
     )
     Y=0
     Z=0
     W=1
    )
    Particles.Mesh_02_Size:Vector3f=&Multiply_Vector(
     A=$Particles.Initial.Mesh_02_Size
     B=&VectorFromFloat(Value=&FloatFromCurve(
       FloatCurve=DataInterface() {propertyValues {Curve {keys [
           {arriveTangent=4.1785, interpMode=RCIM_Cubic, leaveTangent=4.1785, tangentMode=RCTM_User, time=0, value=0}
           {arriveTangent=0.9995, interpMode=RCIM_Cubic, leaveTangent=0.9995, tangentMode=RCTM_User, time=0.2223, value=0.6825}
           {arriveTangent=-0.0405, interpMode=RCIM_Cubic, leaveTangent=-0.0405, tangentMode=RCTM_User, time=1, value=1}
          ]}}}
       CurveIndex=&Divide_Float(A=$Particles.NormalizedAge, B=0.3)
       Scale Curve=2.5
      ))
    )
   ) {isSetParameters=true}
   /SetVariables(
    Particles.Mesh_03_Color:LinearColor=&MultiplyLinearColor(
     Linear Color=$Particles.Initial.Mesh_03_Color
     Scale Factor=&MakeFloat4FromVectorAndFloat(
      "Vector (RGB)"=<1 1 1>
      "Float (Alpha)"=&FloatFromCurve(
       FloatCurve=DataInterface() {propertyValues {Curve {keys [
           {arriveTangent=6.2301, interpMode=RCIM_Cubic, leaveTangent=6.2301, tangentMode=RCTM_User, time=0, value=0.5}
           {arriveTangent=-0.7357, interpMode=RCIM_Cubic, leaveTangent=-0.7357, tangentMode=RCTM_User, time=0.1, value=2}
           <at 0.4 0.2>
           {arriveTangent=0.4048, interpMode=RCIM_Cubic, leaveTangent=0.4048, tangentMode=RCTM_User, time=0.7, value=0}
          ]}}}
       CurveIndex=&Divide_Float(A=$Particles.NormalizedAge, B=0.5)
       Scale Curve=0.5
      )
     )
    )
    Particles.Mesh_03_Size:Vector3f=&Multiply_Vector(
     A=$Particles.Initial.Mesh_03_Size
     B=&Multiply_VectorByFloat(
      Vector=&MakeVector(
       X=1
       Y=1
       Z=&FloatFromCurve(
        FloatCurve=DataInterface() {propertyValues {Curve {keys [<at 0 1>,<at 1 0>]}}}
        CurveIndex=&Divide_Float(A=$Particles.NormalizedAge, B=0.5)
       )
      )
      Float=&FloatFromCurve(
       FloatCurve=DataInterface() {propertyValues {Curve {keys [
           {arriveTangent=7.2197, interpMode=RCIM_Cubic, leaveTangent=7.2197, tangentMode=RCTM_User, time=0, value=0}
           {arriveTangent=1.9952, interpMode=RCIM_Cubic, leaveTangent=1.9952, time=0.4917, value=1.9432}
           {arriveTangent=-0.0405, interpMode=RCIM_Cubic, leaveTangent=-0.0405, tangentMode=RCTM_User, time=1.0024, value=2}
          ]}}}
       CurveIndex=&Divide_Float(A=$Particles.NormalizedAge, B=0.5)
      )
     )
    )
   ) {isSetParameters=true}
   /SetVariables(
    Particles.Mesh_04_Color:LinearColor=&MultiplyLinearColor(
     Linear Color=$Particles.Initial.Mesh_04_Color
     Scale Factor=&MakeFloat4FromVectorAndFloat(
      "Vector (RGB)"=<1 1 1>
      "Float (Alpha)"=&FloatFromCurve(
       FloatCurve=DataInterface() {propertyValues {Curve {keys [
           {arriveTangent=8.2356, interpMode=RCIM_Cubic, leaveTangent=8.2356, tangentMode=RCTM_User, time=0, value=0}
           {arriveTangent=0.1969, interpMode=RCIM_Cubic, leaveTangent=0.1969, tangentMode=RCTM_User, time=0.2223, value=1}
           {arriveTangent=-3.0486, interpMode=RCIM_Cubic, leaveTangent=-3.0486, tangentMode=RCTM_User, time=0.4781, value=0.509}
           {arriveTangent=-0.2378, interpMode=RCIM_Cubic, leaveTangent=-0.2378, tangentMode=RCTM_User, time=1, value=0}
          ]}}}
       CurveIndex=&Divide_Float(A=$Particles.NormalizedAge, B=&RandomRangeFloat(Minimum=0.5, Maximum=0.75))
      )
     )
    )
    Particles.Mesh_04_DynamicParameter:Vector4f=&MakeVector4(
     X=&FloatFromCurve(
      FloatCurve=DataInterface() {propertyValues {Curve {keys [{interpMode=RCIM_Cubic, tangentMode=RCTM_User, time=0.5, value=0}, {interpMode=RCIM_Cubic, tangentMode=RCTM_User, time=1, value=1}]}}}
      CurveIndex=&Divide_Float(A=$Particles.NormalizedAge, B=&RandomRangeFloat(Minimum=0.5, Maximum=0.75))
     )
     Y=&FNTimeOfDayScalar(Night Scale=0.05, Morning Scale=-0.26, Day Scale=0, Evening Scale=-0.35)
     Z=&FNTimeOfDayScalar(Night Scale=0.5, Morning Scale=0.23, Evening Scale=0.2)
     W=&FNTimeOfDayScalar(Night Scale=0.02, Morning Scale=-0.01, Day Scale=0, Evening Scale=-0.01)
    )
    Particles.Mesh_04_Orientation:Quat4f=&MultiplyQuaternion(
     Quaternion A=$Particles.Initial.Mesh_04_Orientation
     Quaternion B=&MakeQuatFromAxisAngle(
      Axis=&RandomRangeVector(Minimum=<0 0 -1>, Maximum=<0 0 1>)
      AngleInDegrees=&RandomRangeFloat(Minimum=-50, Maximum=50)
     )
    )
    Particles.Mesh_04_Size:Vector3f=&Multiply_Vector(
     A=$Particles.Initial.Mesh_04_Size
     B=&Multiply_VectorByFloat(
      Vector=&MakeVector(
       X=&FloatFromCurve(
        FloatCurve=DataInterface() {propertyValues {Curve {keys [
            {arriveTangent=0.8546, interpMode=RCIM_Cubic, leaveTangent=3.1486, tangentMode=RCTM_Break, time=0, value=0}
            {arriveTangent=1.4415, interpMode=RCIM_Cubic, leaveTangent=1.4415, tangentMode=RCTM_User, time=0.3242, value=0.7}
            {arriveTangent=1.2727, interpMode=RCIM_Cubic, leaveTangent=1.2727, tangentMode=RCTM_User, time=1, value=1.5}
           ]}}}
        CurveIndex=&Divide_Float(A=$Particles.NormalizedAge, B=&RandomRangeFloat(Minimum=0.5, Maximum=0.75))
       )
       Y=&FloatFromCurve(
        FloatCurve=DataInterface() {propertyValues {Curve {keys [
            {arriveTangent=0.8546, interpMode=RCIM_Cubic, leaveTangent=3.1486, tangentMode=RCTM_Break, time=0, value=0}
            {arriveTangent=1.4415, interpMode=RCIM_Cubic, leaveTangent=1.4415, tangentMode=RCTM_User, time=0.3242, value=0.7}
            {arriveTangent=1.2727, interpMode=RCIM_Cubic, leaveTangent=1.2727, tangentMode=RCTM_User, time=1, value=1.5}
           ]}}}
        CurveIndex=&Divide_Float(A=$Particles.NormalizedAge, B=&RandomRangeFloat(Minimum=0.5, Maximum=0.75))
       )
       Z=&FloatFromCurve(
        FloatCurve=DataInterface() {propertyValues {Curve {keys [
            {arriveTangent=1.8479, interpMode=RCIM_Cubic, leaveTangent=1.8479, tangentMode=RCTM_User, time=0, value=0}
            {arriveTangent=0.9995, interpMode=RCIM_Cubic, leaveTangent=0.9995, tangentMode=RCTM_User, time=0.4, value=0.4}
            {arriveTangent=0.4855, interpMode=RCIM_Cubic, leaveTangent=0.4855, tangentMode=RCTM_User, time=1, value=0.6}
           ]}}}
        CurveIndex=&Divide_Float(A=$Particles.NormalizedAge, B=&RandomRangeFloat(Minimum=0.5, Maximum=0.75))
        Scale Curve=1.2
       )
      )
      Float=&FloatFromCurve(
       FloatCurve=DataInterface() {propertyValues {Curve {keys [
           {interpMode=RCIM_Cubic, leaveTangent=6.5404, tangentMode=RCTM_Break, time=0, value=1.25}
           {arriveTangent=0.0345, interpMode=RCIM_Cubic, leaveTangent=0.0345, tangentMode=RCTM_User, time=0.0648, value=1.4045}
           {arriveTangent=0.0556, interpMode=RCIM_Cubic, leaveTangent=0.0556, tangentMode=RCTM_User, time=1, value=1.4}
          ]}}}
       CurveIndex=&Divide_Float(A=$Particles.NormalizedAge, B=&RandomRangeFloat(Minimum=0.5, Maximum=0.75))
      )
     )
    )
   ) {isSetParameters=true}
   /SetVariables(Particles.Mesh_Position:NiagaraPosition=&Add_Vector(
     A=Linked() {linkedVariable {name=Particles.Position, type=NiagaraPosition}}
     B=&RandomRangeVector(Minimum=<0 0 25>, Maximum=<0 0 50>)
    )) {isSetParameters=true}
  ]
  rendererClasses [&NiagaraMeshRendererProperties]
  renderers [
   {rendererClass=&NiagaraMeshRendererProperties, rendererIndex=0}
   {rendererClass=&NiagaraMeshRendererProperties, rendererIndex=1}
   {rendererClass=&NiagaraMeshRendererProperties, rendererIndex=2}
   {rendererClass=&NiagaraMeshRendererProperties, rendererIndex=3}
  ]
 }
 Combined_Emitter_Sprite001 {
  emitterUpdateScript [
   &EmitterState(Life Cycle Mode=::NewEnumerator1, Loop Behavior=::NewEnumerator1, Loop Duration=0.5, UseLoopDelay=false)
   &SpawnBurst_Instantaneous(Use Spawn Probability=false, UseLoopCountLimit=false)
  ]
  particleSpawnScript [
   &InitializeParticle(Color Mode=::NewEnumerator0, Mesh Scale=<7.5 7.5 7.5>, Mesh Scale Mode=::NewEnumerator4, UsePositionOffset=false, Write Material Random=true)
   /SetVariables(
    Particles.Sprite_01_CameraOffset=50
    Particles.Sprite_01_Color:LinearColor=<rgba 2 2 2 1>
    Particles.Sprite_01_Rotation:NiagaraFloat=&RandomRangeFloat(Minimum=-360, Maximum=360)
    Particles.Sprite_01_Size:Vector2f=&Multiply_Vector2D_ByFloat(
     Float=&Multiply_Float(A=$System.ScaleFX, B=$User.FX_Scale_Impact_Center)
     Vector2D=&Vector2DFromFloat(Value=&RandomRangeFloat(Minimum=250, Maximum=300))
    )
   ) {isSetParameters=true, module=SetVariables_#7}
   /SetVariables(Particles.Lifetime_DBO=0.3, Particles.Lifetime_Flash=0.2, Particles.Lifetime_Glow=0.5, Particles.Lifetime_Shockwave=0.4) {isSetParameters=true}
   /SetVariables(Particles.DMP_DBO_01:Vector4f=&MakeVector4(X=100, Y=0, Z=0, W=0)) {isSetParameters=true}
   /SetVariables(
    Particles.DMP_BG_00:Vector4f=<1 1 1 1>
    Particles.Position_BG_Offset:NiagaraPosition=<0 0 20>
    Particles.Sprite_BG_CameraOffset=100
    Particles.Sprite_BG_Size:Vector2f=<350 350>
   ) {isSetParameters=true}
   &InitialMeshOrientation(Mesh Orientation Mode=::NewEnumerator2, Rotation=<0 0.25 0>, Use Rotation Vector=true)
  ]
  particleUpdateScript [
   &ParticleState()
   /SetVariables(
    Particles.Sprite_01_Color:LinearColor=&ScaleLinearColorByCurve(
     InitialValue=$Particles.Initial.Sprite_01_Color
     ScaleCurve=DataInterface() {propertyValues {
       AlphaCurve {keys [<at 0 1>,<at 1 0>]}
       BlueCurve {keys [<at 0 0>,<at 0.3744 0>,<at 1 0>]}
       GreenCurve {keys [<at 0 6.6667>,<at 0.3744 1.413>,<at 1 0>]}
       RedCurve {keys [<at 0 20>,<at 0.3744 4.8758>,<at 1 1>]}
      }}
     CurveIndex=&NormalizeFloat(Value=&Divide_Float(A=$Particles.NormalizedAge, B=$Particles.Lifetime_Flash), Min=0, Max=1)
    )
    Particles.Sprite_01_Size:Vector2f=&Multiply_Vector2D(
     A=$Particles.Initial.Sprite_01_Size
     B=&Vector2DFromFloat(Value=&FloatFromCurve(
       FloatCurve=DataInterface() {propertyValues {Curve {keys [
           {arriveTangent=4.1785, interpMode=RCIM_Cubic, leaveTangent=4.1785, tangentMode=RCTM_User, time=0, value=0}
           {arriveTangent=0.9995, interpMode=RCIM_Cubic, leaveTangent=0.9995, tangentMode=RCTM_User, time=0.2, value=0.65}
           {arriveTangent=-0.0405, interpMode=RCIM_Cubic, leaveTangent=-0.0405, tangentMode=RCTM_User, time=1, value=1}
          ]}}}
       CurveIndex=&NormalizeFloat(Value=&Divide_Float(A=$Particles.NormalizedAge, B=$Particles.Lifetime_Flash), Min=0, Max=1)
       Scale Curve=1.25
      ))
    )
   ) {isSetParameters=true, module=SetVariables_#8}
   /SetVariables(
    Particles.Sprite_BG_Color:LinearColor=&MakeLinearColorFromVectorAndFloat(
     "Vector (RGB)"=&Multiply_VectorByFloat(
      Vector=&MakeVectorFromLinearColorRGB(Color=&ColorFromCurve(
        ColorCurve=DataInterface() {propertyValues {
          AlphaCurve {keys [<at 0 1 RCIM_Cubic>]}
          BlueCurve {keys [<at 0.0152 0>, {arriveTangent=0.0005, interpMode=RCIM_Cubic, leaveTangent=0.0005, time=0.6127, value=0}, <at 1 0.0005 RCIM_Cubic>]}
          GreenCurve {keys [
            <at 0.0152 0.2478>
            {arriveTangent=-0.2482, interpMode=RCIM_Cubic, leaveTangent=-0.2482, time=0.6127, value=0.0409}
            <at 1 0.0033 RCIM_Cubic>
           ]}
          RedCurve {keys [<at 0.0152 1>, {arriveTangent=-1.0026, interpMode=RCIM_Cubic, leaveTangent=-1.0026, time=0.6127, value=0.1651}, <at 1 0.0126 RCIM_Cubic>]}
         }}
        CurveIndex=&NormalizeFloat(Value=&Divide_Float(A=$Particles.NormalizedAge, B=$Particles.Lifetime_Glow), Min=0, Max=1)
       ))
      Float=5
     )
     "Float (Alpha)"=&Multiply_Float(
      A=&FloatFromCurve(
       FloatCurve=DataInterface() {propertyValues {Curve {keys [
           {arriveTangent=4.0782, interpMode=RCIM_Cubic, leaveTangent=4.0782, tangentMode=RCTM_User, time=0, value=0}
           {arriveTangent=0.1106, interpMode=RCIM_Cubic, leaveTangent=0.1106, tangentMode=RCTM_User, time=0.0314, value=0.4249}
           {arriveTangent=-0.4386, interpMode=RCIM_Cubic, leaveTangent=-0.4386, time=0.7387, value=0.2967}
           <at 1 0>
          ]}}}
       CurveIndex=&NormalizeFloat(Value=&Divide_Float(A=$Particles.NormalizedAge, B=$Particles.Lifetime_Glow), Min=0, Max=1)
       Scale Curve=2
      )
      B=1
     )
    )
    Particles.Sprite_BG_Size:Vector2f=&Multiply_Vector2D_ByFloat(
     Float=&FloatFromCurve(
      FloatCurve=DataInterface() {propertyValues {Curve {keys [
          {tangentMode=RCTM_User, time=0, value=0.5}
          {arriveTangent=0.5, interpMode=RCIM_Cubic, leaveTangent=0.5, time=0.3276, value=0.8147}
          {tangentMode=RCTM_User, time=1, value=1}
         ]}}}
      CurveIndex=&NormalizeFloat(Value=&Divide_Float(A=$Particles.NormalizedAge, B=$Particles.Lifetime_Glow), Min=0, Max=1)
      Scale Curve=$System.Size
     )
     Vector2D=$Particles.Initial.Sprite_BG_Size
    )
   ) {isSetParameters=true}
   &ScaleMeshSize(Scale Factor=&Multiply_VectorByFloat(
     Vector=<1 1 1>
     Float=&FloatFromCurve(
      FloatCurve=DataInterface() {propertyValues {Curve {keys [
          {arriveTangent=1.396, interpMode=RCIM_Cubic, leaveTangent=1.396, tangentMode=RCTM_User, time=0, value=0}
          {arriveTangent=1.4779, tangentMode=RCTM_User, time=0.4724, value=0.5733}
          {interpMode=RCIM_Cubic, tangentMode=RCTM_User, time=1, value=1}
         ]}}}
      CurveIndex=&NormalizeFloat(Value=&Divide_Float(A=$Particles.NormalizedAge, B=$Particles.Lifetime_Shockwave), Min=0, Max=1)
      Scale Curve=1.25
     )
    ))
   &ScaleColor(
    Scale Alpha=&FloatFromCurve(
     FloatCurve=DataInterface() {propertyValues {Curve {keys [
         {tangentMode=RCTM_User, time=0, value=1}
         {arriveTangent=-0.0405, interpMode=RCIM_Cubic, leaveTangent=-0.0405, tangentMode=RCTM_User, time=0.5011, value=0.7756}
         {arriveTangent=-2.4747, interpMode=RCIM_Cubic, leaveTangent=-2.4747, tangentMode=RCTM_User, time=1, value=0}
        ]}}}
     CurveIndex=&NormalizeFloat(Value=&Divide_Float(A=$Particles.NormalizedAge, B=$Particles.Lifetime_Shockwave), Min=0, Max=1)
     Scale Curve=5
    )
    Scale RGB=&Multiply_VectorByFloat(
     Vector=&MakeVectorFromLinearColorRGB(Color=&ColorFromCurve(
       ColorCurve=DataInterface() {propertyValues {
         AlphaCurve {keys []}
         BlueCurve {keys [
           <at 0.0086 0.2174>
           <at 0.2512 0.0722>
           {arriveTangent=-0.2119, leaveTangent=-0.2119, time=0.5107, value=0.0662}
           <at 0.7726 0.0342>
           <at 0.9945 0>
          ]}
         GreenCurve {keys [
           <at 0.0086 1.413>
           <at 0.2512 0.2432>
           {arriveTangent=-1.5861, leaveTangent=-1.5861, time=0.5107, value=0.1951}
           <at 0.7726 0.0417>
           <at 0.9945 0>
          ]}
         RedCurve {keys [
           <at 0.0086 5>
           <at 0.2512 1.0827>
           {arriveTangent=-2.5029, leaveTangent=-2.5029, time=0.5107, value=0.9219}
           <at 0.7726 0.0841>
           <at 0.9945 0>
          ]}
        }}
       CurveIndex=&NormalizeFloat(Value=&Divide_Float(A=$Particles.NormalizedAge, B=$Particles.Lifetime_Shockwave), Min=0, Max=1)
      ))
     Float=1
    )
    ScaleA=true
    ScaleRGB=true
   )
   &FN_Mesh_Attributes(
    Color=&MakeLinearColorFromVectorAndFloat(
     "Vector (RGB)"=&Multiply_Vector(
      A=&MakeVectorFromLinearColorRGB(Color=<rgba 0.5 0.2139 0.0666 2>)
      B=&MakeVectorFromLinearColorRGB(Color=&ColorFromCurve(
        ColorCurve=DataInterface() {propertyValues {
          AlphaCurve {keys [<at 0 1 RCIM_Cubic>]}
          BlueCurve {keys [<at 0 1 RCIM_Cubic>,<at 1 1 RCIM_Cubic>]}
          GreenCurve {keys [<at 0 1 RCIM_Cubic>,<at 1 0 RCIM_Cubic>]}
          RedCurve {keys [<at 0 1 RCIM_Cubic>,<at 1 0.2703 RCIM_Cubic>]}
         }}
        CurveIndex=&NormalizeFloat(Value=&Divide_Float(A=$Particles.NormalizedAge, B=$Particles.Lifetime_DBO), Min=0, Max=1)
       ))
     )
     "Float (Alpha)"=&FloatFromCurve(
      FloatCurve=DataInterface() {propertyValues {Curve {keys [<at 0 0.5 RCIM_Cubic>, {arriveTangent=-1.2761, tangentMode=RCTM_User, time=1, value=0}]}}}
      CurveIndex=&NormalizeFloat(Value=&Divide_Float(A=$Particles.NormalizedAge, B=$Particles.Lifetime_DBO), Min=0, Max=1)
      Scale Curve=2.5
     )
    )
    Create Color Attribute=true
    Create Orientation Attribute=true
    Create Scale Attribute=true
    Orientation:Quat4f=&MakeQuaternion(
     Quaternion From=0
     Axis=<0 1 0>
     Angle=&RandomRangeFloat(Minimum=0, Maximum=1)
     From=?
     To=?
     Angle Type=::NewEnumerator1
     Normalize Vectors=false
    )
    Scale=&Multiply_VectorByFloat(
     Vector=&Multiply_Vector(
      A=<1 1 1>
      B=&VectorFromFloat(Value=&FloatFromCurve(
        FloatCurve=DataInterface() {propertyValues {Curve {keys [{interpMode=RCIM_Cubic, tangentMode=RCTM_User, time=0, value=0}, {interpMode=RCIM_Cubic, tangentMode=RCTM_User, time=1, value=1}]}}}
        CurveIndex=&NormalizeFloat(Value=&Divide_Float(A=$Particles.NormalizedAge, B=$Particles.Lifetime_DBO), Min=0, Max=1)
        Scale Curve=&Multiply_Float(A=1, B=&Multiply_Float(A=0.3, B=$System.Size))
       ))
     )
     Float=2
    )
    Write to Module Namespace=true
   )
   /SetVariables(Particles.DMP_DBO_00:Vector4f=&MakeVector4(
     X=&Lerp_Float(
      A=0
      B=-0.05
      Alpha=&FloatFromCurve(
       FloatCurve=DataInterface() {propertyValues {Curve {
          defaultValue=0
          keys [
           {interpMode=RCIM_Cubic, tangentMode=RCTM_User, time=0.5, value=0}
           {arriveTangent=5.5119, interpMode=RCIM_Cubic, leaveTangent=5.5119, tangentMode=RCTM_User, time=1, value=1}
          ]
         }}}
       CurveIndex=&NormalizeFloat(Value=&Divide_Float(A=$Particles.NormalizedAge, B=$Particles.Lifetime_DBO), Min=0, Max=1)
      )
     )
     Y=0
     Z=0
     W=0
    )) {isSetParameters=true}
   &DynamicMaterialParameters(
    Index 0 Param 1=&FloatFromCurve(
     FloatCurve=DataInterface() {propertyValues {Curve {keys [<at 0 0.85>,<at 1 1>]}}}
     CurveIndex=&NormalizeFloat(Value=&Divide_Float(A=$Particles.NormalizedAge, B=$Particles.Lifetime_Shockwave), Min=0, Max=1)
    )
    Index 0 Param 2=&FloatFromCurve(
     FloatCurve=DataInterface() {propertyValues {Curve {keys [
         {interpMode=RCIM_Cubic, tangentMode=RCTM_User, time=0, value=1}
         {arriveTangent=-1, interpMode=RCIM_Cubic, leaveTangent=-1, time=0.6, value=0.7}
         {interpMode=RCIM_Cubic, leaveTangent=-3.3753, time=1, value=0}
        ]}}}
     CurveIndex=&NormalizeFloat(Value=&Divide_Float(A=$Particles.NormalizedAge, B=$Particles.Lifetime_Shockwave), Min=0, Max=1)
    )
    Index 0 Param 3=1
    Index 0 Param 4=1
    Module.Param0WriteEnabled=true
    Module.Param1WriteEnabled=true
    Module.Param2WriteEnabled=true
    Module.Param3WriteEnabled=true
   )
   /SetVariables(Particles.DMP_DG_00:Vector4f=&MakeVector4(X=3, Y=100, Z=0, W=0)) {isSetParameters=true}
   &FN_Light_Attributes(
    Color=<rgba 30 8.625 1.5 1>
    Create Color Attribute=true
    Create Exponent Attribute=true
    Create Position Attribute=true
    Create Radius Attribute=true
    Exponent=2
    Position=&AddVectorToPosition(Position=$Particles.Initial.Position_BG_Offset, Vector=<0 0 100>)
    Radius=750
    Scale Color=&FloatFromCurve(
     FloatCurve=DataInterface() {propertyValues {Curve {keys [
         {tangentMode=RCTM_User, time=0, value=1}
         {arriveTangent=-0.0405, interpMode=RCIM_Cubic, leaveTangent=-0.0405, tangentMode=RCTM_User, time=0.6, value=1}
         {arriveTangent=-3.4417, interpMode=RCIM_Cubic, leaveTangent=-3.4417, tangentMode=RCTM_User, time=1, value=0}
        ]}}}
     CurveIndex=&NormalizeFloat(Value=&Divide_Float(A=$Particles.NormalizedAge, B=$Particles.Lifetime_Glow), Min=0, Max=1)
    )
   )
   &FN_Sprite_Attributes(
    Camera Offset=200
    Color=&MakeLinearColorFromVectorAndFloat(
     "Vector (RGB)"=&Multiply_Vector(A=&MakeVectorFromLinearColorRGB(Color=<rgba 10 3.0125 0.25 1>), B=<1 1 1>)
     "Float (Alpha)"=&Multiply_Float(
      A=&FloatFromCurve(
       FloatCurve=DataInterface() {propertyValues {Curve {keys [<at 0 1>,<at 0.3 0>]}}}
       CurveIndex=&NormalizeFloat(Value=&Divide_Float(A=$Particles.NormalizedAge, B=$Particles.Lifetime_Glow), Min=0, Max=1)
      )
      B=1
     )
    )
    Create Camera Offset Attribute=true
    Create Color Attribute=true
    Create SpriteSize Attribute=true
    SpriteSize=&Multiply_Vector2D_ByFloat(
     Float=&FloatFromCurve(
      FloatCurve=DataInterface() {propertyValues {Curve {keys [{interpMode=RCIM_Cubic, tangentMode=RCTM_User, time=0, value=1}, {interpMode=RCIM_Cubic, tangentMode=RCTM_User, time=0.5, value=0.5}]}}}
      CurveIndex=&NormalizeFloat(Value=&Divide_Float(A=$Particles.NormalizedAge, B=$Particles.Lifetime_Glow), Min=0, Max=1)
      Scale Curve=$System.Size
     )
     Vector2D=<500 500>
    )
    Write to Module Namespace=true
   )
  ]
  rendererClasses [
   &NiagaraSpriteRendererProperties
   &NiagaraMeshRendererProperties
   &NiagaraLightRendererProperties
  ]
  renderers [
   {rendererClass=&NiagaraSpriteRendererProperties}
   {rendererClass=&NiagaraMeshRendererProperties, rendererIndex=1}
   {rendererClass=&NiagaraMeshRendererProperties, rendererIndex=2}
   {rendererClass=&NiagaraSpriteRendererProperties, rendererIndex=3}
   {rendererClass=&NiagaraLightRendererProperties, rendererIndex=4}
   {rendererClass=&NiagaraSpriteRendererProperties, rendererIndex=5}
  ]
 }
 Debris {
  simTarget=GPUComputeSim
  emitterUpdateScript [
   &EmitterState(
    Enable Distance Culling=true
    Life Cycle Mode=::NewEnumerator1
    Loop Behavior=::NewEnumerator1
    Loop Duration=0.125
    MaxDistance=4000
    Module.bUseMaxDistance=true
    Module.bUseMinDistance=false
    Scalability Mode=::NewEnumerator1
    UseLoopDelay=false
   )
   &SpawnBurst_Instantaneous(Spawn Count=&UniformRangedInt(Minimum=50, Maximum=150), Use Spawn Probability=false)
  ]
  particleSpawnScript [
   &InitializeParticle(
    Color=<rgba 0.343 0.2773 0.2472 0.8>
    Lifetime=&UniformRangedFloat(Minimum=2, Maximum=3.5)
    Position Offset=$System.SystemOffset
    Sprite Rotation Mode=::NewEnumerator2
    Sprite Size=&Vector2DFromFloat(Value=&UniformRangedFloat(Minimum=4, Maximum=20))
    Sprite Size Mode=::NewEnumerator4
    UsePositionOffset=true
    Write Material Random=true
   ) {module=InitializeParticle001}
   /Niagara/Modules/Update/SubUV/SubUVAnimation(SubUV Animation Mode=::NewEnumerator1)
   /Niagara/Modules/Spawn/Location/SphereLocation(
    Hemisphere Z=true
    MaskSpawnGroup=false
    Module.Override Randomness=false
    Sphere Radius=&Multiply_Float(A=12, B=$System.ExplosionScale)
    Surface Only=false
   )
   &AddVelocityFromPoint(
    Module.Use Falloff Distance=false
    Origin Offset=<0 0 50>
    Velocity Strength=&Multiply_Float(A=&RandomRangeFloat(Minimum=50, Maximum=780), B=$System.ExplosionScale)
   )
   &AddVelocityInCone(
    Cone Angle=18
    Cone Axis=<0 0 1>
    Module.Use Velocity Falloff On Cone Axis=false
    Velocity Strength=&UniformRangedFloat(Minimum=125, Maximum=1250)
   )
  ]
  particleUpdateScript [
   &ParticleState()
   &AccelerationForce(Acceleration=<0 0 -2800>)
   &CurlNoiseForce(Bias Noise=false, Noise Frequency=5, Noise Strength=750, Pan Noise=true, Pan Noise Field=<0.2 0 -0.2>)
   &SolveForcesAndVelocity(Clamp Velocity=false, Limit Acceleration=false)
   &SpriteRotationRate(Rotation Rate=&AngleConversion(Angle:NiagaraFloat=0, Angle Input:ENiagara_AngleInput=::NewEnumerator0, Angle Output:ENiagara_AngleInput=::NewEnumerator1))
   &ScaleSpriteSize(Scale Factor=&Vector2DFromFloat(Value=&FloatFromCurve(FloatCurve=DataInterface() {propertyValues {Curve {keys [<at 0.6918 1>,<at 1 0>]}}})))
  ]
  rendererClasses [&NiagaraSpriteRendererProperties]
  renderers [{rendererClass=&NiagaraSpriteRendererProperties}]
 }
 DecalRenderer {
  emitterUpdateScript [
   &EmitterState()
   &SpawnBurst_Instantaneous(Use Spawn Probability=false, UseLoopCountLimit=false)
   /SetVariables(
    Emitter.OpacityMultCrack:NiagaraFloat=&RandomRangeFloat(Minimum=1, Maximum=1.5)
    Emitter.RandRotCracks:NiagaraFloat=&RandomRangeFloat(Minimum=0, Maximum=1)
   ) {isSetParameters=true}
  ]
  particleSpawnScript [&InitializeParticle(Lifetime=2, Position Offset=$System.SystemOffset, UsePositionOffset=true, Write Material Random=true)]
  particleUpdateScript [
   &ParticleState()
   /SetVariables(
    Particles.DecalFadeCrack:NiagaraFloat=&Lerp_Float(
     A=0
     B=1
     Alpha=&FloatFromCurve(FloatCurve=DataInterface() {propertyValues {Curve {keys [
          {arriveTangent=41.722, time=0, value=0}
          {interpMode=RCIM_Cubic, tangentMode=RCTM_User, time=0.02, value=1}
          {arriveTangent=-1.7378, interpMode=RCIM_Cubic, leaveTangent=-1.7378, tangentMode=RCTM_User, time=0.7, value=0.7}
          {arriveTangent=-0.5041, interpMode=RCIM_Cubic, leaveTangent=-0.5041, tangentMode=RCTM_User, time=1, value=0}
         ]}}})
    )
    Particles.DecalFadeScorch:NiagaraFloat=&Lerp_Float(
     A=0
     B=1
     Alpha=&FloatFromCurve(FloatCurve=DataInterface() {propertyValues {Curve {keys [
          {arriveTangent=41.722, time=0, value=0.25}
          {interpMode=RCIM_Cubic, tangentMode=RCTM_User, time=0.02, value=1}
          {arriveTangent=-1.795, interpMode=RCIM_Cubic, leaveTangent=-1.795, tangentMode=RCTM_User, time=0.6, value=0.6}
          {arriveTangent=-3.611, arriveTangentWeight=0.112, interpMode=RCIM_Cubic, leaveTangent=-3.611, tangentMode=RCTM_User, time=1, value=0}
         ]}}})
    )
   ) {isSetParameters=true}
   /SetVariables(
    Particles.DecalSizeCrack:Vector3f=&Multiply_VectorByFloat(Vector=<300 800 800>, Float=$User.ExplosionScale)
    Particles.DecalSizeEmissive:Vector3f=&Multiply_VectorByFloat(Vector=<300 450 450>, Float=$User.ExplosionScale)
   ) {isSetParameters=true}
  ]
  rendererClasses [&NiagaraDecalRendererProperties]
  renderers [
   {rendererClass=&NiagaraDecalRendererProperties, rendererIndex=0}
   {rendererClass=&NiagaraDecalRendererProperties, rendererIndex=1}
  ]
 }
 DropsHitGround {
  simTarget=GPUComputeSim
  emitterUpdateScript [
   &EmitterState(Life Cycle Mode=::NewEnumerator1, Loop Behavior=::NewEnumerator1, Loop Duration=1.5, UseLoopDelay=false)
   &SpawnBurst_Instantaneous(Spawn Count=100, Use Spawn Probability=false, UseLoopCountLimit=false)
  ]
  particleSpawnScript [
   &InitializeParticle(
    Lifetime Max=1.5
    Lifetime Min=0.8
    Lifetime Mode=::NewEnumerator1
    Position Offset=<20 0 0>
    Sprite Rotation Mode=::NewEnumerator2
    Sprite Size Mode=::NewEnumerator1
    Sprite UV Mode=::NewEnumerator3
    Uniform Sprite Size Max=12
    Uniform Sprite Size Min=6
    UsePositionOffset=true
    Write Material Random=true
   )
   &ShapeLocation(MaskSpawnGroup=false, Override Seed=false, Sphere Radius=40, UseOwnerScale=true)
   &AddVelocity(
    Module.Use Falloff Distance=false
    Override Seed=false
    Velocity Mode=::NewEnumerator1
    Velocity Speed=&Multiply_Float(A=&RandomRangeFloat(Minimum=0.2, Maximum=1), B=1600)
   ) {module=AddVelocity002}
  ]
  particleUpdateScript [
   &ParticleState()
   &GravityForce(Gravity=&RandomRangeVector(Minimum=<0 0 -700>, Maximum=<0 0 -1200>))
   &SpriteRotationRate(Rotation Rate=&Multiply_Float(A=&RandomRangeFloat(Minimum=-1, Maximum=1), B=800))
   &Drag(
    Drag=&RandomRangeFloat(Minimum=1, Maximum=1.5)
    Rotational Drag=7
    Use Linear Drag=true
    Use Rotational Drag=true
   )
   &SolveForcesAndVelocity(Clamp Velocity=false, Limit Acceleration=false)
   &ScaleSpriteSize(Scale Sprite Size Mode=::NewEnumerator0, Uniform Scale Factor=1.5) {module=ScaleSpriteSize001}
   &ScaleSpriteSize(Scale Sprite Size Mode=::NewEnumerator0)
   &ScaleColor(
    Linear Color Curve=DataInterface() {propertyValues {
      AlphaCurve {keys [<at 0 1>]}
      BlueCurve {keys [<at 0 0.25>,<at 0.5 0>]}
      GreenCurve {keys [<at 0 0.5625>,<at 0.5 0>]}
      RedCurve {keys [<at 0 1>,<at 0.5 0>]}
     }}
    Scale Mode=::NewEnumerator2
   )
   &DynamicMaterialParameters(
    Index 0 Param 1=&FloatFromCurve(FloatCurve=DataInterface() {propertyValues {Curve {keys [
         {interpMode=RCIM_Cubic, tangentMode=RCTM_User, time=0, value=1}
         {arriveTangent=-2.7096, interpMode=RCIM_Cubic, leaveTangent=-2.7096, tangentMode=RCTM_User, time=1, value=0}
        ]}}})
    Index 0 Param 3=1
    Module.Param0WriteEnabled=true
    Module.Param1WriteEnabled=true
    Module.Param2WriteEnabled=true
    Module.Param3WriteEnabled=true
   )
  ]
  rendererClasses [&NiagaraSpriteRendererProperties]
  renderers [{rendererClass=&NiagaraSpriteRendererProperties}]
 }
 Explode_Sparkles {
  simTarget=GPUComputeSim
  emitterUpdateScript [
   &EmitterState(Life Cycle Mode=::NewEnumerator1, Loop Behavior=::NewEnumerator1, Loop Delay=0.1, Loop Duration=1.5, UseLoopDelay=false)
   &SpawnBurst_Instantaneous(Spawn Count=80, Use Spawn Probability=false, UseLoopCountLimit=false)
  ]
  particleSpawnScript [
   &InitializeParticle(
    Color=<rgba 1 0.3667 0 1>
    Lifetime=&RandomRangeFloat(Minimum=0.5, Maximum=1.5)
    Sprite Rotation Mode=::NewEnumerator0
    Sprite Size Max=<1.5 10>
    Sprite Size Min=<1.5 5>
    Sprite Size Mode=::NewEnumerator2
    UsePositionOffset=false
    Write Material Random=true
   )
   &ShapeLocation(MaskSpawnGroup=false, Override Seed=false, Rotation Mode=::NewEnumerator1, Sphere Radius=200, UseOwnerScale=true)
   &AddVelocity(
    Module.Use Falloff Distance=false
    Origin Offset=&RandomRangeVector(Minimum=<0 0 -100>, Maximum=<0 0 -40>)
    Override Seed=false
    Velocity Mode=::NewEnumerator1
    Velocity Speed=&RandomRangeFloat(Minimum=600, Maximum=1000)
   )
   &AddVelocity(
    Rotation Coordinate Space=::Simulation
    Velocity=<0 0 500>
    Velocity Speed Scale=&RandomRangeFloat(Minimum=0, Maximum=0.5)
   ) {module=AddVelocity001}
  ]
  particleUpdateScript [
   /SetVariables(Particles.Glow03_Color=&MakeLinearColorFromVectorAndFloat(
     "Vector (RGB)"=&MakeVectorFromLinearColorRGB(Color=&MultiplyLinearColor(Linear Color=<rgba 1 0.1667 0 0.2>, Scale Factor=&MakeVector4_FromFloat(Float=1)))
     "Float (Alpha)"=&Multiply_Float(
      A=0.4
      B=&FloatFromCurve(FloatCurve=DataInterface() {propertyValues {Curve {keys [
           {tangentMode=RCTM_User, time=0, value=1}
           {arriveTangent=-0.0405, interpMode=RCIM_Cubic, leaveTangent=-0.0405, tangentMode=RCTM_User, time=0.7009, value=1}
           {arriveTangent=-8.0089, interpMode=RCIM_Cubic, leaveTangent=-8.0089, tangentMode=RCTM_User, time=1, value=0}
          ]}}})
     )
    )) {isSetParameters=true}
   /SetVariables(Particles.Glow03_Size=&Multiply_Vector2D_ByFloat(
     Float=&FloatFromCurve(
      FloatCurve=DataInterface() {propertyValues {Curve {keys [
          <at 0 0 RCIM_Cubic>
          <at 0.2 1.3 RCIM_Cubic>
          {arriveTangent=-0.1423, interpMode=RCIM_Cubic, leaveTangent=-0.1423, tangentMode=RCTM_User, time=1, value=0}
         ]}}}
      Scale Curve=&Multiply_Float(A=10, B=2)
     )
     Vector2D=<1 2>
    )) {isSetParameters=true}
   &ParticleState()
   &CurlNoiseForce(Noise Strength=&Multiply_Float(A=500, B=1), Pan Noise=true, Pan Noise Field=<0 0 1>)
   &ScaleSpriteSize(Scale Factor=&Vector2DFromFloat(Value=&FloatFromCurve(
      FloatCurve=DataInterface() {propertyValues {Curve {keys [
          <at 0 0 RCIM_Cubic>
          {interpMode=RCIM_Cubic, tangentMode=RCTM_User, time=0.2, value=1.2}
          {interpMode=RCIM_Cubic, leaveTangent=-2.7132, time=1, value=0}
         ]}}}
      Scale Curve=&Multiply_Float(A=0.7, B=2)
     )))
   &ScaleColor(
    Scale Alpha=&Multiply_Float(A=1, B=1)
    Scale RGB=&Multiply_VectorByFloat(Vector=<2 2 2>, Float=1)
    ScaleA=true
    ScaleRGB=true
   )
   &Drag(Drag=&RandomRangeFloat(Minimum=1, Maximum=3), Use Linear Drag=true, Use Rotational Drag=true)
   &GravityForce(Gravity=&RandomRangeVector(Minimum=<0 0 -300>, Maximum=<0 0 -500>))
   &SolveForcesAndVelocity(Clamp Velocity=false, Limit Acceleration=false)
   &ScaleSpriteSizeBySpeed(Max Scale Factor=<1 3>, Min Scale Factor=<1 0.5>)
  ]
  rendererClasses [&NiagaraSpriteRendererProperties]
  renderers [
   {rendererClass=&NiagaraSpriteRendererProperties}
   {rendererClass=&NiagaraSpriteRendererProperties, rendererIndex=1}
  ]
 }
 Ground_energy {
  emitterUpdateScript [
   &EmitterState(Life Cycle Mode=::NewEnumerator1, Loop Behavior=::NewEnumerator1, Loop Duration=0.65, UseLoopDelay=false)
   &SpawnBurst_Instantaneous(Use Spawn Probability=false, UseLoopCountLimit=false)
  ]
  particleSpawnScript [
   &InitializeParticle(
    Color=<rgba 2 2 2 1>
    Lifetime=0.65
    Mesh Scale=<3 3 1.8>
    Mesh Scale Mode=::NewEnumerator4
    UsePositionOffset=false
    Write Material Random=true
   )
   &CameraOffset(Camera Offset Amount=20)
   &InitialMeshOrientation(Mesh Orientation Mode=::NewEnumerator2, Use Rotation Vector=true)
  ]
  particleUpdateScript [
   &ParticleState()
   &ScaleMeshSize(Scale Factor=&Multiply_VectorByFloat(
     Vector=<0.8 0.8 0.8>
     Float=&FloatFromCurve(
      FloatCurve=DataInterface() {propertyValues {Curve {keys [
          {arriveTangent=4.1785, interpMode=RCIM_Cubic, leaveTangent=4.1785, tangentMode=RCTM_User, time=0, value=0}
          {arriveTangent=1.1179, interpMode=RCIM_Cubic, leaveTangent=1.1179, tangentMode=RCTM_User, time=0.2, value=0.9}
          {arriveTangent=0.0224, interpMode=RCIM_Cubic, leaveTangent=0.0224, tangentMode=RCTM_User, time=0.8, value=1.5}
         ]}}}
      Scale Curve=&Multiply_Float(A=2, B=$System.Size)
     )
    ))
   &ScaleColor(
    Scale Alpha=&Multiply_Float(
     A=1
     B=&FloatFromCurve(FloatCurve=DataInterface() {propertyValues {Curve {keys [
          {arriveTangent=-0.8483, interpMode=RCIM_Cubic, leaveTangent=-0.8483, tangentMode=RCTM_User, time=0, value=1}
          {arriveTangent=-0.9759, interpMode=RCIM_Cubic, leaveTangent=-0.9759, tangentMode=RCTM_User, time=0.5, value=0.55}
          {arriveTangent=-1.3542, interpMode=RCIM_Cubic, leaveTangent=-1.3542, tangentMode=RCTM_User, time=0.8, value=0}
         ]}}})
    )
    Scale RGB=&MakeVectorFromLinearColorRGB(Color=&ColorFromCurve(ColorCurve=DataInterface() {propertyValues {
        AlphaCurve {keys []}
        BlueCurve {keys [
          <at 0.0326 0.5734>
          <at 0.0538 0.4592>
          {arriveTangent=-0.5927, interpMode=RCIM_Cubic, leaveTangent=-0.5927, time=0.3137, value=0}
          <at 1 0 RCIM_Cubic>
         ]}
        GreenCurve {keys [
          <at 0.0326 1.2584>
          <at 0.0538 2.1601>
          {arriveTangent=-1.3008, interpMode=RCIM_Cubic, leaveTangent=-1.3008, time=0.3137, value=0.6694}
          <at 1 0 RCIM_Cubic>
         ]}
        RedCurve {keys [
          <at 0.0326 2>
          <at 0.0538 10>
          {arriveTangent=-2.0674, interpMode=RCIM_Cubic, leaveTangent=-2.0674, time=0.3137, value=2}
          <at 1 0 RCIM_Cubic>
         ]}
       }}))
    ScaleA=true
    ScaleRGB=true
   )
   &FN_Mesh_Attributes(
    Color=&MakeLinearColorFromVectorAndFloat(
     "Vector (RGB)"=&Multiply_VectorByFloat(
      Vector=&MakeVectorFromLinearColorRGB(Color=&MultiplyLinearColor(
        Linear Color=<rgba 2 2 2 1>
        Scale Factor=&MakeVector4_FromLinearColor(LinearColor=&ColorFromCurve(ColorCurve=DataInterface() {propertyValues {
            AlphaCurve {keys []}
            BlueCurve {keys [<at 0.3124 0>,<at 0.4755 0.0459>,<at 0.7427 0 RCIM_Cubic>]}
            GreenCurve {keys [<at 0.3124 0>,<at 0.4755 0.216>,<at 0.7427 0 RCIM_Cubic>]}
            RedCurve {keys [<at 0.3124 0>,<at 0.4755 1>,<at 0.7427 0 RCIM_Cubic>]}
           }}))
       ))
      Float=1
     )
     "Float (Alpha)"=&Multiply_Float(
      A=&FloatFromCurve(FloatCurve=DataInterface() {propertyValues {Curve {keys [
           {arriveTangent=-2.5289, interpMode=RCIM_Cubic, leaveTangent=-2.5289, tangentMode=RCTM_User, time=0, value=1}
           {arriveTangent=-0.9759, interpMode=RCIM_Cubic, leaveTangent=-0.9759, tangentMode=RCTM_User, time=0.5, value=0.55}
           {arriveTangent=-1.3542, interpMode=RCIM_Cubic, leaveTangent=-1.3542, tangentMode=RCTM_User, time=1, value=0}
          ]}}})
      B=20
     )
    )
    Create Color Attribute=true
    Create Scale Attribute=true
    Scale=&Multiply_VectorByFloat(
     Vector=&Multiply_Vector(A=<0.8 0.8 0.8>, B=<3 3 2>)
     Float=&FloatFromCurve(
      FloatCurve=DataInterface() {propertyValues {Curve {keys [
          {arriveTangent=4.1785, interpMode=RCIM_Cubic, leaveTangent=4.1785, tangentMode=RCTM_User, time=0, value=0}
          {arriveTangent=1.1179, interpMode=RCIM_Cubic, leaveTangent=1.1179, tangentMode=RCTM_User, time=0.2, value=0.9}
          {arriveTangent=0.0224, interpMode=RCIM_Cubic, leaveTangent=0.0224, tangentMode=RCTM_User, time=1, value=1.5}
         ]}}}
      Scale Curve=&Multiply_Float(A=1.5, B=$System.Size)
     )
    )
    Write to Module Namespace=true
   )
  ]
  rendererClasses [&NiagaraMeshRendererProperties]
  renderers [
   {rendererClass=&NiagaraMeshRendererProperties, rendererIndex=0}
   {rendererClass=&NiagaraMeshRendererProperties, rendererIndex=1}
  ]
 }
 Rocks {
  emitterUpdateScript [
   &EmitterState(
    Enable Distance Culling=true
    Life Cycle Mode=::NewEnumerator1
    Loop Behavior=::NewEnumerator1
    Loop Duration=0.4
    Module.bUseMaxDistance=true
    Module.bUseMinDistance=false
    Scalability Mode=::NewEnumerator1
    UseLoopDelay=false
   )
   &SpawnBurst_Instantaneous(Spawn Count=15, Use Spawn Probability=false)
  ]
  particleSpawnScript [
   &InitializeParticle(
    Lifetime=&UniformRangedFloat(Minimum=0.8, Maximum=1.4)
    Mass Mode=::NewEnumerator1
    Mesh Scale=&VectorFromFloat(Value=&UniformRangedFloat(Minimum=0.2, Maximum=1.1))
    Mesh Scale Mode=::NewEnumerator4
    Position Offset=$System.SystemOffset
    Sprite Size Mode=::NewEnumerator4
    UsePositionOffset=true
    Write Material Random=true
   ) {module=InitializeParticle001}
   &InitialMeshOrientation(Use Rotation Vector=true) {module=InitialMeshOrientation001}
   &AddVelocity(Velocity=&Multiply_VectorByFloat(
     Vector=&UniformRangedVector(Minimum=<-600 -600 200>, Maximum=<600 600 1200>)
     Float=$System.ExplosionScale
    ))
  ]
  particleUpdateScript [
   &ParticleState()
   &AccelerationForce(Acceleration=<0 0 -2200>)
   &Drag(Drag=&RandomRangeFloat(Minimum=0.75, Maximum=2), Use Linear Drag=true, Use Rotational Drag=true)
   &SolveForcesAndVelocity(Clamp Velocity=false, Limit Acceleration=false)
   &ScaleMeshSize(Scale Factor=&VectorFromFloat(Value=&FloatFromCurve(FloatCurve=DataInterface() {propertyValues {Curve {keys [<at 0.6 1 RCIM_Cubic>,<at 1 0 RCIM_Cubic>]}}})))
   &ScaleColor(
    Scale Alpha=&Lerp_Float(A=0, B=1, Alpha=$Particles.NormalizedAge)
    Scale RGB=<0.5 0.5 0.5>
    ScaleA=true
    ScaleRGB=true
   )
   &UpdateMeshOrientation(
    Rotation Rate=&Lerp_Float(A=-3, B=3, Alpha=$Particles.Initial.MaterialRandom)
    Rotation Vector=/Niagara/DynamicInputs/Math/SlerpVector(Vector A:Vector3f=<-1 0 1>, Vector B:Vector3f=<0 1 -1>, Alpha:NiagaraFloat=$Particles.Initial.MaterialRandom, Scale:NiagaraFloat=1)
   ) {module=UpdateMeshOrientation001}
  ]
  rendererClasses [&NiagaraMeshRendererProperties]
  renderers [{rendererClass=&NiagaraMeshRendererProperties, rendererIndex=0}]
 }
 Smoke {
  emitterUpdateScript [
   &EmitterState(Life Cycle Mode=::NewEnumerator1, Loop Behavior=::NewEnumerator1, Loop Duration=1, UseLoopDelay=false)
   &SpawnBurst_Instantaneous(Spawn Count=&UniformRangedInt(Minimum=20, Maximum=40), Spawn Time=0.1, Use Spawn Probability=false)
  ]
  particleSpawnScript [
   /SetVariables(Particles.VelCurveRand:NiagaraFloat=&FloatFromCurve(
     FloatCurve=DataInterface() {propertyValues {Curve {keys [
         {interpMode=RCIM_Cubic, leaveTangent=0.1132, tangentMode=RCTM_Break, time=0, value=0}
         {arriveTangent=1.9858, interpMode=RCIM_Cubic, leaveTangent=1.7203, tangentMode=RCTM_Break, time=0.3, value=0.4}
         {interpMode=RCIM_Cubic, tangentMode=RCTM_Break, time=1, value=1}
        ]}}}
     CurveIndex=&RandomRangeFloat(Minimum=0, Maximum=1)
     Scale Curve=5
    )) {isSetParameters=true}
   /SetVariables(Particles.InitSpriteRotRate=&RandomRangeFloat(Minimum=-1, Maximum=1)) {isSetParameters=true}
   &InitializeParticle(
    Color Maximum=<rgba 0.1 0.1 0.1 1.25>
    Color Minimum=<rgba 0.15 0.1361 0.1322 0.75>
    Color Mode=::NewEnumerator2
    Lifetime Max=2.5
    Lifetime Min=2
    Lifetime Mode=::NewEnumerator1
    Sprite Rotation Angle=&RandomRangeFloat(Minimum=0, Maximum=360)
    Sprite Rotation Mode=::NewEnumerator0
    Sprite Size Mode=::NewEnumerator1
    Uniform Sprite Size Max=300
    Uniform Sprite Size Min=250
    UsePositionOffset=false
    Write Material Random=true
   )
   /SetVariables(Particles.SpriteSize=&Multiply_Vector2D_ByFloat(Float=$System.ScaleFX, Vector2D=$Particles.SpriteSize)) {isSetParameters=true}
   &CameraOffset(Camera Offset Amount=&UniformRangedFloat(Minimum=25, Maximum=50))
   &ShapeLocation(
    Handle Radius=15
    Large Radius=50
    MaskSpawnGroup=false
    Non Uniform Scale=&Multiply_VectorByFloat(Vector=<1 1 1>, Float=$System.ScaleFX)
    Override Seed=false
    Shape Primitive=::NewEnumerator3
    UseOwnerScale=true
   )
   &AddVelocityFromPoint(Module.Use Falloff Distance=false, Velocity Strength=&Lerp_Float(A=1250, B=1300, Alpha=$Particles.VelCurveRand))
  ]
  particleUpdateScript [
   &ParticleState()
   &Drag(Drag=8, Use Linear Drag=true, Use Rotational Drag=false)
   &SpriteRotationRate(Rotation Rate=&Multiply_Float(
     A=$Particles.InitSpriteRotRate
     B=&FloatFromCurve(
      FloatCurve=DataInterface() {propertyValues {Curve {keys [
          {arriveTangent=-4.4624, interpMode=RCIM_Cubic, leaveTangent=-4.4624, tangentMode=RCTM_User, time=0, value=1}
          {arriveTangentWeight=0.1718, interpMode=RCIM_Cubic, tangentMode=RCTM_User, time=0.5, value=0.15}
         ]}}}
      Scale Curve=50
     )
    ))
   &ScaleColor(
    Scale Alpha=&FloatFromCurve(
     FloatCurve=DataInterface() {propertyValues {Curve {
        defaultValue=1
        keys [{interpMode=RCIM_Cubic, tangentMode=RCTM_User, time=0, value=1}, {interpMode=RCIM_Cubic, tangentMode=RCTM_User, time=1, value=0}]
       }}}
     Scale Curve=&RandomRangeFloat(Minimum=0.5, Maximum=1)
    )
    Scale RGB=<0.6 0.6 0.6>
    ScaleA=true
    ScaleRGB=true
   )
   &ScaleSpriteSize(Scale Factor=&Vector2DFromFloat(Value=&FloatFromCurve(
      FloatCurve=DataInterface() {propertyValues {Curve {keys [
          {arriveTangent=0.0418, interpMode=RCIM_Cubic, leaveTangent=0.0418, tangentMode=RCTM_User, time=0, value=0.5}
          {arriveTangent=0.8771, interpMode=RCIM_Cubic, leaveTangent=0.8771, tangentMode=RCTM_User, time=0.7, value=0.775}
          {arriveTangent=0.7996, arriveTangentWeight=0.3375, interpMode=RCIM_Cubic, leaveTangent=0.7996, tangentMode=RCTM_User, time=1, value=1}
         ]}}}
      Scale Curve=0.9
     )))
   &DynamicMaterialParameters(
    Index 0 Param 1=&Add_Float(
     A=&RandomRangeFloat(Minimum=0, Maximum=1)
     B=&FloatFromCurve(
      FloatCurve=DataInterface() {propertyValues {Curve {keys [
          {arriveTangent=2.7533, interpMode=RCIM_Cubic, leaveTangent=1.8958, tangentMode=RCTM_User, time=0, value=1}
          {arriveTangent=0.8292, interpMode=RCIM_Cubic, leaveTangent=0.8292, tangentMode=RCTM_User, time=0.321, value=1.3453}
          {arriveTangent=0.0529, interpMode=RCIM_Cubic, leaveTangent=0.2021, tangentMode=RCTM_User, time=1, value=1.5}
         ]}}}
      CurveIndex=&Add_Float(A=&Multiply_Float(A=$Particles.Initial.MaterialRandom, B=0.5), B=$Particles.NormalizedAge)
     )
    )
    Index 0 Param 2=&FloatFromCurve(FloatCurve=DataInterface() {propertyValues {Curve {keys [
         {arriveTangent=0.0453, interpMode=RCIM_Cubic, tangentMode=RCTM_Break, time=0, value=-1}
         {arriveTangent=0.0604, interpMode=RCIM_Cubic, leaveTangent=0.0604, tangentMode=RCTM_User, time=1, value=-1}
        ]}}})
    Index 0 Param 3=&FloatFromCurve(
     FloatCurve=DataInterface() {propertyValues {Curve {keys [
         {arriveTangent=0.0632, interpMode=RCIM_Cubic, leaveTangent=0.0632, tangentMode=RCTM_User, time=0, value=1}
         {
          arriveTangent=0.2851
          arriveTangentWeight=0.0333
          interpMode=RCIM_Cubic
          leaveTangent=0.2851
          leaveTangentWeight=0.3137
          tangentMode=RCTM_User
          time=0.025
          value=0.05
         }
         {arriveTangentWeight=0.1027, interpMode=RCIM_Cubic, tangentMode=RCTM_Break, time=0.2, value=0}
        ]}}}
     Scale Curve=&RandomRangeFloat(Minimum=0.25, Maximum=0.5)
    )
    Index 0 Param 4=&FloatFromCurve(FloatCurve=DataInterface() {propertyValues {Curve {keys [<at 0.5 1>,<at 1 0>]}}})
    Index 1 Param 1=&FloatFromCurve(
     FloatCurve=DataInterface() {propertyValues {Curve {keys [
         {interpMode=RCIM_Cubic, tangentMode=RCTM_Break, time=0, value=1}
         {arriveTangentWeight=0.0833, interpMode=RCIM_Cubic, tangentMode=RCTM_Break, time=0.25, value=0.3}
        ]}}}
     Scale Curve=&RandomRangeFloat(Minimum=1.5, Maximum=3)
    )
    Index 1 Param 2=&FloatFromCurve(
     FloatCurve=DataInterface() {propertyValues {Curve {keys [
         {interpMode=RCIM_Cubic, leaveTangent=-2.5489, tangentMode=RCTM_Break, time=0, value=1}
         {interpMode=RCIM_Cubic, leaveTangent=0.9066, tangentMode=RCTM_Break, time=1, value=0}
        ]}}}
     Scale Curve=&RandomRangeFloat(Minimum=0.01, Maximum=0.1)
    )
    Index 1 Param 3=&FloatFromCurve(FloatCurve=DataInterface() {propertyValues {Curve {keys [
         {interpMode=RCIM_Cubic, leaveTangent=6.3917, tangentMode=RCTM_Break, time=0, value=0.2}
         {arriveTangent=1.7549, interpMode=RCIM_Cubic, leaveTangent=10.6767, tangentMode=RCTM_User, time=0.25, value=1}
        ]}}})
    Index 1 Param 4=&FloatFromCurve(
     FloatCurve=DataInterface() {propertyValues {Curve {keys [{arriveTangent=-0.136, interpMode=RCIM_Cubic, leaveTangent=-0.136, tangentMode=RCTM_User, time=0, value=1}, <at 1 0.95>]}}}
     Scale Curve=&RandomRangeFloat(Minimum=1.3, Maximum=2)
    )
    Index 2 Param 1=&FloatFromCurve(FloatCurve=DataInterface() {propertyValues {Curve {keys [
         {interpMode=RCIM_Cubic, leaveTangent=4.2485, tangentMode=RCTM_Break, time=0, value=0}
         {arriveTangent=1.0548, interpMode=RCIM_Cubic, leaveTangent=1.0548, tangentMode=RCTM_User, time=0.17, value=0.5715}
         <at 1 1>
        ]}}})
    Index 2 Param 2=&FloatFromCurve(
     FloatCurve=DataInterface() {propertyValues {Curve {keys [
         {arriveTangent=1.9056, interpMode=RCIM_Cubic, leaveTangent=1.9056, tangentMode=RCTM_User, time=0, value=0.2}
         {arriveTangent=0.0259, interpMode=RCIM_Cubic, leaveTangent=0.0259, tangentMode=RCTM_User, time=1, value=1.4}
        ]}}}
     Scale Curve=&RandomRangeFloat(Minimum=0.1, Maximum=0.2)
    )
    Index 2 Param 3=&FloatFromCurve(FloatCurve=DataInterface() {propertyValues {Curve {keys [
         {interpMode=RCIM_Cubic, leaveTangent=2.3064, tangentMode=RCTM_Break, time=0, value=1}
         {interpMode=RCIM_Cubic, leaveTangent=11.1233, tangentMode=RCTM_Break, time=1, value=2}
        ]}}})
    Index 2 Param 4=&FloatFromCurve(
     FloatCurve=DataInterface() {propertyValues {Curve {keys [
         {arriveTangent=-0.6867, interpMode=RCIM_Cubic, leaveTangent=-0.6867, tangentMode=RCTM_User, time=0, value=1}
         {interpMode=RCIM_Cubic, leaveTangent=-1.2956, tangentMode=RCTM_Break, time=1, value=0.75}
        ]}}}
     Scale Curve=1.5
    )
    Module.Param0WriteEnabled=true
    Module.Param0WriteEnabled1=true
    Module.Param0WriteEnabled2=true
    Module.Param1WriteEnabled=false
    Module.Param1WriteEnabled1=true
    Module.Param1WriteEnabled2=true
    Module.Param2WriteEnabled=true
    Module.Param2WriteEnabled1=true
    Module.Param2WriteEnabled2=true
    Module.Param3WriteEnabled=true
    Module.Param3WriteEnabled1=true
    Module.Param3WriteEnabled2=true
    Write Parameter Index 1=true
    Write Parameter Index 2=true
   )
   &GravityForce(Gravity=&RandomRangeVector(Minimum=<0 0 666>, Maximum=<0 0 222>))
   &SolveForcesAndVelocity(Clamp Velocity=false, Limit Acceleration=false)
  ]
  rendererClasses [&NiagaraSpriteRendererProperties]
  renderers [
   {rendererClass=&NiagaraSpriteRendererProperties}
   {rendererClass=&NiagaraSpriteRendererProperties, rendererIndex=1}
  ]
 }
 spikes001 {
  emitterUpdateScript [
   &EmitterState(Life Cycle Mode=::NewEnumerator1, Loop Behavior=::NewEnumerator1, Loop Duration=0.4, UseLoopDelay=false)
   &SpawnBurst_Instantaneous(Spawn Count=9, Use Spawn Probability=false, UseLoopCountLimit=false)
  ]
  particleSpawnScript [
   &InitializeParticle(
    Color=<rgba 8 1.4 0.08 0.5>
    Lifetime Max=0.4
    Lifetime Min=0.3
    Lifetime Mode=::NewEnumerator1
    Sprite Size Max=<100 90>
    Sprite Size Min=<50 110>
    Sprite Size Mode=::NewEnumerator2
    UsePositionOffset=false
    Write Material Random=true
   )
   /Niagara/Modules/Update/SubUV/V2/SubUVAnimation(End Frame=0, SubUV Animation Mode:ENiagara_SubUVLookupMode=::NewEnumerator2)
   &ShapeLocation(
    Hemisphere Distribution=<1 0.4>
    MaskSpawnGroup=false
    Offset=<0 0 -75>
    Offset Mode=::NewEnumerator0
    Override Seed=false
    Sphere Radius=50
    Sphere Surface Distribution=1
    UseOwnerScale=true
   )
   &AddVelocity(
    Velocity=$Particles.ShapeLocation.ShapeNormal
    Velocity Speed Scale=&RandomRangeFloat(Minimum=400, Maximum=800)
   )
  ]
  particleUpdateScript [
   &ParticleState()
   &Drag(Drag=5, Use Linear Drag=true, Use Rotational Drag=false)
   &SolveForcesAndVelocity(Clamp Velocity=false, Limit Acceleration=false)
   &ScaleSpriteSize(
    Non-Uniform Curve Scale=&Multiply_Vector2D_ByFloat(Float=1, Vector2D=<0.9221 1.3>)
    Non-Uniform Curve Sprite Scale=DataInterface() {propertyValues {
      XCurve {keys [{tangentMode=RCTM_User, time=0, value=1}, {tangentMode=RCTM_User, time=1, value=0.8}]}
      YCurve {keys [{tangentMode=RCTM_User, time=0, value=0.5}, {tangentMode=RCTM_User, time=1, value=1.7}]}
     }}
    Scale Sprite Size Mode=::NewEnumerator3
   )
   &ScaleColor(
    Scale Alpha=&FloatFromCurve(FloatCurve=DataInterface() {propertyValues {Curve {keys [<at 0 0>,<at 0.5 1>,<at 1 0>]}}}, Scale Curve=0.2)
    Scale RGB=&MakeVectorFromLinearColorRGB(Color=&ColorFromCurve(ColorCurve=DataInterface() {propertyValues {
        AlphaCurve {keys [<at 0 1 RCIM_Cubic>,<at 1 0 RCIM_Cubic>]}
        BlueCurve {keys [<at 0 25 RCIM_Cubic>,<at 0.7015 0.5 RCIM_Cubic>]}
        GreenCurve {keys [<at 0 25 RCIM_Cubic>,<at 0.7015 0.5 RCIM_Cubic>]}
        RedCurve {keys [<at 0 25 RCIM_Cubic>,<at 0.7015 0.5 RCIM_Cubic>]}
       }}))
    ScaleA=true
    ScaleRGB=true
   )
   /SetVariables(Particles.Color dark:LinearColor=&MakeLinearColorFromVectorAndFloat(
     "Vector (RGB)"=&MakeVectorFromLinearColorRGB(Color=&MultiplyLinearColor(Linear Color=<rgba 0.0424 0.0047 0 10>, Scale Factor=&MakeVector4_FromFloat(Float=1)))
     "Float (Alpha)"=&FloatFromCurve(FloatCurve=DataInterface() {propertyValues {Curve {keys [<at 0 0>,<at 0.5 1>,<at 1 0>]}}}, Scale Curve=1.5)
    )) {isSetParameters=true}
   /SetVariables(Particles.Size02:Vector2f=&Multiply_Vector2D(
     A=&RandomRangeVector2D(Minimum=<25 90>, Maximum=<50 110>)
     B=&Vector2DFromCurve(
      Vector2Curve=DataInterface() {propertyValues {
        XCurve {keys [{tangentMode=RCTM_User, time=0, value=1}, {tangentMode=RCTM_User, time=1, value=0.8}]}
        YCurve {keys [{tangentMode=RCTM_User, time=0, value=0.5}, {tangentMode=RCTM_User, time=1, value=1.7}]}
       }}
      ScaleCurve=<3 3>
     )
    )) {isSetParameters=true}
  ]
  rendererClasses [&NiagaraSpriteRendererProperties]
  renderers [
   {rendererClass=&NiagaraSpriteRendererProperties}
   {rendererClass=&NiagaraSpriteRendererProperties, rendererIndex=1}
  ]
 }
}
summary {
 systemName=NS_TidaleCraneDrop_Impact
 emitters [
  {emitterName=Debris, simTarget=GPUComputeSim, rendererClasses [&NiagaraSpriteRendererProperties]}
  {emitterName=Rocks, rendererClasses [&NiagaraMeshRendererProperties]}
  {emitterName=DecalRenderer, rendererClasses [&NiagaraDecalRendererProperties]}
  {emitterName=Smoke, rendererClasses [&NiagaraSpriteRendererProperties]}
  {
   emitterName=Combined_Emitter_Sprite001
   rendererClasses [
    &NiagaraSpriteRendererProperties
    &NiagaraMeshRendererProperties
    &NiagaraLightRendererProperties
   ]
  }
  {emitterName=Combined_Emitter_Mesh, rendererClasses [&NiagaraMeshRendererProperties]}
  {emitterName=spikes001, rendererClasses [&NiagaraSpriteRendererProperties]}
  {emitterName=DropsHitGround, simTarget=GPUComputeSim, rendererClasses [&NiagaraSpriteRendererProperties]}
  {emitterName=Explode_Sparkles, simTarget=GPUComputeSim, rendererClasses [&NiagaraSpriteRendererProperties]}
  {emitterName=Ground_energy, rendererClasses [&NiagaraMeshRendererProperties]}
 ]
 userVariables [
  User.ExplosionScale {type=NiagaraFloat, defaultValue=2, description=""}
  User.FX_Scale_Global {type=NiagaraFloat, defaultValue=0.35, description=""}
  User.FX_Scale_Impact_Center {type=NiagaraFloat, defaultValue=1.75, description=""}
 ]
}
