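# Niagara effect for the spawn of a Tidale crane drop, judging by its name: a large, short-lived burst shockwave.
# A single Shockwave emitter spawns by instantaneous burst and feeds three mesh renderers and one sprite renderer from the same particle data.
# Age-driven curves grow mesh and sprite size, fade color, and animate dynamic material parameters.
# Custom per-layer attributes separately drive cone wind, radial shockwave and refraction shockwave, several on rescaled age timelines.
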NiagaraSystem /DropPodParentGameplay/Assets/FX/NS_TidaleCraneDrop_Spawn
dependencies {
 usedDataInterfaces [&NiagaraDataInterfaceColorCurve, &NiagaraDataInterfaceCurve]
 usedDynamicInputs []
 usedModules [&SystemState]
}
emitters {Shockwave {
  emitterUpdateScript [
   &EmitterState()
   &SpawnBurst_Instantaneous(Use Spawn Probability=false, UseLoopCountLimit=false)
  ]
  particleSpawnScript [
   &InitializeParticle(
    Color=<rgba 1 1 1 0.5>
    Lifetime=3.5
    Mesh Scale Mode=::NewEnumerator3
    Mesh Uniform Scale=77
    Sprite Size Mode=::NewEnumerator3
    Uniform Sprite Size=4500
    UsePositionOffset=false
    Write Material Random=true
   )
   &InitialMeshOrientation(Mesh Orientation Mode=::NewEnumerator2, Use Rotation Vector=false)
   &CameraOffset()
   /SetVariables(Particles.SpriteFacing=<-1 0 0>) {isSetParameters=true}
  ]
  particleUpdateScript [
   &ParticleState()
   &ScaleColor(
    Linear Color Curve=DataInterface() {propertyValues {
      AlphaCurve {keys [<at 0 1>,<at 0.0119 1>,<at 0.2527 0.5>,<at 0.4968 0>]}
      BlueCurve {keys [<at 0 1>]}
      GreenCurve {keys [<at 0 0.8526>]}
      RedCurve {keys [<at 0 0.6719>]}
     }}
    Scale Mode=::NewEnumerator2
   )
   &ScaleMeshSize(Scale Factor=&VectorFromFloat(Value=&FloatFromCurve(FloatCurve=DataInterface() {propertyValues {Curve {keys [
          {arriveTangent=4.1785, interpMode=RCIM_Cubic, leaveTangent=4.1785, tangentMode=RCTM_User, time=0, value=0}
          {arriveTangent=0.9995, interpMode=RCIM_Cubic, leaveTangent=0.9995, tangentMode=RCTM_User, time=0.2223, value=0.6825}
          {arriveTangent=-0.0405, interpMode=RCIM_Cubic, leaveTangent=-0.0405, tangentMode=RCTM_User, time=1, value=1}
         ]}}})))
   &ScaleSpriteSize(Scale Factor=&Vector2DFromFloat(Value=&FloatFromCurve(
      FloatCurve=DataInterface() {propertyValues {Curve {keys [
          {arriveTangent=3.2319, interpMode=RCIM_Cubic, leaveTangent=3.2319, tangentMode=RCTM_User, time=0, value=0}
          {arriveTangent=12.6155, interpMode=RCIM_Cubic, leaveTangent=12.6155, tangentMode=RCTM_User, time=0.2125, value=1.7818}
          <at 1 4.2>
         ]}}}
      CurveIndex=&NormalizeFloat(Value=&Divide_Float(A=$Particles.NormalizedAge, B=0.9), Min=0, Max=1)
     )))
   &DynamicMaterialParameters(
    Index 0 Param 1=&FloatFromCurve(
     FloatCurve=DataInterface() {propertyValues {Curve {keys [{arriveTangent=3.726, interpMode=RCIM_Cubic, leaveTangent=3.726, tangentMode=RCTM_User, time=0, value=-0.3}, <at 1 1>]}}}
     Scale Curve=0.8
    )
    Index 0 Param 2=&FloatFromCurve(
     FloatCurve=DataInterface() {propertyValues {Curve {keys [{arriveTangent=-0.5818, interpMode=RCIM_Cubic, leaveTangent=-0.5818, tangentMode=RCTM_User, time=0, value=1}, <at 1 0.8>]}}}
     Scale Curve=0.8
    )
    Index 0 Param 3=&FloatFromCurve(FloatCurve=DataInterface() {propertyValues {Curve {keys [<at 0 0>,<at 0.15 0.1>,<at 1 0.4>]}}})
    Module.Param0WriteEnabled=true
    Module.Param1WriteEnabled=true
    Module.Param2WriteEnabled=true
    Module.Param3WriteEnabled=false
   )
   /SetVariables(
    Particles.Color_ConeWind:LinearColor=&MultiplyLinearColor(
     Linear Color=<rgba 0.869 0.9597 1 0.75>
     Scale Factor=&MakeVector4FromVectorAndFloat(
      "Vector (RGB)"=<1 1 1>
      "Float (Alpha)"=&FloatFromCurve(
       FloatCurve=DataInterface() {propertyValues {Curve {keys [
           {arriveTangent=0.0662, interpMode=RCIM_Cubic, leaveTangent=0.0662, tangentMode=RCTM_User, time=0, value=0}
           {arriveTangent=2.5, interpMode=RCIM_Cubic, leaveTangent=2.5, time=0.0403, value=2.8151}
           {arriveTangent=-4.1734, interpMode=RCIM_Cubic, leaveTangent=-4.1734, time=0.2, value=0.5}
           {arriveTangent=-0.0405, interpMode=RCIM_Cubic, leaveTangent=-0.0405, tangentMode=RCTM_User, time=0.7148, value=0}
           {arriveTangent=-0.2378, interpMode=RCIM_Cubic, leaveTangent=-0.2378, tangentMode=RCTM_User, time=1, value=0}
          ]}}}
       CurveIndex=&NormalizeFloat(
        Value=&Subtract_Float(A=&Divide_Float(A=$Particles.NormalizedAge, B=1.5), B=0.015)
        Min=0
        Max=1
       )
      )
     )
    )
    Particles.Scale_ConeWind:Vector3f=&VectorFromFloat(Value=20)
   ) {isSetParameters=true}
   /SetVariables(
    Particles.Color_RadialShockwave:LinearColor=&MultiplyLinearColor(
     Linear Color=<rgba 0.625 0.6562 0.75 0.35>
     Scale Factor=&MakeVector4_FromLinearColor(LinearColor=&ColorFromCurve(
       ColorCurve=DataInterface() {propertyValues {
         AlphaCurve {keys [<at 0.0504 1 RCIM_Cubic>,<at 0.3977 0.2>,<at 0.9103 0 RCIM_Cubic>]}
         BlueCurve {keys [<at 0.104 3 RCIM_Cubic>,<at 0.964 0.4149 RCIM_Cubic>]}
         GreenCurve {keys [<at 0.104 2.712 RCIM_Cubic>,<at 0.964 0.4149 RCIM_Cubic>]}
         RedCurve {keys [<at 0.104 2.5312 RCIM_Cubic>,<at 0.964 0.4149 RCIM_Cubic>]}
        }}
       CurveIndex=&NormalizeFloat(Value=&Divide_Float(A=$Particles.NormalizedAge, B=0.9), Min=0, Max=1)
      ))
    )
    Particles.DMP_RadialShockwave_00:Vector4f=&MakeVector4(
     X=&FloatFromCurve(
      FloatCurve=DataInterface() {propertyValues {Curve {keys [
          {arriveTangent=26.8123, interpMode=RCIM_Cubic, leaveTangent=26.8123, tangentMode=RCTM_User, time=0, value=0}
          {arriveTangent=0.9995, interpMode=RCIM_Cubic, leaveTangent=0.9995, tangentMode=RCTM_User, time=0.0743, value=0.7602}
          {arriveTangent=-0.0405, interpMode=RCIM_Cubic, leaveTangent=-0.0405, tangentMode=RCTM_User, time=1, value=1}
         ]}}}
      CurveIndex=&NormalizeFloat(Value=&Divide_Float(A=$Particles.NormalizedAge, B=0.9), Min=0, Max=1)
     )
     Y=0
     Z=0
     W=0
    )
   ) {isSetParameters=true}
   /SetVariables(Particles.DMP_RefractionShockwave_00:Vector4f=&MakeVector4(
     X=0
     Y=2
     Z=&FloatFromCurve(
      FloatCurve=DataInterface() {propertyValues {Curve {keys [<at 0 1>,<at 0.1 0.925>,<at 0.7 0.925>,<at 0.9 1>]}}}
      CurveIndex=&NormalizeFloat(Value=&Divide_Float(A=$Particles.NormalizedAge, B=0.5), Min=0, Max=1)
      Scale Curve=1.25
     )
     W=0
    )) {isSetParameters=true}
  ]
  rendererClasses [&NiagaraMeshRendererProperties, &NiagaraSpriteRendererProperties]
  renderers [
   {rendererClass=&NiagaraMeshRendererProperties, rendererIndex=0}
   {rendererClass=&NiagaraMeshRendererProperties, rendererIndex=1}
   {rendererClass=&NiagaraSpriteRendererProperties, rendererIndex=2}
   {rendererClass=&NiagaraMeshRendererProperties, rendererIndex=3}
  ]
 }}
summary {
 systemName=NS_TidaleCraneDrop_Spawn
 emitters [{emitterName=Shockwave, rendererClasses [&NiagaraMeshRendererProperties, &NiagaraSpriteRendererProperties]}]
 userVariables []
}
