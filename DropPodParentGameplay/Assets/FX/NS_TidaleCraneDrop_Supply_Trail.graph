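# Niagara trail effect for a falling supply drop pod, combining airflow meshes, glow with fake light, a fire ribbon, vortex-pulled sparks and distant smoke.
# The fire ribbon places burst-spawned points along the owner's Z axis, lerping toward goal positions so the tail lags.
# A system camera-proximity value scales size and intensity across most emitters.
# User.bIsMMOPod lengthens loops, widens and lengthens the fire ribbon, and enables the smoke ribbon.
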NiagaraSystem /DropPodParentGameplay/Assets/FX/NS_TidaleCraneDrop_Supply_Trail
dependencies {
 usedDataInterfaces [&NiagaraDataInterfaceCurve, &NiagaraDataInterfaceColorCurve, &NiagaraDataInterfaceCamera]
 usedDynamicInputs [
  &Cosine
  &ClampFloat
  &NormalizeFloat
  &DistanceBetweenPositions
  &ReturnCameraPosition
 ]
 usedModules [
  &SystemState
  &Timeline
  &self:SystemSpawnScript.SystemScriptSource.SystemScriptGraph.NiagaraNodeAssignment_1.SetVariables_#1
  &self:SystemSpawnScript.SystemScriptSource.SystemScriptGraph.NiagaraNodeAssignment_2.SetVariables_#2
  &self:SystemSpawnScript.SystemScriptSource.SystemScriptGraph.NiagaraNodeAssignment_0.SetVariables_#3
 ]
}
emitters {
 AirFlow {
  emitterUpdateScript [
   &EmitterState(
    Life Cycle Mode=::NewEnumerator1
    Loop Behavior=::NewEnumerator1
    Loop Duration=&MakeCustomFloatFromBool(Bool=$User.bIsMMOPod, True Float=7, False Float=1)
    UseLoopDelay=false
   )
   &SpawnRate(SpawnRate=5, Use Spawn Probability=false)
   &SpawnBurst_Instantaneous(Spawn Count=2, Use Spawn Probability=false, UseLoopCountLimit=false)
  ]
  particleSpawnScript [
   &PartitionParticles()
   &InitializeParticle(
    Color=&MergeHelper_ColorSwitch(0=<rgba 1 0.1579 0.0682 1>, 1=<rgba 1 0.6141 0.3474 1>, 2=?, 3=?, Index=$Particles.PartitionParticles.Partition, Num=::NewEnumerator0)
    Lifetime Max=1
    Lifetime Min=0.5
    Lifetime Mode=::NewEnumerator1
    Mesh Scale=&MergeHelper_VectorSwitch(
     0=&RandomRangeVector(Minimum=<1 0.75 0.75>, Maximum=<1 1 1>)
     1=<0.75 0.25 0.25>
     2=?
     3=?
     Index=$Particles.PartitionParticles.Partition
     Num=::NewEnumerator0
    )
    Mesh Scale Mode=::NewEnumerator4
    UsePositionOffset=false
    Write Material Random=true
   )
   &InitialMeshOrientation(
    Mesh Orientation Mode=::NewEnumerator2
    Rotation=&RandomRangeVector(Minimum=<0 0.25 0>, Maximum=<0 0.25 0>)
    Use Rotation Vector=true
   )
   &DynamicMaterialParameters(
    Index 0 Param 1=0.7
    Index 0 Param 2=1.5
    Index 0 Param 3=1
    Index 0 Param 4=1
    Module.Param0WriteEnabled=true
    Module.Param1WriteEnabled=true
    Module.Param2WriteEnabled=true
    Module.Param3WriteEnabled=true
   )
   &SystemLocation(MaskSpawnGroup=false, Offset=<0 0 -25>)
  ]
  particleUpdateScript [
   &ParticleState()
   &ScaleColor(
    Scale Alpha=&FloatFromCurve(
     FloatCurve=DataInterface() {propertyValues {Curve {keys [
         {interpMode=RCIM_Cubic, tangentMode=RCTM_User, time=0, value=0}
         {interpMode=RCIM_Cubic, tangentMode=RCTM_User, time=0.5, value=1}
         {tangentMode=RCTM_User, time=1, value=0}
        ]}}}
     Scale Curve=&MergeHelper_FloatSwitch(0=2.5, 1=1, 2=?, 3=?, Index=$Particles.PartitionParticles.Partition, Num=::NewEnumerator0)
    )
    Scale RGB=&Multiply_VectorByFloat(
     Vector=<1 1 1>
     Float=&FloatFromCurve(FloatCurve=DataInterface() {propertyValues {Curve {keys [<at 0 1>,<at 1 2>]}}})
    )
    ScaleA=true
    ScaleRGB=true
   )
   &ScaleMeshSize(Scale Factor=&VectorFromFloat(Value=&Lerp_Float(
      A=&MergeHelper_FloatSwitch(0=0.1, 1=0.2, 2=?, 3=?, Index=$Particles.PartitionParticles.Partition, Num=::NewEnumerator0)
      B=0.5
      Alpha=$System.ArtifactProximityToCamera
     )))
  ]
  rendererClasses [&NiagaraMeshRendererProperties]
  renderers [{rendererClass=&NiagaraMeshRendererProperties, rendererIndex=0}]
 }
 Glow_and_Light {
  emitterUpdateScript [
   &EmitterState(
    Inactive Response=::NewEnumerator1
    Life Cycle Mode=::NewEnumerator1
    Loop Behavior=::NewEnumerator1
    Loop Duration=&MakeCustomFloatFromBool(Bool=$User.bIsMMOPod, True Float=8, False Float=3)
    UseLoopDelay=false
   )
   &SpawnBurst_Instantaneous(Spawn Count=1, Use Spawn Probability=false)
   /SetVariables(Emitter.BurstAlpha:NiagaraFloat=&FloatFromCurve(
     FloatCurve=DataInterface() {propertyValues {Curve {keys [
         {arriveTangent=5.6877, interpMode=RCIM_Cubic, leaveTangent=5.6877, tangentMode=RCTM_User, time=0, value=0}
         {arriveTangent=-0.3288, interpMode=RCIM_Cubic, leaveTangent=-0.3288, tangentMode=RCTM_User, time=0.1, value=0.4}
         {arriveTangent=-0.1432, interpMode=RCIM_Cubic, leaveTangent=-0.1432, tangentMode=RCTM_User, time=0.2864, value=0.1903}
         {arriveTangent=-0.072, interpMode=RCIM_Cubic, leaveTangent=-0.072, tangentMode=RCTM_User, time=2, value=0}
        ]}}}
     CurveIndex=$Emitter.Age
    )) {isSetParameters=true}
  ]
  particleSpawnScript [
   &InitializeParticle(
    Color=<rgba 5 1 0 2>
    Lifetime=3
    Mesh Scale=&Multiply_VectorByFloat(Vector=<1 1 3>, Float=0.5)
    Mesh Scale Mode=::NewEnumerator4
    Position Offset=<0 0 25>
    Sprite Size Mode=::NewEnumerator3
    Uniform Sprite Size=100
    UsePositionOffset=true
    Write Material Random=true
   ) {module=InitializeParticle001}
   /SetVariables(Particles.LightExponent=2.5) {isSetParameters=true}
   /SetVariables(Particles.MeshOrientation=&MakeQuatFromAxisAngle(Axis=<0 0 1>, AngleInDegrees=90)) {isSetParameters=true}
   /SetVariables(Particles.GlowDirection:Vector3f=&TransformVector(Vector=<0 0 1>, Destination Space=::World)) {isSetParameters=true}
   /SetVariables(
    Particles.CameraOffset_FakeLight=50
    Particles.Color_FakeLight:LinearColor=<rgba 5 0.4167 0 0.5>
    Particles.SpriteSize_FakeLight:Vector2f=<500 500>
   ) {isSetParameters=true}
   &CameraOffset(Camera Offset Amount=30)
  ]
  particleUpdateScript [
   /SetVariables(Particles.LightColor=&MultiplyLinearColor(
     Linear Color=<rgba 10 0.6667 0 1>
     Scale Factor=&MakeVector4_FromFloat(Float=&Multiply_Float(
       A=&Sine(Normalized Angle=$Emitter.Age, Period=1, Scale=0.4, Bias=1.5)
       B=&FNTimeOfDayScalar(Night Scale=0.5, Morning Scale=0.2, Evening Scale=0.3)
      ))
    )) {isSetParameters=true}
   /SetVariables(Particles.LightRadius=80) {isSetParameters=true}
   &ParticleState()
   &SpriteRotationRate(Rotation Rate=30)
   &ScaleSpriteSize(
    Scale Sprite Size Mode=::NewEnumerator0
    Uniform Scale Factor=&Multiply_Float(
     A=&Lerp_Float(A=1, B=1.5, Alpha=$Emitter.BurstAlpha)
     B=&Lerp_Float(A=1.1, B=4, Alpha=$System.ArtifactProximityToCamera)
    )
   )
   &ScaleMeshSize(Scale Factor=&VectorFromFloat(Value=&Lerp_Float(A=0.5, B=1, Alpha=$System.ArtifactProximityToCamera)))
   &SolveForcesAndVelocity(Clamp Velocity=false, Limit Acceleration=false)
   &ScaleColor(
    Scale Alpha=&Multiply_Float(
     A=&Multiply_Float(A=$System.ScaleAlpha, B=&OneMinusFloat(Float=$System.Timeline.NormalizedTime))
     B=&Multiply_Float(
      A=&FNTimeOfDayScalar(Night Scale=0.5, Morning Scale=0.2, Evening Scale=0.3)
      B=&FloatFromCurve(FloatCurve=DataInterface() {propertyValues {Curve {
          defaultValue=1
          keys [
           {arriveTangent=-2.9436, interpMode=RCIM_Cubic, leaveTangent=-2.9436, tangentMode=RCTM_User, time=0, value=1}
           {interpMode=RCIM_Cubic, tangentMode=RCTM_User, time=1, value=0}
          ]
         }}})
     )
    )
    Scale RGB=&VectorFromFloat(Value=&Multiply_Float(
      A=&Multiply_Float(A=$System.ScaleAlpha, B=$System.GlowVariation)
      B=&Lerp_Float(A=1, B=10, Alpha=$Emitter.BurstAlpha)
     ))
    ScaleA=true
    ScaleRGB=true
   )
   &DynamicMaterialParameters(
    Index 0 Param 1=&MakeFloatFromVector(Vector=$Particles.GlowDirection, Channel=::NewEnumerator0)
    Index 0 Param 2=&MakeFloatFromVector(Vector=$Particles.GlowDirection, Channel=::NewEnumerator1)
    Index 0 Param 3=&MakeFloatFromVector(Vector=$Particles.GlowDirection, Channel=::NewEnumerator2)
    Index 0 Param 4=200
    Module.Index0Param0Write=true
    Module.Index0Param1Write=true
    Module.Index0Param2Write=true
    Module.Index0Param3Write=true
   )
   /SetVariables() {isSetParameters=true}
  ]
  rendererClasses [&NiagaraSpriteRendererProperties, &NiagaraMeshRendererProperties]
  renderers [
   {rendererClass=&NiagaraSpriteRendererProperties}
   {rendererClass=&NiagaraMeshRendererProperties, rendererIndex=1}
   {rendererClass=&NiagaraSpriteRendererProperties, rendererIndex=2}
  ]
 }
 NE_FireTrail {
  emitterSpawnScript [/SetVariables(Emitter.ParticlePerRibbon=20, Emitter.RibbonNumber=1) {isSetParameters=true}]
  emitterUpdateScript [
   &EmitterState(
    Inactive Response=::NewEnumerator1
    Life Cycle Mode=::NewEnumerator1
    Loop Behavior=::NewEnumerator1
    Loop Duration Mode=::NewEnumerator1
    UseLoopDelay=false
   )
   &SpawnBurst_Instantaneous(
    Spawn Count=&Multiply_Int(A=$Emitter.RibbonNumber, B=$Emitter.ParticlePerRibbon)
    Use Spawn Probability=false
    UseLoopCountLimit=false
   )
  ]
  particleSpawnScript [
   &PartitionParticles(
    ID to Partition:NiagaraInt32=&ReturnExecIndex()
    Number Of Partitions=$Emitter.RibbonNumber
    Particles Per Partition:NiagaraInt32=$Emitter.ParticlePerRibbon
    Partition Mode:ENiagara_PartitionMode=::NewEnumerator1
    Write Partition to Ribbon ID=::NewEnumerator0
   )
   /SetVariables() {isSetParameters=true}
   &InitializeParticle(
    Color Mode=::NewEnumerator0
    Lifetime=5
    Ribbon Width=&MakeCustomFloatFromBool(Bool=$User.bIsMMOPod, True Float=8, False Float=6.5)
    Ribbon Width Mode=::NewEnumerator1
    UsePositionOffset=false
    Write Material Random=true
   )
   /SetVariables(Particles.RibbonU0Override=$Particles.PartitionParticles.Normalized Index In Partition) {isSetParameters=true}
  ]
  particleUpdateScript [
   &ParticleState(Kill Particles When Lifetime Has Elapsed=false)
   /SetVariables(Particles.RibbonLength:NiagaraFloat=&MakeCustomFloatFromBool(Bool=$User.bIsMMOPod, True Float=2500, False Float=750)) {isSetParameters=true}
   /SetVariables(Particles.RibbonWidth=&Multiply_Float(
     A=$Particles.Initial.RibbonWidth
     B=&Lerp_Float(
      A=8
      B=20
      Alpha=&FloatFromCurve(
       FloatCurve=DataInterface() {propertyValues {Curve {keys [
           {arriveTangent=2.4963, interpMode=RCIM_Cubic, leaveTangent=2.4963, tangentMode=RCTM_User, time=0, value=0.3}
           {arriveTangent=5.8828, interpMode=RCIM_Cubic, leaveTangent=5.8828, tangentMode=RCTM_User, time=0.1, value=0.7}
           {arriveTangent=-0.0177, interpMode=RCIM_Cubic, leaveTangent=-0.0177, tangentMode=RCTM_User, time=0.7, value=2}
          ]}}}
       CurveIndex=$Particles.PartitionParticles.Normalized Index In Partition
       Scale Curve=&Lerp_Float(A=1, B=3, Alpha=$System.ArtifactProximityToCamera)
      )
     )
    )) {isSetParameters=true}
   /SetVariables(Particles.GlowWidth:NiagaraFloat=&Multiply_Float(A=$Particles.RibbonWidth, B=&Lerp_Float(A=1.1, B=1.3, Alpha=$System.ArtifactProximityToCamera))) {isSetParameters=true}
   /SetVariables(Particles.XOffsetValue:NiagaraFloat=&Add_Float(
     A=$Particles.XOffsetValue
     B=&Multiply_Float(A=$Engine.DeltaTime, B=&Lerp_Float(A=1.5, B=0.5, Alpha=$System.ArtifactProximityToCamera))
    )) {isSetParameters=true}
   /SetVariables(Particles.GoalPos:NiagaraPosition=&AddVectorToPosition(
     Position=$Engine.Owner.Position
     Vector=&Multiply_VectorByFloat(
      Vector=&TransformVector(
       Vector=&Lerp_Vector(
        A=<0 0 1>
        B=&MakeVector(X=0, Y=0, Z=&Lerp_Float(A=1, B=5, Alpha=$System.ArtifactProximityToCamera))
        Alpha=&FloatFromCurve(
         FloatCurve=DataInterface() {propertyValues {Curve {keys [<at 0 0>,<at 1 1>]}}}
         CurveIndex=$Particles.PartitionParticles.Normalized Index In Partition
        )
       )
       Destination Space=::Simulation
      )
      Float=&Lerp_Float(A=0, B=$Particles.RibbonLength, Alpha=$Particles.PartitionParticles.Normalized Index In Partition)
     )
    )) {isSetParameters=true}
   /SetVariables(Particles.Position=&Lerp_Vector(
     A=Linked() {linkedVariable {name=Particles.GoalPos, type=NiagaraPosition}}
     B=&Lerp_Vector(
      A=Linked() {linkedVariable {name=Particles.Position, type=NiagaraPosition}}
      B=Linked() {linkedVariable {name=Particles.GoalPos, type=NiagaraPosition}}
      Alpha=&Multiply_Float(
       A=$Engine.DeltaTime
       B=&Lerp_Float(
        A=15
        B=2
        Alpha=&FloatFromCurve(
         FloatCurve=DataInterface() {propertyValues {Curve {keys [
             {arriveTangent=3.7014, interpMode=RCIM_Cubic, leaveTangent=3.7014, tangentMode=RCTM_User, time=0, value=0}
             {arriveTangent=-0.0363, interpMode=RCIM_Cubic, leaveTangent=-0.0363, tangentMode=RCTM_User, time=1, value=1}
            ]}}}
         CurveIndex=$Particles.PartitionParticles.Normalized Index In Partition
        )
       )
      )
     )
     Alpha=&FloatFromCurve(
      FloatCurve=DataInterface() {propertyValues {Curve {keys [
          {arriveTangent=2.6466, interpMode=RCIM_Cubic, leaveTangent=2.6466, tangentMode=RCTM_User, time=0, value=0}
          {arriveTangent=0.2102, interpMode=RCIM_Cubic, leaveTangent=0.2102, tangentMode=RCTM_User, time=1, value=1}
         ]}}}
      CurveIndex=$Particles.PartitionParticles.Normalized Index In Partition
     )
    )) {isSetParameters=true}
   &DynamicMaterialParameters(
    Index 0 Param 1=&Add_Float(
     A=/Game/Effects/Niagara/DynamicInputs/Random/FN_SeededRandomRangeFloat(Minimum:NiagaraFloat=1, Maximum:NiagaraFloat=5, Random Seed:NiagaraInt32=$Particles.PartitionParticles.Partition)
     B=$Particles.XOffsetValue
    )
    Index 0 Param 2=$System.SmoothDiveAlpha
    Index 1 Param 3=&FloatFromCurve(
     FloatCurve=DataInterface() {propertyValues {Curve {keys [<at 0.2 0>,<at 1 2>]}}}
     CurveIndex=&Add_Float(A=$Particles.PartitionParticles.Normalized Index In Partition, B=$System.Timeline.NormalizedTime)
    )
    Module.Index0Param0Write=true
    Module.Index0Param1Write=true
    Module.Index0Param2Write=true
    Module.Index0Param3Write=true
    Module.Index1Param0Write=true
    Module.Index1Param1Write=true
    Module.Index1Param2Write=true
    Module.Index1Param3Write=false
    Write Parameter Index 1=true
   )
   &ScaleColor(
    Scale Alpha=&Multiply_Float(
     A=&FloatFromCurve(
      FloatCurve=DataInterface() {propertyValues {Curve {keys [
          {arriveTangent=-1.6635, interpMode=RCIM_Cubic, leaveTangent=-1.6635, tangentMode=RCTM_User, time=0.15, value=1}
          {arriveTangent=-0.3046, interpMode=RCIM_Cubic, leaveTangent=-0.3046, tangentMode=RCTM_User, time=0.8, value=0}
         ]}}}
      CurveIndex=$Particles.PartitionParticles.Normalized Index In Partition
     )
     B=0.5
    )
    Scale RGB=&VectorFromFloat(Value=&FloatFromCurve(
      FloatCurve=DataInterface() {propertyValues {Curve {keys [
          {arriveTangent=-26.9347, interpMode=RCIM_Cubic, leaveTangent=-26.9347, tangentMode=RCTM_User, time=0, value=7}
          {interpMode=RCIM_Cubic, tangentMode=RCTM_User, time=0.5, value=1}
         ]}}}
      CurveIndex=$Particles.PartitionParticles.Normalized Index In Partition
     ))
    ScaleA=true
    ScaleRGB=true
   )
   /SetVariables(Particles.ColorGlow:LinearColor=&MultiplyLinearColor(
     Linear Color=$Particles.Color
     Scale Factor=&MakeFloat4FromVectorAndFloat(
      "Vector (RGB)"=&VectorFromFloat(Value=&Lerp_Float(A=1, B=3, Alpha=$System.ArtifactProximityToCamera))
      "Float (Alpha)"=1
     )
    )) {isSetParameters=true}
  ]
  rendererClasses [&NiagaraRibbonRendererProperties]
  renderers [
   {rendererClass=&NiagaraRibbonRendererProperties}
   {rendererClass=&NiagaraRibbonRendererProperties, rendererIndex=1}
   {rendererClass=&NiagaraRibbonRendererProperties, rendererIndex=2}
  ]
 }
 Smoke_Ribbon_Far {
  emitterUpdateScript [
   &EmitterState(Life Cycle Mode=::NewEnumerator1, Loop Behavior=::NewEnumerator1, Loop Duration Mode=::NewEnumerator1, UseLoopDelay=false)
   &SpawnRate(
    Spawn Probability=&MakeCustomFloatFromBool(Bool=$User.bIsMMOPod, True Float=1, False Float=0)
    SpawnRate=&FloatFromCurve(
     FloatCurve=DataInterface() {propertyValues {Curve {
        defaultValue=0
        keys [
         {arriveTangent=10.9349, interpMode=RCIM_Cubic, leaveTangent=10.9349, tangentMode=RCTM_User, time=0, value=0}
         {arriveTangent=0.2911, interpMode=RCIM_Cubic, leaveTangent=0.2911, tangentMode=RCTM_User, time=0.1707, value=0.8917}
         {arriveTangent=0.0224, interpMode=RCIM_Cubic, leaveTangent=0.0224, tangentMode=RCTM_User, time=1, value=1}
        ]
       }}}
     CurveIndex=$Emitter.Age
     Scale Curve=40
    )
    Use Spawn Probability=true
   )
  ]
  particleSpawnScript [
   &InitializeParticle(
    Color Mode=::NewEnumerator0
    Lifetime=3
    Ribbon Facing Mode=::NewEnumerator1
    Ribbon Width=2000
    Ribbon Width Mode=::NewEnumerator1
    UsePositionOffset=false
    Write Material Random=true
   )
   &AddVelocity(Velocity=<0 0 1>, Velocity Speed Scale=2000)
  ]
  particleUpdateScript [
   &SolveForcesAndVelocity(Clamp Velocity=false, Limit Acceleration=false)
   &ParticleState()
   &Color(
    Color=&UniformRangedLinearColor(Minimum=<rgba 0.1 0.112 0.125 3>, Maximum=<rgba 0.0975 0.1108 0.125 3>)
    Scale Alpha=&Multiply_Float(
     A=&FloatFromCurve(
      FloatCurve=DataInterface() {propertyValues {Curve {keys [
          {arriveTangent=23.7432, interpMode=RCIM_Cubic, leaveTangent=23.7432, tangentMode=RCTM_User, time=0, value=0}
          <at 0.0926 0.8137>
          {arriveTangent=-1.2137, interpMode=RCIM_Cubic, leaveTangent=-1.2137, tangentMode=RCTM_User, time=0.4764, value=0.3008}
          {arriveTangent=0.2062, interpMode=RCIM_Cubic, leaveTangent=0.2062, tangentMode=RCTM_User, time=1, value=0}
         ]}}}
      Scale Curve=&FN_LOD_Distance_Remap(Start Distance=2500, End Distance=10000, Minimum Output=0.25, Maximum Output=0.8, Clamp Output=true)
     )
     B=&FloatFromCurve(
      FloatCurve=DataInterface() {propertyValues {Curve {keys [
          {arriveTangent=1.1513, interpMode=RCIM_Cubic, leaveTangent=1.1513, tangentMode=RCTM_User, time=0, value=0}
          {arriveTangent=0.0271, interpMode=RCIM_Cubic, leaveTangent=0.0271, tangentMode=RCTM_User, time=1, value=1}
          {arriveTangent=-0.0029, interpMode=RCIM_Cubic, leaveTangent=-0.0029, tangentMode=RCTM_User, time=5, value=0.9885}
          <at 8 0 RCIM_Cubic>
         ]}}}
      CurveIndex=$Emitter.Age
     )
    )
   )
   &DynamicMaterialParameters(
    Index 2 Param 3=0.3
    Index 2 Param 4=3
    Module.Param0WriteEnabled=true
    Module.Param0WriteEnabled1=true
    Module.Param0WriteEnabled2=true
    Module.Param0WriteEnabled3=true
    Module.Param1WriteEnabled=true
    Module.Param1WriteEnabled1=true
    Module.Param1WriteEnabled2=true
    Module.Param1WriteEnabled3=true
    Module.Param2WriteEnabled=true
    Module.Param2WriteEnabled1=true
    Module.Param2WriteEnabled2=true
    Module.Param2WriteEnabled3=true
    Module.Param3WriteEnabled=true
    Module.Param3WriteEnabled1=true
    Module.Param3WriteEnabled2=true
    Module.Param3WriteEnabled3=true
    Write Parameter Index 1=true
    Write Parameter Index 2=true
    Write Parameter Index 3=true
   )
   &ScaleRibbonWidth(Ribbon Width Scale=&FloatFromCurve(
     FloatCurve=DataInterface() {propertyValues {Curve {keys [
         {arriveTangent=5.7391, interpMode=RCIM_Cubic, leaveTangent=5.7391, tangentMode=RCTM_User, time=0, value=0}
         {arriveTangent=1.1179, interpMode=RCIM_Cubic, leaveTangent=1.1179, tangentMode=RCTM_User, time=0.1852, value=0.5105}
         {arriveTangent=0.1708, interpMode=RCIM_Cubic, leaveTangent=0.1708, time=0.732, value=0.8651}
         {arriveTangent=-2.3376, interpMode=RCIM_Cubic, leaveTangent=-2.3376, tangentMode=RCTM_User, time=1.0042, value=0.6504}
        ]}}}
     Scale Curve=&FN_LOD_Distance_Remap(Start Distance=2500, End Distance=10000, Minimum Output=1, Maximum Output=2, Clamp Output=true)
    )) {module=ScaleRibbonWidth001}
  ]
  rendererClasses [&NiagaraRibbonRendererProperties]
  renderers [{rendererClass=&NiagaraRibbonRendererProperties}]
 }
 Sparkles {
  emitterUpdateScript [
   &EmitterState(Life Cycle Mode=::NewEnumerator1, Loop Duration=1, UseLoopDelay=false)
   &SpawnRate(
    SpawnRate=&Multiply_Float(A=10, B=&OneMinusFloat(Float=$System.Timeline.NormalizedTime))
    Use Spawn Probability=false
   )
  ]
  particleSpawnScript [
   &InitializeParticle(
    Color=<rgba 10 10 10 1>
    Lifetime=&Multiply_Float(
     A=&RandomRangeFloat(Minimum=0.6, Maximum=1.2)
     B=&Lerp_Float(A=1, B=2, Alpha=$System.ArtifactProximityToCamera)
    )
    Sprite Size=&Multiply_Vector2D_ByFloat(
     Float=&Multiply_Float(
      A=&RandomRangeFloat(Minimum=6, Maximum=8)
      B=&Lerp_Float(A=1, B=3, Alpha=$System.ArtifactProximityToCamera)
     )
     Vector2D=<0.5 1.5>
    )
    Sprite Size Mode=::NewEnumerator4
    UsePositionOffset=false
    Write Material Random=true
   ) {module=InitializeParticle001}
   &ShapeLocation(
    MaskSpawnGroup=false
    Override Seed=false
    Sphere Radius=&Lerp_Float(A=50, B=100, Alpha=$System.ArtifactProximityToCamera)
    Sphere Surface Distribution=1
    UseOwnerScale=true
   )
   &AddVelocity(
    Module.Use Falloff Distance=false
    Origin Offset=&MakeVector(X=0, Y=0, Z=&Lerp_Float(A=-200, B=-300, Alpha=$System.ArtifactProximityToCamera))
    Override Seed=false
    Velocity Mode=::NewEnumerator1
    Velocity Speed=&Multiply_Float(
     A=&RandomRangeFloat(Minimum=650, Maximum=1000)
     B=&Lerp_Float(A=1, B=2, Alpha=$System.ArtifactProximityToCamera)
    )
   ) {module=AddVelocity001}
  ]
  particleUpdateScript [
   &ParticleState()
   &ScaleSpriteSize(Scale Factor=&Vector2DFromFloat(Value=&FloatFromCurve(FloatCurve=DataInterface() {propertyValues {Curve {keys [
          {interpMode=RCIM_Cubic, tangentMode=RCTM_User, time=-0.0046, value=0}
          {interpMode=RCIM_Cubic, tangentMode=RCTM_User, time=0.1746, value=0.9952}
          {interpMode=RCIM_Cubic, tangentMode=RCTM_User, time=1.0011, value=0.002}
         ]}}})))
   &Color(
    Color=&ColorFromCurve(ColorCurve=DataInterface() {propertyValues {
       AlphaCurve {keys [<at 0 1 RCIM_Cubic>,<at 1 1 RCIM_Cubic>]}
       BlueCurve {keys [<at 0 0 RCIM_Cubic>,<at 0.1526 0>,<at 0.7427 0 RCIM_Cubic>,<at 1 0>]}
       GreenCurve {keys [
         <at 0 4.0263 RCIM_Cubic>
         <at 0.1526 0.4158>
         {arriveTangent=-0.4907, interpMode=RCIM_Cubic, leaveTangent=-0.4907, time=0.7427, value=0.0752}
         <at 1 0>
        ]}
       RedCurve {keys [<at 0 10 RCIM_Cubic>, <at 0.1526 3>, {arriveTangent=-3.1864, interpMode=RCIM_Cubic, leaveTangent=-3.1864, time=0.7427, value=1}, <at 1 0.3>]}
      }})
    Scale Alpha=&FloatFromCurve(FloatCurve=DataInterface() {propertyValues {Curve {keys [<at -0.0092 0.0144>,<at 0.0967 0.9767>,<at 1.0011 0.002>]}}})
   )
   &DynamicMaterialParameters(
    Index 0 Param 1=1
    Index 0 Param 2=0.5
    Index 0 Param 3=5
    Module.Index0Param0Write=true
    Module.Index0Param1Write=true
    Module.Index0Param2Write=true
    Module.Index0Param3Write=true
   ) {module=DynamicMaterialParameters001}
   /SetVariables(Particles.GlowColor=&MultiplyLinearColor(Linear Color=$Particles.Color, Scale Factor=<1 1 1 0.08>)) {isSetParameters=true}
   /SetVariables(Particles.GlowSize=&Multiply_Vector2D(A=$Particles.SpriteSize, B=<2.5 2.5>)) {isSetParameters=true}
   &VortexForce(Origin Pull Amount=850, UseOriginOffset=false, Vortex Force Amount=1500)
   &SolveForcesAndVelocity(Clamp Velocity=false, Limit Acceleration=false)
  ]
  rendererClasses [&NiagaraSpriteRendererProperties]
  renderers [
   {rendererClass=&NiagaraSpriteRendererProperties}
   {rendererClass=&NiagaraSpriteRendererProperties, rendererIndex=1}
  ]
 }
}
summary {
 systemName=NS_TidaleCraneDrop_Supply_Trail
 emitters [
  {emitterName=AirFlow, rendererClasses [&NiagaraMeshRendererProperties]}
  {emitterName=Glow_and_Light, rendererClasses [&NiagaraSpriteRendererProperties, &NiagaraMeshRendererProperties]}
  {emitterName=NE_FireTrail, rendererClasses [&NiagaraRibbonRendererProperties]}
  {emitterName=Sparkles, rendererClasses [&NiagaraSpriteRendererProperties]}
  {emitterName=Smoke_Ribbon_Far, rendererClasses [&NiagaraRibbonRendererProperties]}
 ]
 userVariables [
  User.bIsMMOPod {type=NiagaraBool}
  User.FadeOut {type=NiagaraBool}
  User.uPhase {type=NiagaraFloat, defaultValue=1, description=""}
 ]
}
