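# Material function that draws the Tidal Crane drop pod's emissive screen into incoming material attributes, limited by a mask channel.
# A tangent-space view vector offsets UI panel, grid, logo and rank layers at separate parameterized depths to fake parallax, with triple sampling ghosting logo and rank.
# The Rank parameter picks a five-frame SubUV icon and switches between logo and rank looks.
# Scrolling noise and Fresnel add glow before default, rank or deactivated tinting.
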MaterialFunction /DropPodParentGameplay/Assets/Materials/MF_DropPod_TidalCrane_Screen
graph {nodes [
  Constant() {r=0.2}
  Constant() {r=0.8}
  Constant() {r=0.4}
  n1 = GetMaterialAttributes(None=FunctionInput() {inputName=Result, inputType=FunctionInput_MaterialAttributes}) {attributeGetTypes [#1]}
  n2 = LinearInterpolate(
   A=Multiply(A=@n1.EmissiveColor, B=ScalarParameter() {defaultValue=4, group=Screen, parameterName=RankIntensity})
   B=@n1.EmissiveColor
   Alpha=Step(Y=NamedRerouteUsage()) {constX=-1, constY=-1}
  )
  n3 = CameraVectorWS()
  n4 = Subtract(
   A=Transform(None=Multiply(A=@n3) {constB=-1}) {transformSourceType=TRANSFORMSOURCE_World, transformType=TRANSFORM_Tangent}
   B=Transform(None=VertexNormalWS()) {transformSourceType=TRANSFORMSOURCE_World, transformType=TRANSFORM_Tangent}
  )
  n5 = MaterialFunctionCall(Float3=@n4) {materialFunction=&BreakOutFloat3Components}
  n6 = VectorParameter() {parameterName=ScreenGraphDepth, channelNames {a=Rank, b=Grid, g=Logo, r=UIScreenPanel}, defaultValue=<rgba 0.2 0.4 0.8 0.4>}
  n7 = Divide(
   A=Multiply(
    A=MaterialFunctionCall(X=@n5.R, Y=@n5.R) {materialFunction=&MakeFloat2}.Result
    B=@n6.Grid
   )
   B=Multiply(A=@n5.B) {constB=-1}
  )
  n8 = Max(A=Add(A=ComponentMask(None=@n7) {g=true}) {constB=0.25}, B=Fresnel() {exponent=4})
  n9 = Multiply(A=TextureCoordinate()) {constB=5}
  n10 = Add(A=Subtract(A=ConstantBiasScale(None=@n9) {bias=-0.5, scale=1.5}) {constB=-0.5}, B=Constant2Vector() {g=-2.65, r=-1.875})
  n11 = MaterialFunctionCall(Float3=@n4) {materialFunction=&BreakOutFloat3Components}
  n12 = Constant() {r=0.08}
  n13 = Divide(
   A=Multiply(
    A=MaterialFunctionCall(X=@n11.R, Y=@n11.R) {materialFunction=&MakeFloat2}.Result
    B=Subtract(A=@n6.Logo, B=@n12)
   )
   B=Multiply(A=@n11.B) {constB=-1}
  )
  n14 = Constant2Vector() {g=0.045}
  n15 = TextureSample(UVs=Add(A=Add(A=@n10, B=@n13), B=@n14)) {
   samplerSource=SSM_Clamp_WorldGroupSettings
   samplerType=SAMPLERTYPE_LinearColor
   texture=/DropPodParentGameplay/Assets/Textures/T_DropPod_TidalCrane_Screen_TA
  }
  n16 = MaterialFunctionCall(Float3=@n4) {materialFunction=&BreakOutFloat3Components}
  n17 = Divide(
   A=Multiply(
    A=MaterialFunctionCall(X=@n16.R, Y=@n16.R) {materialFunction=&MakeFloat2}.Result
    B=@n6.Logo
   )
   B=Multiply(A=@n16.B) {constB=-1}
  )
  n18 = TextureSample(UVs=Add(A=Add(A=@n10, B=@n17), B=@n14)) {
   samplerSource=SSM_Clamp_WorldGroupSettings
   samplerType=SAMPLERTYPE_LinearColor
   texture=&T_DropPod_TidalCrane_Screen_TA
  }
  n19 = Add(A=@n15.G, B=Multiply(A=@n18.G) {constB=0.2})
  n20 = MaterialFunctionCall(Float3=@n4) {materialFunction=&BreakOutFloat3Components}
  n21 = Divide(
   A=Multiply(
    A=MaterialFunctionCall(X=@n20.R, Y=@n20.R) {materialFunction=&MakeFloat2}.Result
    B=Add(A=@n6.Logo, B=@n12)
   )
   B=Multiply(A=@n20.B) {constB=-1}
  )
  n22 = TextureSample(UVs=Add(A=Add(A=@n10, B=@n21), B=@n14)) {
   samplerSource=SSM_Clamp_WorldGroupSettings
   samplerType=SAMPLERTYPE_LinearColor
   texture=&T_DropPod_TidalCrane_Screen_TA
  }
  n23 = Multiply(A=@n22.G) {constB=0.05}
  n24 = LinearInterpolate(A=NamedRerouteUsage(), B=Saturate(None=Add(A=@n19, B=@n23)), Alpha=Step(Y=NamedRerouteUsage()) {constX=-1, constY=-1})
  n25 = MaterialFunctionCall(Float3=@n4) {materialFunction=&BreakOutFloat3Components}
  n26 = Divide(
   A=Multiply(
    A=MaterialFunctionCall(X=@n25.R, Y=@n25.R) {materialFunction=&MakeFloat2}.Result
    B=@n6.UIScreenPanel
   )
   B=Multiply(A=@n25.B) {constB=-1}
  )
  n27 = TextureSample(UVs=Add(A=Add(A=@n10, B=@n26), B=@n14)) {
   samplerSource=SSM_Clamp_WorldGroupSettings
   samplerType=SAMPLERTYPE_LinearColor
   texture=&T_DropPod_TidalCrane_Screen_TA
  }
  n28 = Add(A=@n10, B=@n7)
  n29 = Add(A=Add(A=Multiply(A=@n28) {constB=2}) {constB=-0.5}, B=@n14)
  n30 = TextureSample(UVs=@n29) {
   samplerSource=SSM_Wrap_WorldGroupSettings
   samplerType=SAMPLERTYPE_LinearColor
   texture=&T_DropPod_TidalCrane_Screen_TA
  }
  n31 = Subtract(A=Multiply(A=Add(A=@n29, B=@n14)) {constB=2})
  n32 = SmoothStep(Value=Abs(None=ComponentMask(None=@n31) {r=true})) {constMax=0, constMin=3}
  n33 = Multiply(A=@n30.B, B=LinearInterpolate(Alpha=@n32) {constA=0.05, constB=0.25}) {constB=0.1}
  n34 = Multiply(A=Add(A=@n29, B=@n14)) {constB=2}
  n35 = Abs(None=ComponentMask(None=Subtract(A=@n34)) {g=true})
  n36 = Saturate(None=Multiply(A=@n32, B=SmoothStep(Value=@n35) {constMax=0, constMin=2}))
  n37 = Constant() {r=1}
  n38 = ScalarParameter() {defaultValue=0.5, parameterName=BG Noise Alpha}
  n39 = LinearInterpolate(Alpha=@n38) {constA=0.1, constB=0.005}
  n40 = Panner(Coordinate=Multiply(A=@n29, B=MaterialFunctionCall(X=@n37, Y=@n39) {materialFunction=&MakeFloat2}.Result)) {speedY=0.2}
  n41 = Power(Base=Multiply(A=TextureSample(UVs=@n40) {samplerSource=SSM_Wrap_WorldGroupSettings, texture=/Game/Global/Textures/Default/Noise/T_godray_Noise}.R, B=@n38))
  n42 = Add(A=Add(A=Multiply(A=@n27.R) {constB=0.5}, B=Multiply(A=@n33, B=@n36)), B=Multiply(A=Multiply(A=@n36, B=@n41)) {constB=0.5})
  n43 = VectorParameter() {group=Screen, parameterName=RankTint, defaultValue=<rgba 1 1 1 1>}
  n44 = Multiply(A=Multiply(A=@n43.RGB, B=@n43.RGB), B=ScalarParameter() {defaultValue=4, group=Screen, parameterName=RankIntensity})
  n45 = VectorParameter() {group=Screen, parameterName=DefaultScreenTint, defaultValue=<rgba 0.17 1 0.84 1>}
  n46 = Multiply(A=@n45.RGB, B=@n45.RGB)
  n47 = Step(Y=NamedRerouteUsage()) {constX=-1, constY=-1}
  n48 = VectorParameter() {group=Screen, parameterName=DeactivedScreenTint, channelNames {a=Alpha, b="", g="", r=""}, defaultValue=<rgba 1 0 0 0.1>}
  n49 = Multiply(
   A=Add(A=Multiply(A=@n8) {constB=0.5}, B=Add(A=@n24, B=@n42))
   B=LinearInterpolate(
    A=LinearInterpolate(A=@n44, B=@n46, Alpha=@n47)
    B=Multiply(A=@n48, B=@n48)
    Alpha=ScalarParameter() {group=Screen, parameterName=bIsDeactived}
   )
  )
  n50 = ChannelMaskParameter(None=FunctionInput(Preview=Constant() {r=1}) {inputName=TA Mask, sortPriority=1}) {group=Screen, maskChannel=Green, parameterName=Screen Msk Channel, sortPriority=0, defaultValue=<rgba 0 1 0 0>}
  FunctionOutput(None=SetMaterialAttributes!(MaterialAttributes=@n1.MaterialAttributes, EmissiveColor=LinearInterpolate(A=@n2, B=@n49, Alpha=@n50.R)) {attributeSetTypes [#1]}) {outputName=SetMaterialAttributes}
  n51 = Constant2Vector() {g=0.05, r=-0.14}
  Multiply(A=@n51) {constB=0.35}
  n52 = TextureObject() {texture=/DropPodParentGameplay/Assets/Textures/T_DropPod_TidalCrane_Screen_Ranks_TA}
  n53 = MaterialFunctionCall(Float3=@n4) {materialFunction=&BreakOutFloat3Components}
  n54 = Constant() {r=0.08}
  n55 = Multiply(
   A=MaterialFunctionCall(X=@n53.R, Y=@n53.R) {materialFunction=&MakeFloat2}.Result
   B=Subtract(A=@n6.Rank, B=@n54)
  )
  n56 = Multiply(A=@n53.B) {constB=-1}
  n57 = Constant2Vector() {g=0.025}
  n58 = Add(A=Add(A=@n10, B=Divide(A=@n55, B=@n56)), B=@n57) {constB=0}
  n59 = Multiply(A=Constant2Vector() {g=3.2, r=4}) {constB=0.32}
  n60 = NamedRerouteDeclaration(None=ScalarParameter() {defaultValue=-1, group=Screen, parameterName=Rank, sliderMax=4}) {name=Rank, nodeColor=<rgba 1 0 0.75 1>}
  n61 = MaterialFunctionCall!(Texture=@n52, UVs=Add(A=Multiply(A=@n58, B=@n59) {constB=3}, B=@n51) {constB=0.5}, SubImages=Constant2Vector() {g=1, r=5}, Frame=@n60) {materialFunction=/Engine/Functions/Engine_MaterialFunctions03/Texturing/SubUV_Function}
  n62 = Subtract(A=Multiply(A=@n58) {constB=2})
  n63 = Constant2Vector() {g=0.5, r=0.5}
  n64 = Step(Y=Abs(None=ComponentMask(None=@n62) {r=true}), X=@n63.R) {constX=0.33, constY=0.4}
  n65 = Step(Y=Abs(None=ComponentMask(None=@n62) {g=true}), X=@n63.G) {constX=0.33, constY=0.4}
  n66 = MaterialFunctionCall(Float3=@n4) {materialFunction=&BreakOutFloat3Components}
  n67 = Multiply(
   A=MaterialFunctionCall(X=@n66.R, Y=@n66.R) {materialFunction=&MakeFloat2}.Result
   B=@n6.Rank
  )
  n68 = Multiply(A=@n66.B) {constB=-1}
  n69 = Add(A=Add(A=@n10, B=Divide(A=@n67, B=@n68)), B=@n57) {constB=0}
  n70 = Multiply(A=@n69, B=@n59) {constB=3}
  n71 = ComponentMask(None=MaterialFunctionCall(Texture=@n52, UVs=Add(A=@n70, B=@n51) {constB=0.5}, SubImages=Constant2Vector() {g=1, r=5}, Frame=@n60) {materialFunction=&SubUV_Function}.RGB) {r=true}
  n72 = Subtract(A=Multiply(A=@n69) {constB=2})
  n73 = ComponentMask(None=@n72) {r=true}
  n74 = ComponentMask(None=@n72) {g=true}
  n75 = Multiply(A=Step(Y=Abs(None=@n73), X=@n63.R) {constX=0.33, constY=0.4}, B=Step(Y=Abs(None=@n74), X=@n63.G) {constX=0.33, constY=0.4})
  n76 = Add(A=Multiply(A=ComponentMask(None=@n61.RGB) {r=true}, B=Multiply(A=@n64, B=@n65)), B=Multiply(A=Multiply(A=@n71, B=@n75)) {constB=0.2})
  n77 = MaterialFunctionCall(Float3=@n4) {materialFunction=&BreakOutFloat3Components}
  n78 = Multiply(
   A=MaterialFunctionCall(X=@n77.R, Y=@n77.R) {materialFunction=&MakeFloat2}.Result
   B=Add(A=@n6.Rank, B=@n54)
  )
  n79 = Multiply(A=@n77.B) {constB=-1}
  n80 = Add(A=Add(A=@n10, B=Divide(A=@n78, B=@n79)), B=@n57) {constB=0}
  n81 = MaterialFunctionCall(Texture=@n52, UVs=Add(A=Multiply(A=@n80, B=@n59) {constB=3}, B=@n51) {constB=0.5}, SubImages=Constant2Vector() {g=1, r=5}, Frame=@n60) {materialFunction=&SubUV_Function}
  n82 = Subtract(A=Multiply(A=@n80) {constB=2})
  n83 = Step(Y=Abs(None=ComponentMask(None=@n82) {r=true}), X=@n63.R) {constX=0.3, constY=0.4}
  n84 = Step(Y=Abs(None=ComponentMask(None=@n82) {g=true}), X=@n63.G) {constX=0.3, constY=0.4}
  n85 = Multiply(A=Multiply(A=ComponentMask(None=@n81.RGB) {r=true}, B=Multiply(A=@n83, B=@n84))) {constB=0.05}
  NamedRerouteDeclaration(None=Saturate(None=Add(A=@n76, B=@n85))) {name=RankGraph, nodeColor=<rgba 0 1 0.9297 1>}
  Step()
  VectorParameter() {parameterName=RGBChannel, defaultValue=<rgba 0 1 0 1>}
 ]}
