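# Material function that animates a drop pod material's World Position Offset and emissive over supplied material attributes.
# Blueprint-set start time and duration give normalized time, while curve atlas channels drive stomp, shake and blink blends.
# Object-position-offset sine noise supplies shake, and TA Mask green and blue select screen and emissive areas for unstable tint and flicker.
# CurrentStage remaps into per-stage curves, with a DebugMode switch cycling it by time.
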MaterialFunction /DropPodParentGameplay/Assets/Materials/MF_DropPod_TidalCrane_TA
graph {nodes [
  n1 = GetMaterialAttributes(None=FunctionInput() {inputName=IN MA, inputType=FunctionInput_MaterialAttributes}) {attributeGetTypes [#1, #2]}
  n2 = Add(A=NamedRerouteUsage(), B=NamedRerouteUsage())
  n3 = NamedRerouteUsage()
  FunctionOutput(None=SetMaterialAttributes!(
    MaterialAttributes=@n1.MaterialAttributes
    WorldPositionOffset=Add(A=@n2, B=@n3)
    Normal=NamedRerouteDeclaration(None=@n1.Normal) {name=OriginalNormal, nodeColor=<rgba 0 1 0.8125 1>}
    EmissiveColor=NamedRerouteUsage()
   ) {attributeSetTypes [#3, #1, #2]})
  n4 = Fmod(A=Time(), B=Constant() {r=3})
  Multiply(A=MaterialFunctionCall(Number=Floor(None=@n4)) {materialFunction=&DebugScalarValues}.Result)
  n5 = CurveAtlasRowParameter(CurveTime=NamedRerouteUsage()) {
   atlas=/DropPodParentGameplay/Assets/Textures/Curves/CA_DropPodParent
   curve=/DropPodParentGameplay/Assets/Textures/Curves/C_DropPodParent_01
   parameterName=DropPod_CurvesAnimation
  }
  NamedRerouteDeclaration(None=LinearInterpolate(A=Constant3Vector(), B=Constant3Vector() {constant=<rgba 0 0 8 0>}, Alpha=@n5.R) {constA=1, constB=0.8}) {name=Stomp, nodeColor=<rgba 0 1 0.4844 1>}
  n6 = Subtract(A=Time(), B=ScalarParameter() {defaultValue=-99999, group=WPO_Animation, parameterName=GameTimeFromBP, sortPriority=100})
  n7 = Add(A=ScalarParameter() {defaultValue=1, group=WPO_Animation, parameterName=Animation_Duration, sortPriority=101}) {constB=0}
  NamedRerouteDeclaration(None=Saturate(None=Divide(A=@n6, B=@n7))) {name=TimeNative, nodeColor=<rgba 0.2891 1 0 1>}
  n8 = Multiply(
   A=NamedRerouteDeclaration(None=@n1.EmissiveColor) {name=OriginalEmissive, nodeColor=<rgba 0 1 0.4375 1>}
   B=LinearInterpolate(A=NamedRerouteUsage(), B=NamedRerouteUsage(), Alpha=NamedRerouteUsage())
  )
  n9 = Desaturation(None=@n8)
  n10 = VectorParameter() {parameterName=Tint_Unstable, defaultValue=<rgba 10 0.01 0 0.2>}
  n11 = Multiply(A=@n10.RGB, B=@n10.RGB)
  n12 = FunctionInput(Preview=Constant() {r=1}) {inputName=TA Mask, sortPriority=1}
  n13 = ComponentMask(None=@n12) {g=true}
  n14 = NamedRerouteDeclaration(None=ComponentMask(None=@n12) {b=true}) {name=EmissiveMask, nodeColor=<rgba 0.2656 1 0 1>}
  n15 = LinearInterpolate(
   A=@n8
   B=LinearInterpolate(A=@n8, B=Multiply(A=@n9, B=@n11), Alpha=NamedRerouteUsage())
   Alpha=Add(A=NamedRerouteDeclaration(None=@n13) {name=ScreenMask, nodeColor=<rgba 0.8281 1 0 1>}, B=@n14)
  )
  n16 = NamedRerouteUsage()
  NamedRerouteDeclaration(None=LinearInterpolate(A=@n15, B=LinearInterpolate(B=@n15, Alpha=@n16), Alpha=@n14)) {name=OUT_Emissive, nodeColor=<rgba 0.4062 1 0 1>}
  n17 = StaticSwitchParameter(True=@n4, False=ScalarParameter() {group=WPO_Animation, parameterName=CurrentStage, sortPriority=3}) {parameterName=DebugMode}
  n18 = LinearInterpolate(A=Subtract(A=@n17), B=@n17, Alpha=NamedRerouteUsage()) {desc=Native lerp toward}
  n19 = Constant4Vector() {desc=Stage 0 to 1, constant=<rgba 0 1 0 1>}
  NamedRerouteDeclaration(None=Saturate(None=MaterialFunctionCall(Input=@n18, Input Low=@n19.R, Input High=@n19.R, Target Low=@n19.R, Target High=@n19.R) {materialFunction=&RemapValueRange}.Result)) {name=Stage1_Curve, nodeColor=<rgba 0 1 0.9297 1>}
  n20 = Constant4Vector() {desc=Stage 1 to 2, constant=<rgba 1 2 0 1>}
  NamedRerouteDeclaration(None=Saturate(None=MaterialFunctionCall(Input=@n18, Input Low=@n20.R, Input High=@n20.R, Target Low=@n20.R, Target High=@n20.R) {materialFunction=&RemapValueRange}.Result)) {name=Stage2_Curve, nodeColor=<rgba 0 1 0.9297 1>}
  n21 = Constant4Vector() {desc=Stage 2 to 3, constant=<rgba 2 3 0 1>}
  NamedRerouteDeclaration(None=Saturate(None=MaterialFunctionCall(Input=@n18, Input Low=@n21.R, Input High=@n21.R, Target Low=@n21.R, Target High=@n21.R) {materialFunction=&RemapValueRange}.Result)) {name=Stage3_Curve, nodeColor=<rgba 0 1 0.9297 1>}
  n22 = Constant4Vector() {desc=Stage 3 to 4, constant=<rgba 4 5 0 1>}
  NamedRerouteDeclaration(None=Saturate(None=MaterialFunctionCall(Input=@n18, Input Low=@n22.R, Input High=@n22.R, Target Low=@n22.R, Target High=@n22.R) {materialFunction=&RemapValueRange}.Result)) {name=Stage4_Curve, nodeColor=<rgba 0 1 0.9297 1>}
  n23 = Constant4Vector() {desc=Stage 4 to 5, constant=<rgba 4 5 0 1>}
  NamedRerouteDeclaration(None=Saturate(None=MaterialFunctionCall(Input=@n18, Input Low=@n23.R, Input High=@n23.R, Target Low=@n23.R, Target High=@n23.R) {materialFunction=&RemapValueRange}.Result)) {name=Stage5_Curve, nodeColor=<rgba 0 1 0.9297 1>}
  n24 = CurveAtlasRowParameter(CurveTime=NamedRerouteUsage()) {
   atlas=&CA_DropPodParent
   curve=&C_DropPodParent_01
   parameterName=DropPod_CurvesAnimation
  }
  NamedRerouteDeclaration(None=LinearInterpolate(B=NamedRerouteUsage(), Alpha=@n24.B)) {name=Stage02_Shake, nodeColor=<rgba 1 0 0.9609 1>}
  NamedRerouteDeclaration(None=LinearInterpolate(A=Constant3Vector(), B=NamedRerouteUsage(), Alpha=@n5.G)) {name=Stage01_Shake, nodeColor=<rgba 0 1 0.3672 1>}
  n25 = Constant3Vector() {desc=Random, constant=<rgba -0.3864 0.6 0.4464 0>}
  n26 = Add(
   A=Multiply(A=Time(), B=ScalarParameter() {defaultValue=100, group=WPO_Wobble, parameterName=Shake_Frequence, sortPriority=1})
   B=ScalarParameter() {group=WPO_Wobble, parameterName=Wobble_Offset, sortPriority=2}
  )
  n27 = MaterialFunctionCall(Float3=ObjectPositionWS()) {materialFunction=&BreakOutFloat3Components}
  n28 = Sine(None=Multiply(A=@n25.R, B=Add(A=@n26, B=@n27.R))) {period=6.283}
  n29 = Sine(None=Multiply(A=@n25.G, B=Add(A=@n26, B=@n27.G))) {period=6.283}
  n30 = Sine(None=Multiply(A=@n25.B, B=Add(A=@n26, B=@n27.B))) {period=6.283}
  n31 = MaterialFunctionCall(Float3=Multiply(
    A=MaterialFunctionCall(X=@n28, Y=@n29, Z=@n30) {materialFunction=&MakeFloat3}.Result
    B=ScalarParameter() {defaultValue=50, group=WPO_Wobble, parameterName=Shake_Amplitude, sortPriority=4}
   ) {desc=wind factor}) {materialFunction=&BreakOutFloat3Components}
  n32 = MaterialFunctionCall(X=Add(A=@n31.G, B=@n31.G), Y=Add(A=@n31.R, B=@n31.R), Z=Add(A=@n31.R, B=@n31.R)) {materialFunction=&MakeFloat3}
  n33 = Multiply(A=ScalarParameter() {defaultValue=0.015, group=WPO_Wobble, parameterName=Shake_Strength, sortPriority=3}) {constB=0.1, desc=adjustStrength}
  n34 = Multiply(
   A=NamedRerouteDeclaration(None=Multiply(A=@n32.Result, B=@n33)) {name=ShakeNoise, nodeColor=<rgba 0 1 0.3203 1>}
   B=Sine(None=Time() {bOverride_Period=true, period=1}) {period=6.28}
  ) {constB=3}
  NamedRerouteDeclaration(None=Saturate(None=Frac(None=@n34))) {name=BlinkNoise, nodeColor=<rgba 1 0.8438 0 1>}
  n35 = NamedRerouteUsage()
  NamedRerouteDeclaration(None=LinearInterpolate(B=LinearInterpolate(Alpha=@n35) {constA=0.5}, Alpha=@n5.G) {constA=1}) {name=Stage01_Blink, nodeColor=<rgba 0.5234 0 1 1>}
  n36 = Time() {bOverride_Period=true, period=1}
  n37 = Frac(None=Sine(None=Multiply(A=@n36) {constB=2}))
  NamedRerouteDeclaration(None=Multiply(A=@n5.G, B=Abs(None=@n37))) {name=ShakeCoolDownCurve, nodeColor=<rgba 1 0.75 0 1>}
  n38 = NamedRerouteUsage()
  NamedRerouteDeclaration(None=LinearInterpolate(B=LinearInterpolate(Alpha=@n38), Alpha=@n24.B) {constA=1}) {name=Stage02_Blink, nodeColor=<rgba 0.5234 0 1 1>}
 ]}
