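# Material function that applies a masked tint to the base color of incoming material attributes, named for the DropPod TidalCrane.
# The base color is desaturated, hue-shifted and multiplied by Tint_Color, then blended over the original through a parameter-selected channel of the TA Mask input.
# Both the desaturation amount and hue shift percentage read the same Tint_Desaturate parameter name, so one value drives both.
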MaterialFunction /DropPodParentGameplay/Assets/Materials/MF_DropPod_TidalCrane_Tint
graph {nodes [
  n1 = GetMaterialAttributes(None=FunctionInput() {inputName=Result, inputType=FunctionInput_MaterialAttributes}) {attributeGetTypes [#1]}
  n2 = ScalarParameter() {group=Tint, parameterName=Tint_Desaturate}
  n3 = Multiply(
   A=MaterialFunctionCall(Hue Shift Percentage=ScalarParameter() {group=Tint, parameterName=Tint_Desaturate}, Texture=Desaturation(None=@n1.BaseColor, Fraction=@n2)) {materialFunction=&HueShift}.Result
   B=VectorParameter() {group=Tint, parameterName=Tint_Color, defaultValue=<rgba 1 1 1 1>}.RGB
  )
  n4 = ChannelMaskParameter(None=FunctionInput() {inputName=TA Mask, sortPriority=1}) {group=Tint, parameterName=Tint_TAMask Channel, sortPriority=0}
  FunctionOutput(None=SetMaterialAttributes!(MaterialAttributes=@n1.MaterialAttributes, BaseColor=LinearInterpolate(A=@n1.BaseColor, B=@n3, Alpha=@n4.R)) {attributeSetTypes [#1]}) {outputName=SetMaterialAttributes}
 ]}
