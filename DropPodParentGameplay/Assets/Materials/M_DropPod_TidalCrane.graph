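# Master material for the TidalCrane drop pod, layering Fortnite's shared building-style base functions with pod-specific tint, screen and tech-art effects.
# MF_GigaMix_2023 feeds shared damage, wind and vibration functions, then DirectFX, Veneer, jungle-biome and RVT blending, with a backface normal fix.
# Low-quality mobile flattens specular and roughness.
# An optional SRM texture then drives surface response, followed by emissive, metal LUT and TidalCrane functions masked by tech-art inputs.
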Material /DropPodParentGameplay/Assets/Materials/M_DropPod_TidalCrane
graph {
 properties {
  bUseMaterialAttributes=true
  bUsedWithClothing=true
  bUsedWithMeshDeformer=true
  bUsedWithMorphTargets=true
  bUsedWithNanite=true
  bUsedWithNiagaraMeshParticles=true
  bUsedWithSkeletalMesh=true
  pixelDepthOffsetMode=PDOM_Legacy
  shadingModel=MSM_FromMaterialExpression
  usedShadingModels=Default Lit | Subsurface | Clear Coat | Two Sided Foliage
 }
 nodes [
  n1 = MaterialFunctionCall() {materialFunction=/Game/Global/Base4Layers_2022/MaterialFunctions/MF_GigaMix_2023}
  MaterialExpressionReroute_70 = Reroute(None=@n1.Result)
  n2 = ScalarParameter() {
   desc=Documentation link
   group="0 (Expand)  Documentation (RMB - Copy display name) and Debug view"
   parameterName=https://confluence.it.epicgames.com/display/FORT/Fortnite+Base+Material+2022
  }
  n3 = SetMaterialAttributes!(MaterialAttributes=@MaterialExpressionReroute_70.MaterialAttributes, World Position Offset=Multiply(A=@n2)) {attributeSetTypes [#1]}
  MaterialExpressionReroute_71 = Reroute(None=@n3)
  MaterialExpressionReroute_69 = Reroute(None=@MaterialExpressionReroute_71.Output)
  n4 = MaterialFunctionCall(Material_In=@MaterialExpressionReroute_69.MaterialAttributes) {materialFunction=&MF_FN_DestructibleDamage}
  n5 = MaterialFunctionCall!(
   In=MaterialFunctionCall!(Material Attributes=MaterialFunctionCall!(In=@n4.Material_Out) {materialFunction=/Game/Athena/WinterArt/Materials/MF_WPO_WindNew}.Result) {materialFunction=/Game/Creative/Sets/MilitaryBase/Props/MF_FanSpin}.Result
   UseSphericalMappingOverride
  ) {materialFunction=/Game/Packages/Fortress_SharedMaterials/Base_Material_Functions/MF_CurieFallbackEffect}
  n6 = MaterialFunctionCall!(In=MaterialFunctionCall!(In=MaterialFunctionCall!(In=@n5.Result, Custom Mask, Custom Rotation Axis) {materialFunction=/Game/Athena/WinterArt/Materials/MF_WindWPO_v2}.Result) {materialFunction=/Game/Global/Base4Layers_2022/MaterialFunctions/MF_SnowAndIceArchitecture}.Result) {materialFunction=/Game/Global/Base4Layers_2022/MaterialFunctions/MF_VibrationRattle}
  MaterialExpressionReroute_68 = Reroute(None=@n6.Result)
  Composite!(None=@MaterialExpressionReroute_68.Result) {desc=Transform to World in case we will use RVT after this point, subgraphName=Tangent to World}
  Composite!(MaterialAttributes=@n1.Result) {subgraphName=Documentation Link}
  Composite(MaterialAttributes=@MaterialExpressionReroute_71.Output) {desc=Game Shared WPO and Fire Effect, subgraphName=WPO And Shared Functions}
  MaterialExpressionReroute_1 = Reroute(None=@n9.RVT H)
  n7 = MaterialFunctionCall!(In=MaterialFunctionCall!(In=@MaterialExpressionReroute_68.Result) {materialFunction=/Game/ArtTools/Tech_Art/DirectFX/Assets/MaterialFunctions/MF_DirectFX}.Result) {materialFunction=/Game/ArtTools/Tech_Art/Veneer/MaterialFunctions/MF_Veneer}
  MaterialExpressionReroute_8 = Reroute(None=@n7.Result)
  n8 = MaterialFunctionCall!(AO, RVT_Height=@MaterialExpressionReroute_1."RVT_Height (S)", For Buildings=@MaterialExpressionReroute_8."For Moss (MA)", For Moss) {materialFunction=/Game/Environments/Asteria/Rocks/Materials/MaterialFunctions/MF_JungleBiome}
  MaterialExpressionReroute_4 = Reroute(None=@n8.Buildings Jungl)
  n9 = MaterialFunctionCall!(
   In=@MaterialExpressionReroute_4.Result
   "RVT Height Map (Roughness)"=GetMaterialAttributes(None=@n7.Result) {attributeGetTypes [#2]}.Roughness
  ) {materialFunction=/Game/Global/Base4Layers_2022/MaterialFunctions/MF_RVTBlending}
  Composite!(InputPin=@n9.Result) {subgraphName=Mobile Opt}
  Composite!("For Moss (MA)"=@n7.Result, "RVT_Height (S)"=@n9.RVT H) {subgraphName="Seasonal "}
  Constant3Vector() {constant=<rgba 0.5 0 0.7 1>}
  MaterialExpressionReroute_7 = Reroute(None=@n9.Result)
  n10 = GetMaterialAttributes(None=@MaterialExpressionReroute_7.InputPin)
  n11 = Constant() {r=0.5}
  n12 = Constant() {r=0.8}
  n13 = ShadingPathSwitch(
   Default=@MaterialExpressionReroute_7.InputPin
   Mobile=QualitySwitch(
    Default=@MaterialExpressionReroute_7.InputPin
    Low=SetMaterialAttributes!(MaterialAttributes=@n10.MaterialAttributes, Specular=@n11, Roughness=@n12) {attributeSetTypes [#3, #2]}
    High=@MaterialExpressionReroute_7.InputPin
    Medium=@MaterialExpressionReroute_7.InputPin
    Epic=@MaterialExpressionReroute_7.InputPin
   )
  )
  MaterialExpressionReroute_6 = Reroute(None=@n13)
  n14 = GetMaterialAttributes(None=MaterialFunctionCall(MA In=@MaterialExpressionReroute_6.Output) {materialFunction=&MF_CreativeMoveTool}.MA Out) {attributeGetTypes [#4, #3, #2]}
  n15 = TextureSampleParameter2D() {
   desc="R = Spec , G = Roughness, B= Metal"
   group=00_Texture Inputs
   parameterName=SRM
   samplerSource=SSM_Wrap_WorldGroupSettings
   samplerType=SAMPLERTYPE_Masks
   sortPriority=1
   texture=&T_Fortnite_Default_S
  }
  n16 = MaterialFunctionCall(Float3=StaticSwitchParameter(
    True=@n15.RGB
    False=MaterialFunctionCall(X=@n14.Specular, Y=@n14.Specular, Z=@n14.Specular) {materialFunction=&MakeFloat3}.Result
   ) {defaultValue=true, group=09_DisableTextureSamplers, parameterName=useSRM}) {materialFunction=&BreakOutFloat3Components}
  n17 = ScalarParameter() {defaultValue=4, group=00_Texture Inputs, parameterName=MetalMaskPowerIntensity}
  n18 = StaticSwitchParameter(True=Saturate(None=Power(Base=@n16.B, Exp=@n17) {constExponent=4}), False=@n16.B) {group=00_Texture Inputs, parameterName=MetalMaskPower}
  n19 = MaterialFunctionCall!(
   Alpha=NamedRerouteUsage()
   Materials=MaterialFunctionCall!(
    Material=SetMaterialAttributes!(MaterialAttributes=@n14.MaterialAttributes, Metallic=@n18, Specular=@n16.R, Roughness=@n16.R) {attributeSetTypes [#4, #3, #2]}
    Emissive UVs=NamedRerouteUsage()
    Fresnel=NamedRerouteUsage()
    Alpha=NamedRerouteUsage()
   ) {materialFunction=/Game/Global/Materials/Character_Parent_2023/Functions/MF_CharacterParent_2023_Emissive}.Result
  ) {materialFunction=/Game/Global/MaterialFunctions/TechArt/MF_MetalLUT_TechArt}
  n20 = NamedRerouteUsage()
  MaterialFunctionCall!(
   IN MA=MaterialFunctionCall!(
    Result=MaterialFunctionCall!(Result=@n19.Result, TA Mask=@n20) {materialFunction=&/MF_DropPod_TidalCrane_Tint}.SetMaterialAttributes
    TA Mask=NamedRerouteUsage()
   ) {materialFunction=&/MF_DropPod_TidalCrane_Screen}.SetMaterialAttributes
   TA Mask=NamedRerouteUsage()
  ) {materialFunction=&/MF_DropPod_TidalCrane_TA}
  n21 = TextureSampleParameter2D(UVs=NamedRerouteUsage()) {
   group=05_Tech Art Masks
   parameterName=TechArtMask
   samplerSource=SSM_Wrap_WorldGroupSettings
   samplerType=SAMPLERTYPE_Masks
   sortPriority=0
   texture=&T_BlackMask
  }
  n22 = StaticSwitchParameter(True=@n21.RGB, False=Constant()) {defaultValue=true, group=09_DisableTextureSamplers, parameterName=useTechArtMask}
  n23 = ScalarParameter() {desc=Set to 1 to indicate to use TechArtMask for opacity in Hologram, group=05_Tech Art Masks, parameterName=UseTechArtMaskForOpacity}
  NamedRerouteDeclaration(None=LinearInterpolate(A=@n22, B=@n22, Alpha=@n23)) {name=Tech Art Mask, nodeColor=<rgba 1 0.7031 0 1>}
  NamedRerouteDeclaration(None=Fresnel(ExponentIn=Constant() {r=1}) {baseReflectFraction=0}) {name=Fresnel, nodeColor=<rgba 1 0.3047 0 1>}
  NamedRerouteDeclaration(None=StaticSwitchParameter(True=TextureCoordinate() {coordinateIndex=1}, False=TextureCoordinate()) {group=00_Texture Inputs, parameterName=useSecondUVChannel, sortPriority=40}) {name=UVs, nodeColor=<rgba 0.2891 0 1 1>}
  MaterialExpressionReroute_3 = Reroute(None=@MaterialExpressionReroute_68.Result)
  n24 = GetMaterialAttributes(None=@MaterialExpressionReroute_3.Input) {attributeGetTypes [#5]}
  n25 = Transform(None=@n24.Normal)
  n26 = LinearInterpolate(A=Multiply(A=@n25) {constB=-1}, B=@n25, Alpha=Saturate(None=TwoSidedSign())) {desc="Fixes normals\r\non backfaces"}
  n27 = SetMaterialAttributes!(
   MaterialAttributes=@n24.MaterialAttributes
   Normal=StaticSwitchParameter(True=MaterialProxyReplace(Realtime=@n26, MaterialProxy=@n24.Normal), False=@n26) {parameterName=Force TangentSpaceNormalsFor HLOD ProxyReplace}
  ) {attributeSetTypes [#5]}
  PinBase!(Output=@n27) {reroutePins [Output {expression=&self:MaterialExpressionReroute_2}]}
  PinBase!() {reroutePins [Input {expression=&self:MaterialExpressionReroute_3}]}
  PinBase() {reroutePins [MaterialAttributes {expression=&self:MaterialExpressionReroute_70}]}
  PinBase!(Output=@n3) {reroutePins [Output {expression=&self:MaterialExpressionReroute_71}]}
  PinBase!() {reroutePins [MaterialAttributes {expression=&self:MaterialExpressionReroute_69}]}
  PinBase!(Result=@n6.Result) {reroutePins [Result {expression=&self:MaterialExpressionReroute_68}]}
  PinBase!() {reroutePins [InputPin {expression=&self:MaterialExpressionReroute_7}]}
  PinBase!(Output=@n13) {reroutePins [Output {expression=&self:MaterialExpressionReroute_6}]}
  PinBase!() {reroutePins [
    "For Moss (MA)" {expression=&self:MaterialExpressionReroute_8}
    "RVT_Height (S)" {expression=&self:MaterialExpressionReroute_1}
   ]}
  PinBase!(Result=@n8.Buildings Jungl) {reroutePins [Result {expression=&self:MaterialExpressionReroute_4}]}
  MaterialExpressionReroute_2 = Reroute(None=@n27)
 ]
}
