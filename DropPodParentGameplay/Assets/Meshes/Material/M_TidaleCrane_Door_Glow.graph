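# Default Lit material for the glow on a Tidale crane door in DropPod gameplay, judging by its name and path.
# It exposes a Color vector parameter, defaulting to an over-bright red, and a Normal texture parameter defaulting to a flat hard-surface normal map.
# The graph records no links from either parameter to material outputs.
# Usage is enabled for instanced static, skeletal, mesh-deformer and Niagara mesh-particle rendering.
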Material /DropPodParentGameplay/Assets/Meshes/Material/M_TidaleCrane_Door_Glow
graph {
 properties {
  bAutomaticallySetUsageInEditor=false
  bUsedWithInstancedStaticMeshes=true
  bUsedWithMeshDeformer=true
  bUsedWithNiagaraMeshParticles=true
  bUsedWithSkeletalMesh=true
  pixelDepthOffsetMode=PDOM_Legacy
  usedShadingModels=Default Lit
 }
 nodes [
  TextureSampleParameter2D() {parameterName=Normal, samplerType=SAMPLERTYPE_Normal, texture=/Game/Athena/Prototype/HardSurface/Textures/HS_FlatNormal}
  VectorParameter() {parameterName=Color, defaultValue=<rgba 3 0 0.05 1>}
 ]
}
